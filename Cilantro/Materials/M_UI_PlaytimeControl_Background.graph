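# UI material for the playtime control background: a dark navy backdrop with a scrolling pattern and a blue radial glow.
# Centered UVs are tiled, aspect-fitted and rotated by parameter, then scrolled horizontally over time to sample a sleep-themed texture.
# The texture serves as a distance field, thresholded with anti-aliasing and blended faintly into the base using color dodge.
# An offset, scalable radial gradient adds the glow, and its eased inverse forms a pattern mask.
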Material /Cilantro/Materials/M_UI_PlaytimeControl_Background
graph {
 properties {
  bAutomaticallySetUsageInEditor=false
  bUsedWithStaticMesh=false
  blendMode=BLEND_Translucent
  materialDomain=MD_UI
  usedShadingModels=Default Lit
 }
 nodes [
  n1 = Constant3Vector() {constant=<rgba 0 0.0012 0.0194 1>}
  n2 = Multiply(
   A=Subtract(A=MaterialFunctionCall() {materialFunction=&GetUserInterfaceUV}.Normalized UV) {constB=0.5}
   B=ScalarParameter() {defaultValue=4, parameterName=Tiling}
  ) {constB=4}
  n3 = Rotator(
   Coordinate=MaterialFunctionCall(UV=Add(A=@n2) {constB=0.5}) {materialFunction=&MF_UI_UVScaling}.Scaled to Fit
   Time=ScalarParameter() {parameterName=Rotation}
  ) {speed=6.2832}
  n4 = AppendVector(A=Multiply(A=Time(), B=ScalarParameter() {defaultValue=1, parameterName=Speed}), B=Constant())
  n5 = MaterialFunctionCall(
   DistanceField=TextureSample(
    UVs=Subtract(A=@n3, B=@n4)
    Tex=TextureObjectParameter() {parameterName=Background Pattern, samplerType=SAMPLERTYPE_Alpha, texture=/Cilantro/Textures/T_UI_PlaytimeControl_Sleep}
   ) {samplerSource=SSM_Wrap_WorldGroupSettings}.R
   Threshold=Constant() {r=0.5}
  ) {materialFunction=&MF_UI_AntiAliasing}
  n6 = MaterialFunctionCall!(Blend=Multiply(A=@n5.Outer) {constB=0.5}, Base=@n1) {materialFunction=/Engine/Functions/Engine_MaterialFunctions03/Blends/Blend_ColorDodge}
  n7 = LinearInterpolate(A=NamedRerouteUsage(), B=@n1, Alpha=OneMinus(None=NamedRerouteUsage()))
  n8 = Multiply(A=ScalarParameter() {defaultValue=1, parameterName=Pattern Opacity, sliderMax=1}) {constB=0.1125}
  n9 = NamedRerouteUsage()
  n10 = Add(
   A=MaterialFunctionCall() {materialFunction=&GetUserInterfaceUV}.Normalized UV
   B=AppendVector(A=ScalarParameter() {parameterName=Radial Offset X}, B=ScalarParameter() {parameterName=Radial Offset Y})
  )
  n11 = AppendVector(A=ScalarParameter() {defaultValue=1, parameterName=Radial Scale X}, B=ScalarParameter() {defaultValue=1, parameterName=Radial Scale Y})
  n12 = MaterialFunctionCall(UV=MaterialFunctionCall(
    UV=MaterialFunctionCall(UVs=@n10, Texture Scale=@n11) {materialFunction=&ScaleUVsByCenter}.UVs
    TextureSize=MaterialFunctionCall() {materialFunction=&GetUserInterfaceUV}.Pixel Size
   ) {materialFunction=&MF_UI_UVScaling}.Scaled to Fit) {materialFunction=&MF_UI_RadialGradient}
  LinearInterpolate(
   A=LinearInterpolate(A=@n1, B=LinearInterpolate(A=@n6.Result, B=@n7, Alpha=@n5.Outer) {constAlpha=0.1}, Alpha=Multiply(A=@n8, B=@n9))
   B=NamedRerouteDeclaration(None=Constant3Vector() {constant=<rgba 0 0.0437 0.1356 1>}) {name=Glow Color, nodeColor=<rgba 0 0 0 1>}
   Alpha=@n12.Smooth
  )
  NamedRerouteDeclaration(None=OneMinus(None=MaterialFunctionCall("Value (0-1)"=@n12.Smooth) {materialFunction=&MF_UI_EaseCurves}.EaseOutQuint)) {name=Pattern Mask, nodeColor=<rgba 0 0 0 1>}
 ]
}
