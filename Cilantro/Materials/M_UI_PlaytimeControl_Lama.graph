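# Translucent UI material that animates the llama texture for a playtime control screen element.
# AnimRatio times OverrideAnim forms an Anim Value that slides the normalized UI UVs horizontally into place and fades in a looping vertical bob, built from a Time-driven triangle wave.
# The offset UVs pass through MF_UI_UVScaling, then a Rotator whose Time input is a blend of two constants rather than the clock, giving a controlled tilt before the texture sample.
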Material /Cilantro/Materials/M_UI_PlaytimeControl_Lama
graph {
 properties {
  bAutomaticallySetUsageInEditor=false
  bUsedWithStaticMesh=false
  blendMode=BLEND_Translucent
  materialDomain=MD_UI
  usedShadingModels=Default Lit
 }
 nodes [
  n1 = NamedRerouteDeclaration(None=Multiply(
    A=ScalarParameter() {defaultValue=1, parameterName=AnimRatio, sliderMax=1}
    B=ScalarParameter() {defaultValue=1, parameterName=OverrideAnim, sliderMax=1}
   )) {name=Anim Value, nodeColor=<rgba 0 0 0 1>}
  n2 = Constant() {r=0.01}
  n3 = Frac(None=Multiply(A=Time()) {constB=0.125})
  n4 = OneMinus(None=@n3)
  n5 = LinearInterpolate(A=Multiply(A=Add(B=@n2) {constA=0.04}) {constB=-1}, B=@n2, Alpha=Multiply(A=Min(A=@n3, B=@n4)) {constB=2}) {constA=-0.1, constB=0.1}
  n6 = Subtract(
   A=MaterialFunctionCall() {materialFunction=&GetUserInterfaceUV}.Normalized UV
   B=AppendVector(A=LinearInterpolate(Alpha=@n1) {constA=1, constB=0}, B=LinearInterpolate(B=@n5, Alpha=@n1))
  )
  n7 = NamedRerouteUsage()
  TextureSample(
   UVs=Rotator(
    Coordinate=MaterialFunctionCall(UV=@n6) {materialFunction=&MF_UI_UVScaling}.ScaleY
    Time=LinearInterpolate(Alpha=@n7) {constA=0.125, constB=0.0625}
   ) {speed=6.2832}
   Tex=TextureObject() {texture=/Cilantro/Textures/T_UI_PlaytimeControl_Lama}
  )
 ]
}
