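# Translucent UI material that draws an animated spinner ring for the content-on-demand item download indicator.
# Time-driven UV rotation, with a sine wobble from MaxSpeed, feeds a radial angle, which a slower sine offsets between MinFill and MaxFill to grow and shrink the arc.
# A steep SmoothStep cuts the arc, multiplied by a ring from two subtracted sphere gradients and a Color overlay blend of the angle.
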Material /Game/UI/ContentOnDemand/ItemDownloadIndicator/ItemDownloadIndicator_SpinnerMaterial
graph {
 properties {
  bUsedWithStaticMesh=false
  blendMode=BLEND_Translucent
  materialDomain=MD_UI
  pixelDepthOffsetMode=PDOM_Legacy
  shadingModel=MSM_Unlit
  usedShadingModels=Unlit
 }
 nodes [
  n1 = ScalarParameter() {defaultValue=0.4, parameterName=MinFill}
  n2 = Subtract(A=ScalarParameter() {defaultValue=0.9, parameterName=MaxFill}, B=@n1)
  n3 = ScalarParameter() {defaultValue=1.8, parameterName=Speed}
  n4 = MaterialFunctionCall(Value=MaterialFunctionCall(Speed=Multiply(A=@n3) {constB=0.6}) {materialFunction=&TimeWithSpeedVariable}.Frac Time) {materialFunction=&LinearSine}
  n5 = MaterialFunctionCall(Speed=Multiply(A=@n3)) {materialFunction=&TimeWithSpeedVariable}
  n6 = Multiply(
   A=MaterialFunctionCall(Value=@n5.Frac Time) {materialFunction=&LinearSine}."Rounded Linear Sine "
   B=ScalarParameter() {defaultValue=0.05, parameterName=MaxSpeed}
  )
  n7 = MaterialFunctionCall(UVs=TextureCoordinate(), "Rotation Angle (0-1)"=Multiply(A=Add(A=@n5.Frac Time, B=@n6)) {constB=-1}) {materialFunction=&CustomRotator}
  n8 = MaterialFunctionCall(Vector or UVs=Multiply(A=Subtract(A=@n7.Rotated Values) {constB=0.5}, B=Constant2Vector() {g=2, r=2}) {constB=2}) {materialFunction=&VectorToRadialValue}
  n9 = Add(A=Add(A=@n1, B=Multiply(A=@n2, B=@n4."Rounded Linear Sine ")), B=@n8.Vector Converted to Angle)
  Floor(None=@n9)
  n10 = MaterialFunctionCall(Base=VectorParameter() {parameterName=Color, defaultValue=<rgba 0.5 0.5 0.5 1>}.RGB, Blend=@n8.Vector Converted to Angle) {materialFunction=&Blend_Overlay}
  LinearInterpolate(Alpha=@n10.Result)
  n11 = Constant() {r=1}
  n12 = Constant() {r=0.95}
  n13 = Clamp(None=Multiply(A=MaterialFunctionCall(Alpha=@n9, Min=@n11, Max=@n12) {materialFunction=&SmoothStep}.Result) {constB=50})
  n14 = MaterialFunctionCall!(UVs, CenterPosition, Radius=Constant() {r=0.5}) {materialFunction=/Engine/Functions/Engine_MaterialFunctions01/Gradient/SphereGradient-2D}
  n15 = ScalarParameter() {defaultValue=0.3, parameterName=MinInnerRadius}
  n16 = Multiply(B=Subtract(A=ScalarParameter() {defaultValue=0.6, parameterName=MaxInnerRadius}, B=@n15)) {constA=1}
  n17 = MaterialFunctionCall(Radius=Multiply(A=Add(A=@n16, B=@n15)) {constB=0.5}) {materialFunction=&SphereGradient-2D}
  n18 = Subtract(A=Clamp(None=Multiply(A=@n14.Result 0-1) {constB=2}), B=Clamp(None=Multiply(A=@n17.Result 0-1) {constB=2}))
  MaterialFunctionCall(Alpha=Multiply(A=@n10.Result, B=Multiply(A=@n13, B=@n18)), Min=Constant(), Max=Constant() {r=0.2}) {materialFunction=&SmoothStep}
  MaterialFunctionCall(Value=MaterialFunctionCall(Speed=Multiply(A=@n3) {constB=0.05}) {materialFunction=&TimeWithSpeedVariable}.Frac Time) {materialFunction=&LinearSine}
  VectorParameter() {parameterName=Color, defaultValue=<rgba 1 1 1 1>}
 ]
}
