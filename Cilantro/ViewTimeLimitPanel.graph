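# Sidebar panel showing a player's parental time limits and time reports: time left and used, allowed window, and report status.
# Update handlers fill the text from PlayerPlaytimeVM and reveal the ask/add-time buttons at 30 minutes or less remaining.
# Ask sends a request whose status switches Switcher_State between sending and sent/failed pages; Add opens a PIN time grant.
# OnInitialized retries every 0.05s until the view model is valid; sidebar visibility toggles refresh listening.
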WidgetBlueprint /Cilantro/ViewTimeLimitPanel
graphs [
 BP_GetDesiredFocusTarget {nodes [
   Get Desired Focus Target()
   n1 = GetActiveWidget!(self:Widget Switcher Object Reference=Get Switcher_State() -> (Switcher_State:Common Activatable Widget Switcher Object Reference)) -> (ReturnValue:Widget Object Reference)
   > n2 = Branch(Condition="Equal (Object)"(A=@n1, B=Get Box_ContentEmailSent() -> (Box_ContentEmailSent:Vertical Box Object Reference)))
   > Return Node!(execute, ReturnValue:Widget Object Reference=Get Button_Okay() -> (Button_Okay:Main Menu Button Object Reference))
   n3 = Branch(execute=@n2.else, Condition="Equal (Object)"(A=@n1, B=Get Box_ContentFirstPage() -> (Box_ContentFirstPage:Vertical Box Object Reference)))
   n4 = Get Button_AskMoreTime() -> (Button_AskMoreTime:Main Menu Button Object Reference)
   > n5 = Branch(Condition=AND Boolean(A=IsVisible(self=@n4), B=GetIsEnabled!(self:Widget Object Reference=@n4) -> (ReturnValue:bool)))
   > Return Node(ReturnValue=@n4)
   n6 = Get Button_AddMoreTIme() -> (Button_AddMoreTIme:Main Menu Button Object Reference)
   n7 = Branch(execute=@n5.else, Condition=AND Boolean(A=IsVisible(self=@n6), B=GetIsEnabled(self=@n6)))
   > Return Node(ReturnValue=@n6)
   Return Node(execute=[@n7.else, @n3.else], ReturnValue=Self-Reference())
  ]}
 EventGraph {nodes [
   "On Clicked (Button_AskMoreTime)"!() -> (OutputDelegate, then, Button:Common Button Base Object Reference) {is=title K2Node_ComponentBoundEvent}
   > RequestAdditionalPlaytime!(
    execute
    self:Player Playtime View Model Object Reference=Get PlayerPlaytimeVM() -> (PlayerPlaytimeVM:Player Playtime View Model Object Reference)
    ActivityType:EActivityLimitType Enum=CombinedTime
   ) -> (then)
   n1 = Get Button_AskMoreTime()
   > SetIsEnabled(self=@n1, bInIsEnabled=false)
   > Fire Event Element Interacted(Element=@n1)
   "On Clicked (Button_AddMoreTime)"!() -> (OutputDelegate, then, Button:Common Button Base Object Reference) {is=title K2Node_ComponentBoundEvent}
   > n2 = Branch(Condition=IsInSubGame!(SubGame:ESubGame Enum=Campaign) -> (ReturnValue:bool))
   > n3 = DeactivateWidget()
   OpenPinTimeGrant!(
    execute=[@n3.then, @n2.else]
    self:Player Playtime View Model Object Reference=Get PlayerPlaytimeVM()
    Context:User Widget Object Reference=Self-Reference()
   ) -> (then)
   > Fire Event Element Interacted(Element=Get Button_AddMoreTime() -> (Button_AddMoreTime:Main Menu Button Object Reference))
   "On Active Widget Index Changed BP (Switcher_State)"!() -> (OutputDelegate, then, ActiveWidget:Widget Object Reference, ActiveIndex:int=0) {is=title K2Node_ComponentBoundEvent}
   > RefreshFocus()
   "On Clicked (Button_Okay)"!() -> (OutputDelegate, then, Button:Common Button Base Object Reference) {is=title K2Node_ComponentBoundEvent}
   > SetActiveWidget(self:Widget Switcher Object Reference=Get Switcher_State(), Widget=Get Box_ContentFirstPage())
   n4 = On Request Additional Time Status Updated!() -> (OutputDelegate, then, RequestStatus:EPlaytimeRequestState Enum=None) {is=title K2Node_CustomEvent}
   > n5 = Switch on EPlaytimeRequestState!(execute, Selection:EPlaytimeRequestState Enum=@n4.RequestStatus) -> (None:Exec, Waiting:Exec, Success:Exec, Failure:Exec) {is=title K2Node_SwitchEnum}
   SetActiveWidget(execute=@n5.Failure, self=Get Switcher_State(), Widget=Get Box_ContentEmailSent())
   n6 = Get Image_Tick() -> (Image_Tick:Image Object Reference)
   > SetVisibility(self=@n6, InVisibility=Collapsed)
   n7 = Get Text_EmailDescription() -> (Text_EmailDescription:Common Text Object Reference)
   > SetVisibility(self=@n7, InVisibility=Collapsed)
   n8 = Get Text_EmailSent() -> (Text_EmailSent:Common Text Object Reference)
   > Set Text(Text="LOCTEXT(\"EMAIL FAILED\")", self=@n8)
   n9 = GetOwningPlayer()
   > n10 = Cast To FortPlayerController(Object=@n9)
   > "Fire Event Sub View Opened (Non-Localized)"!(
    execute
    FortPC:Fort Player Controller Object Reference=@n10
    ViewName:String=PlaytimeRequestEmailFailed
    AdditionalInfo:Flow Analytics Additional Info Structure
    bIsOverlay:bool=false
   ) -> (then)
   SetActiveWidget(execute=@n5.Success, self=Get Switcher_State(), Widget=Get Box_ContentEmailSent())
   > SetVisibility(self=@n6, InVisibility=SelfHitTestInvisible)
   > SetVisibility(self=@n7, InVisibility=SelfHitTestInvisible)
   > Set Text(Text="LOCTEXT(\"EMAIL SENT\")", self=@n8)
   > n11 = Cast To FortPlayerController(Object=@n9)
   > "Fire Event Sub View Opened (Non-Localized)"(FortPC=@n11, ViewName=PlaytimeRequestEmailSent, bIsOverlay=false)
   SetActiveWidget(execute=@n5.Waiting, self=Get Switcher_State(), Widget=Get Box_EmailSending() -> (Box_EmailSending:Vertical Box Object Reference))
   Event On Activated()
   > Fire Event Sub View Opened!(execute, View:Widget Object Reference=Self-Reference(), AdditionalInfo:Flow Analytics Additional Info Structure, bIsOverlay:bool=false) -> (then)
   Event OnInitialized()
   > n12 = SetActiveWidget(self=Get Switcher_State(), Widget=Get Box_ContentFirstPage())
   n13 = Get!(execute:Exec=[@n12.then, @n14.then]) -> (then:Exec, else:Exec, PlayerPlaytimeVM:Player Playtime View Model Object Reference)
   > Init(self:Player Playtime View Model Object Reference=Get PlayerPlaytimeVM())
   > Assign On Request Additional Playtime Status Update!(execute:Exec, self:Player Playtime View Model Object Reference=Get PlayerPlaytimeVM(), Delegate:"Delegate (by ref)"=Create Event()) -> (then:Exec) {is=title K2Node_AssignDelegate}
   > SetListenToRefresh!(execute, self:Player Playtime View Model Object Reference=Get PlayerPlaytimeVM(), bListenToRefresh:bool=false) -> (then)
   n14 = Delay(execute=@n13.else, Duration=0.05)
   n15 = Event OnSidebarVisibilityChanged!() -> (OutputDelegate, then, bIsVisible:bool) {is=title K2Node_Event}
   > SetListenToRefresh(self=Get PlayerPlaytimeVM(), bListenToRefresh=@n15.bIsVisible)
  ]}
 OnReportStatusUpdated {nodes [
   n1 = OnReportStatusUpdated!() -> (then, NewStatus:bool=~) {is=title K2Node_FunctionEntry}
   n2 = MakeLiteralText(Value="LOCTEXT(\"off\")")
   n3 = MakeLiteralText(Value="LOCTEXT(\"on\")")
   > "SetText (Text)"(
    self=Get Text_ReportStatus() -> (Text_ReportStatus:Common Text Object Reference)
    InText=Format Text!(Format="LOCTEXT(\"Time Reports {Value}\")", Value:Text=SelectText(A=@n3, B=@n2, bPickA=@n1)) -> (Result)
   )
   > SetVisibility(
    self=Get Text_ReportDescription() -> (Text_ReportDescription:Common Text Object Reference)
    InVisibility=Select(Option 0=Hidden, Option 1=HitTestInvisible, Index=@n1)
   )
  ]}
 OnTimePlayedUpdated {nodes [
   n1 = OnTimePlayedUpdated!() -> (then, TimeSpan:Timespan Structure) {is=title K2Node_FunctionEntry}
   n2 = "Timespan > Timespan"(
    A:Timespan Structure=@n1
    B:Timespan Structure=MakeTimespan!(Days:int=0, Hours:int=0, Minutes:int=0, Seconds:int=0, Milliseconds:int=0) -> (ReturnValue:Timespan Structure)
   ) -> (ReturnValue:bool=~) {is=title K2Node_PromotableOperator}
   n3 = "Timespan <= Timespan"(A:Timespan Structure=@n1, B:Timespan Structure=MakeTimespan(Days=1, Hours=0, Minutes=0, Seconds=0, Milliseconds=0)) -> (ReturnValue:bool=~) {is=title K2Node_PromotableOperator}
   n4 = TextToLower!(InText:"Text (by ref)"=GetTimespanAsText!(
     Timespan:"Timespan Structure (by ref)"=@n1
     bShowHoursWhenZero:bool=false
     bShowSeconds:bool=false
     FormatType:ETimespanAsTextFormat Enum=DaysHoursMinutesSeconds
     UnitDisplayType:ETimespanUnitDisplayFormat Enum=Full
     bUseSentenceCase:bool=false
    ) -> (ReturnValue:Text)) -> (ReturnValue:Text)
   > "SetText (Text)"(
    self=Get Text_PlaytimePlayed() -> (Text_PlaytimePlayed:Common Text Object Reference)
    InText=SelectText(
     A=Format Text!(Format="LOCTEXT(\"{Time} used\")", Time:Text=@n4) -> (Result)
     B="LOCTEXT(\"Time used pending\")"
     bPickA=AND Boolean(A=@n2, B=@n3)
    )
   )
  ]}
 OnTimeRemainingUpdated {nodes [
   n1 = OnTimeRemainingUpdated!() -> (then, TimeSpan:Timespan Structure) {is=title K2Node_FunctionEntry}
   n2 = TextToLower(InText=GetTimespanAsText(Timespan=@n1, bShowHoursWhenZero=false, bShowSeconds=false, FormatType=DaysHoursMinutesSeconds, UnitDisplayType=Full, bUseSentenceCase=false))
   n3 = MakeTimespan(Days=1, Hours=0, Minutes=0, Seconds=0, Milliseconds=0)
   > "SetText (Text)"(
    self=Get Text_PlaytimeLeft() -> (Text_PlaytimeLeft:Common Text Object Reference)
    InText=SelectText(
     A=Format Text(Format="LOCTEXT(\"{Time} left\")", Time=@n2)
     B="LOCTEXT(\"Time left pending\")"
     bPickA="Timespan < Timespan"(A:Timespan Structure=@n1, B:Timespan Structure=@n3) -> (ReturnValue:bool=~) {is=title K2Node_PromotableOperator}
    )
   )
   n4 = Get PlayerPlaytimeVM()
   > SetActiveWidget(
    self=Get TimeLimitSwitcher() -> (TimeLimitSwitcher:Common Activatable Widget Switcher Object Reference)
    Widget=Select(
     Option 0:Widget Object Reference=Get Box_TimeLimitOff() -> (Box_TimeLimitOff:Vertical Box Object Reference)
     Option 1:Widget Object Reference=Get Box_TimeLimitOn() -> (Box_TimeLimitOn:Vertical Box Object Reference)
     Index=IsPlaytimeLimitsEnabled!(self:Player Playtime View Model Object Reference=@n4) -> (ReturnValue:bool)
    ) -> (ReturnValue:Widget Object Reference)
   )
   n5 = Get Container_Buttons() -> (Container_Buttons:Vertical Box Object Reference)
   > Set ContainerButtonsPreviousVisibility(ContainerButtonsPreviousVisibility:ESlateVisibility Enum=Get Visibility!(self:Widget Object Reference=@n5) -> (Visibility:ESlateVisibility Enum=Visible)) -> (Output_Get:ESlateVisibility Enum=Visible)
   n6 = Floor(A=Get Total Minutes!(A:Timespan Structure=@n1) -> (ReturnValue:double) {is=title K2Node_CallFunction}) {is=title K2Node_CallFunction}
   n7 = Get Buttons_Switcher() -> (Buttons_Switcher:Common Activatable Widget Switcher Object Reference)
   > SetActiveWidget(
    self=@n7
    Widget=Select(
     Option 0=Get Container_ButtonsUnavailable() -> (Container_ButtonsUnavailable:Horizontal Box Object Reference)
     Option 1=@n5
     Index="integer <= integer"(A=@n6, B=30)
    )
   )
   n8 = Get PlayerPlaytimeVM()
   > SetVisibility(
    self=@n7
    InVisibility=Select(Option 0:ESlateVisibility Enum=Collapsed, Option 1:ESlateVisibility Enum=Visible, Index=IsPlaytimeLimitsEnabled(self=@n8)) -> (ReturnValue:ESlateVisibility Enum)
   )
   > UpdateAskForMoreTimeButton()
   n9 = Get Button_AskMoreTime()
   n10 = Get ContainerButtonsPreviousVisibility() -> (ContainerButtonsPreviousVisibility:ESlateVisibility Enum=Visible)
   > Branch(Condition=AND Boolean(A="Equal (Enum)"(A:ESlateVisibility Enum=@n10, B:ESlateVisibility Enum=Collapsed), B=IsVisible(self=@n9)))
   > SetFocus(self=@n9)
  ]}
 OnWindowBoundsUpdated {nodes [
   n1 = OnWindowBoundsUpdated!() -> (then, PlaytimeWindow:Playtime Window Structure) {is=title K2Node_FunctionEntry}
   > Set Window Bounds!(execute, self, Bounds:Playtime Window Structure=@n1) -> (then) {is=title K2Node_CallFunction}
   n2 = Get PlayerPlaytimeVM()
   n3 = Select(Option 0=Collapsed, Option 1=Visible, Index=HasWindowBounds!(self:Player Playtime View Model Object Reference=@n2) -> (ReturnValue:bool))
   > SetVisibility(
    self=Get Text_WindowBounds() -> (Text_WindowBounds:Common Text Object Reference)
    InVisibility=Select!(
     None:ESlateVisibility Enum=Collapsed
     AllowedTimeOnly:ESlateVisibility Enum=Collapsed
     WindowsOnly:ESlateVisibility Enum=@n3
     All:ESlateVisibility Enum=@n3
     Index:EPlaytimeActiveScheduling Enum=GetPlaytimeScheduling!(self:Player Playtime View Model Object Reference=@n2, ActivityType:EActivityLimitType Enum=CombinedTime) -> (ReturnValue:EPlaytimeActiveScheduling Enum=None)
    ) -> (ReturnValue)
   )
  ]}
 RefreshFocus {nodes [
   RefreshFocus() {is=title K2Node_FunctionEntry}
   > SetFocus(self=GetDesiredFocusTarget!(self:Common Activatable Widget Object Reference=Self-Reference()) -> (ReturnValue:Widget Object Reference))
  ]}
 Set Window Bounds {nodes [
   n1 = Set Window Bounds!() -> (then, Bounds_StartTime:Date Time Structure, Bounds_EndTime:Date Time Structure) {is=title K2Node_FunctionEntry}
   n2 = Get PlayerPlaytimeVM()
   > "SetText (Text)"(
    self=Get Text_WindowBounds()
    InText=Format Text!(
     Format="LOCTEXT(\"Time Window: {Begin} - {End}\")"
     Begin:Text=ConvertServiceTimeToLocalTimeText!(self:Player Playtime View Model Object Reference=@n2, ServiceTime:"Date Time Structure (by ref)"=@n1) -> (ReturnValue:Text)
     End:Text=ConvertServiceTimeToLocalTimeText(self=@n2, ServiceTime=@n1.Bounds_EndTime)
    ) -> (Result)
   )
  ]}
 ToDayHour {nodes [
   n1 = ToDayHour!() -> (then, DateTime:Date Time Structure) {is=title K2Node_FunctionEntry}
   n2 = MakeLiteralText(Value="LOCTEXT(\"PM\")")
   n3 = MakeLiteralText(Value="LOCTEXT(\"AM\")")
   n4 = "integer >= integer"(A=Get Hour!(A:Date Time Structure=@n1) -> (ReturnValue:int=0) {is=title K2Node_CallFunction}, B=12)
   > Return Node!(
    execute
    Result:Text=Format Text!(
     Format="LOCTEXT(\"{Hour} {AMPM}\")"
     Hour:int=Get Hour 12!(A:Date Time Structure=@n1) -> (ReturnValue:int=0) {is=title K2Node_CallFunction}
     AMPM:Text=SelectText(A=@n2, B=@n3, bPickA=@n4)
    ) -> (Result)
   )
  ]}
 UpdateAskForMoreTimeButton {nodes [
   UpdateAskForMoreTimeButton() {is=title K2Node_FunctionEntry}
   n1 = Get PlayerPlaytimeVM()
   n2 = IsAskForMorePlayTimeBlocked!(self:Player Playtime View Model Object Reference=@n1) -> (ReturnValue:bool)
   > Set bIsEnabled!(
    execute
    bIsEnabled:bool=AND Boolean(A=IsAskForMorePlayTimeEnabled!(self:Player Playtime View Model Object Reference=@n1) -> (ReturnValue:bool), B=NOT Boolean(A=@n2))
    self:Widget Object Reference=Get Button_AskMoreTime()
   ) -> (then, Output_Get:bool)
  ]}
]
variables []
widgets {
 info {inheritedWidgetCount=0, parentClass=/Script/CilantroUI.FortViewTimeLimitPanel, rootWidgetClass=&Border, widgetCount=44}
 widgets [
  Border_Content {
   widgetClassPath=&Border
   widget=Border_Content
   properties {
    navigation=&self:WidgetTree.Border_Content.WidgetNavigation_0
    background {
     bIsDynamicallyLoaded=false
     drawAs=NoDrawType
     imageType=NoImage
     mirroring=NoMirror
     resourceName=None
     resourceObject=null
     tiling=NoTile
     imageSize=<32 32>
     margin=<ltrb 0 0 0 0>
     outlineSettings {
      bUseBrushTransparency=false
      roundingType=HalfHeightRadius
      width=0
      color {colorUseRule=UseColor_Specified, specifiedColor=<rgba 0 0 0 0>}
      cornerRadii=<0 0 0 0>
     }
     tintColor {colorUseRule=UseColor_Specified, specifiedColor=<rgba 1 1 1 1>}
     uVRegion {bIsValid=false, max=<0 0>, min=<0 0>}
    }
    padding=<ltrb 0 0 0 0>
   }
  }
  Overlay_1 {widgetClassPath=&Overlay, parent=Border_Content, widget=Overlay_1, slot {class=BorderSlot, properties {padding=<ltrb 0 0 0 0>}}}
  SafeZone_Content {
   widgetClassPath=&SafeZone
   bIsVariable=true
   parent=Overlay_1
   widget=SafeZone_Content
   properties {padLeft=false, padRight=false, padTop=false}
   slot {class=OverlaySlot, properties {horizontalAlignment=HAlign_Fill, verticalAlignment=VAlign_Fill}}
  }
  Overlay_0 {widgetClassPath=&Overlay, parent=SafeZone_Content, widget=Overlay_0, slot {class=SafeZoneSlot}}
  VerticalBox_0 {
   widgetClassPath=&VerticalBox
   parent=Overlay_0
   widget=VerticalBox_0
   slot {class=OverlaySlot, properties {horizontalAlignment=HAlign_Fill, verticalAlignment=VAlign_Fill}}
  }
  Border_1 {
   widgetClassPath=&Border
   parent=VerticalBox_0
   widget=Border_1
   properties {
    background {
     bIsDynamicallyLoaded=false
     drawAs=Image
     imageType=NoImage
     mirroring=NoMirror
     resourceName=None
     resourceObject=/Game/UI/Sidebar/Materials/MI_UI_Sidebar_Header_BaseFade
     tiling=NoTile
     imageSize=<4096 4096>
     margin=<ltrb 0 0 0 0>
     outlineSettings {
      bUseBrushTransparency=false
      roundingType=HalfHeightRadius
      width=0
      color {colorUseRule=UseColor_Specified, specifiedColor=<rgba 0 0 0 0>}
      cornerRadii=<0 0 0 0>
     }
     tintColor {colorUseRule=UseColor_Specified, specifiedColor=<rgba 1 1 1 1>}
     uVRegion {bIsValid=false, max=<0 0>, min=<0 0>}
    }
    padding=<ltrb 20 60 20 20>
   }
   slot {class=VerticalBoxSlot, properties {verticalAlignment=VAlign_Top}}
  }
  HorizontalBox_Header {
   widgetClassPath=&HorizontalBox
   bIsVariable=true
   parent=Border_1
   widget=HorizontalBox_Header
   slot {class=BorderSlot, properties {padding=<ltrb 20 60 20 20>}}
  }
  Image_Info_1 {
   widgetClassPath=&Image
   parent=HorizontalBox_Header
   widget=Image_Info_1
   properties {brush {
     bIsDynamicallyLoaded=false
     drawAs=Image
     imageType=FullColor
     mirroring=NoMirror
     resourceName=None
     resourceObject=/FortUILibrary/Textures/MaterialInstances/MI_UIKit_Icon_UpdatePending
     tiling=NoTile
     imageSize=<32 32>
     margin=<ltrb 0 0 0 0>
     outlineSettings {
      bUseBrushTransparency=false
      roundingType=HalfHeightRadius
      width=0
      color {colorUseRule=UseColor_Specified, specifiedColor=<rgba 0 0 0 0>}
      cornerRadii=<0 0 0 0>
     }
     tintColor {colorUseRule=UseColor_Specified, specifiedColor=<rgba 1 1 1 1>}
     uVRegion {bIsValid=false, max=<0 0>, min=<0 0>}
    }}
   slot {
    class=HorizontalBoxSlot
    properties {horizontalAlignment=HAlign_Left, verticalAlignment=VAlign_Center, padding=<ltrb 0 0 10 0>, size {sizeRule=Automatic, value=0.1}}
   }
  }
  Text_Title {
   widgetClassPath=&CommonTextBlock
   bIsVariable=true
   parent=HorizontalBox_Header
   widget=Text_Title
   properties {
    autoWrapText=true
    text=Time Limits and Reports
    textTransformPolicy=ToUpper
    font {
     bForceMonospaced=false
     bMaterialIsStencil=false
     fontMaterial=null
     fontObject=&BurbankSmall
     letterSpacing=0
     monospacedWidth=1
     size=27
     skewAmount=0
     typefaceFontName=Black
     outlineSettings {
      bApplyOutlineToDropShadows=false
      bMiteredCorners=false
      bSeparateFillAlpha=false
      outlineMaterial=null
      outlineSize=0
      outlineColor=<rgba 0 0 0 1>
     }
    }
    renderTransform {angle=0, scale=<1 1>, shear=<-8 0>, translation=<0 0>}
    shadowOffset=<0 0>
   }
   slot {class=HorizontalBoxSlot, properties {verticalAlignment=VAlign_Center, size {sizeRule=Fill, value=1}}}
  }
  Switcher_State {
   widgetClassPath=&CommonActivatableWidgetSwitcher
   bIsVariable=true
   parent=VerticalBox_0
   widget=Switcher_State
   slot {class=VerticalBoxSlot, properties {padding=<ltrb 10 32 10 0>, size {sizeRule=Fill, value=1}}}
  }
  Box_ContentFirstPage {
   widgetClassPath=&VerticalBox
   bIsVariable=true
   parent=Switcher_State
   widget=Box_ContentFirstPage
   slot {class=WidgetSwitcherSlot}
  }
  Text_Message {
   widgetClassPath=&CommonTextBlock
   parent=Box_ContentFirstPage
   widget=Text_Message
   properties {
    autoWrapText=true
    justification=Center
    text="Your parent or guardian has enabled time limits and time reports for your account in Fortnite and Unreal Editor for Fortnite (UEFN)."
    font {
     bForceMonospaced=false
     bMaterialIsStencil=false
     fontMaterial=null
     fontObject=&BurbankSmall
     letterSpacing=0
     monospacedWidth=1
     size=18
     skewAmount=0
     typefaceFontName=Bold
     outlineSettings {
      bApplyOutlineToDropShadows=false
      bMiteredCorners=false
      bSeparateFillAlpha=false
      outlineMaterial=null
      outlineSize=0
      outlineColor=<rgba 0 0 0 1>
     }
    }
    shadowOffset=<0 0>
   }
   slot {class=VerticalBoxSlot}
  }
  PlayTimeInfo {
   widgetClassPath=&VerticalBox
   parent=Box_ContentFirstPage
   widget=PlayTimeInfo
   slot {class=VerticalBoxSlot, properties {padding=<ltrb 0 32 0 0>}}
  }
  Text_PlayTimeHeader {
   widgetClassPath=&CommonTextBlock
   parent=PlayTimeInfo
   widget=Text_PlayTimeHeader
   properties {
    justification=Center
    text=TIME LIMITS
    font {
     bForceMonospaced=false
     bMaterialIsStencil=false
     fontMaterial=null
     fontObject=&Burbank
     letterSpacing=0
     monospacedWidth=1
     size=21
     skewAmount=0
     typefaceFontName=Burbank-Black
     outlineSettings {
      bApplyOutlineToDropShadows=false
      bMiteredCorners=false
      bSeparateFillAlpha=false
      outlineMaterial=null
      outlineSize=0
      outlineColor=<rgba 0 0 0 1>
     }
    }
    shadowOffset=<0 0>
   }
   slot {class=VerticalBoxSlot}
  }
  TimeLimitSwitcher {
   widgetClassPath=&CommonActivatableWidgetSwitcher
   bIsVariable=true
   parent=PlayTimeInfo
   widget=TimeLimitSwitcher
   slot {class=VerticalBoxSlot}
  }
  Box_TimeLimitOn {widgetClassPath=&VerticalBox, bIsVariable=true, parent=TimeLimitSwitcher, widget=Box_TimeLimitOn, slot {class=WidgetSwitcherSlot}}
  Text_PlaytimeLeft {
   widgetClassPath=&CommonTextBlock
   bIsVariable=true
   parent=Box_TimeLimitOn
   widget=Text_PlaytimeLeft
   properties {
    justification=Center
    text=Time left
    font {
     bForceMonospaced=false
     bMaterialIsStencil=false
     fontMaterial=null
     fontObject=&BurbankSmall
     letterSpacing=0
     monospacedWidth=1
     size=15
     skewAmount=0
     typefaceFontName=Bold
     outlineSettings {
      bApplyOutlineToDropShadows=false
      bMiteredCorners=false
      bSeparateFillAlpha=false
      outlineMaterial=null
      outlineSize=0
      outlineColor=<rgba 0 0 0 1>
     }
    }
    shadowOffset=<0 0>
   }
   slot {class=VerticalBoxSlot, properties {padding=<ltrb 0 4 0 0>}}
  }
  Text_PlaytimePlayed {
   widgetClassPath=&CommonTextBlock
   bIsVariable=true
   parent=Box_TimeLimitOn
   widget=Text_PlaytimePlayed
   properties {
    justification=Center
    text=TimePlayed
    font {
     bForceMonospaced=false
     bMaterialIsStencil=false
     fontMaterial=null
     fontObject=&BurbankSmall
     letterSpacing=0
     monospacedWidth=1
     size=15
     skewAmount=0
     typefaceFontName=Bold
     outlineSettings {
      bApplyOutlineToDropShadows=false
      bMiteredCorners=false
      bSeparateFillAlpha=false
      outlineMaterial=null
      outlineSize=0
      outlineColor=<rgba 0 0 0 1>
     }
    }
    shadowOffset=<0 0>
   }
   slot {class=VerticalBoxSlot, properties {padding=<ltrb 0 2 0 0>}}
  }
  Text_WindowBounds {
   widgetClassPath=&CommonTextBlock
   bIsVariable=true
   parent=Box_TimeLimitOn
   widget=Text_WindowBounds
   properties {
    justification=Center
    text=WindowBounds
    font {
     bForceMonospaced=false
     bMaterialIsStencil=false
     fontMaterial=null
     fontObject=&BurbankSmall
     letterSpacing=0
     monospacedWidth=1
     size=15
     skewAmount=0
     typefaceFontName=Bold
     outlineSettings {
      bApplyOutlineToDropShadows=false
      bMiteredCorners=false
      bSeparateFillAlpha=false
      outlineMaterial=null
      outlineSize=0
      outlineColor=<rgba 0 0 0 1>
     }
    }
    shadowOffset=<0 0>
   }
   slot {class=VerticalBoxSlot, properties {padding=<ltrb 0 2 0 0>}}
  }
  Box_TimeLimitOff {
   widgetClassPath=&VerticalBox
   bIsVariable=true
   parent=TimeLimitSwitcher
   widget=Box_TimeLimitOff
   slot {class=WidgetSwitcherSlot}
  }
  Text_LimitOff {
   widgetClassPath=&CommonTextBlock
   bIsVariable=true
   parent=Box_TimeLimitOff
   widget=Text_LimitOff
   properties {
    justification=Center
    text=Time Limits off
    font {
     bForceMonospaced=false
     bMaterialIsStencil=false
     fontMaterial=null
     fontObject=&BurbankSmall
     letterSpacing=0
     monospacedWidth=1
     size=15
     skewAmount=0
     typefaceFontName=Bold
     outlineSettings {
      bApplyOutlineToDropShadows=false
      bMiteredCorners=false
      bSeparateFillAlpha=false
      outlineMaterial=null
      outlineSize=0
      outlineColor=<rgba 0 0 0 1>
     }
    }
    shadowOffset=<0 0>
   }
   slot {class=VerticalBoxSlot, properties {padding=<ltrb 0 4 0 0>}}
  }
  PlayReportInfo {
   widgetClassPath=&VerticalBox
   parent=Box_ContentFirstPage
   widget=PlayReportInfo
   slot {class=VerticalBoxSlot, properties {padding=<ltrb 0 32 0 0>}}
  }
  Text_ReportInfoHeader {
   widgetClassPath=&CommonTextBlock
   parent=PlayReportInfo
   widget=Text_ReportInfoHeader
   properties {
    justification=Center
    text=TIME REPORTS
    font {
     bForceMonospaced=false
     bMaterialIsStencil=false
     fontMaterial=null
     fontObject=&Burbank
     letterSpacing=0
     monospacedWidth=1
     size=21
     skewAmount=0
     typefaceFontName=Burbank-Black
     outlineSettings {
      bApplyOutlineToDropShadows=false
      bMiteredCorners=false
      bSeparateFillAlpha=false
      outlineMaterial=null
      outlineSize=0
      outlineColor=<rgba 0 0 0 1>
     }
    }
    shadowOffset=<0 0>
   }
   slot {class=VerticalBoxSlot}
  }
  Text_ReportStatus {
   widgetClassPath=&CommonTextBlock
   bIsVariable=true
   parent=PlayReportInfo
   widget=Text_ReportStatus
   properties {
    justification=Center
    text=Time Reports on
    font {
     bForceMonospaced=false
     bMaterialIsStencil=false
     fontMaterial=null
     fontObject=&BurbankSmall
     letterSpacing=0
     monospacedWidth=1
     size=15
     skewAmount=0
     typefaceFontName=Bold
     outlineSettings {
      bApplyOutlineToDropShadows=false
      bMiteredCorners=false
      bSeparateFillAlpha=false
      outlineMaterial=null
      outlineSize=0
      outlineColor=<rgba 0 0 0 1>
     }
    }
    shadowOffset=<0 0>
   }
   slot {class=VerticalBoxSlot, properties {padding=<ltrb 0 4 0 0>}}
  }
  Text_ReportDescription {
   widgetClassPath=&CommonTextBlock
   bIsVariable=true
   parent=PlayReportInfo
   widget=Text_ReportDescription
   properties {
    autoWrapText=true
    justification=Center
    text=Your parent or guardian can see a summary of the time you spend in Fortnite and UEFN.
    colorAndOpacity {colorUseRule=UseColor_Specified, specifiedColor=<rgba 0.5365 0.5365 0.5365 1>}
    font {
     bForceMonospaced=false
     bMaterialIsStencil=false
     fontMaterial=null
     fontObject=&BurbankSmall
     letterSpacing=0
     monospacedWidth=1
     size=15
     skewAmount=0
     typefaceFontName=Bold
     outlineSettings {
      bApplyOutlineToDropShadows=false
      bMiteredCorners=false
      bSeparateFillAlpha=false
      outlineMaterial=null
      outlineSize=0
      outlineColor=<rgba 0 0 0 1>
     }
    }
    shadowOffset=<0 0>
   }
   slot {class=VerticalBoxSlot, properties {padding=<ltrb 8 2 8 0>}}
  }
  Buttons_Switcher {
   widgetClassPath=&CommonActivatableWidgetSwitcher
   bIsVariable=true
   parent=Box_ContentFirstPage
   widget=Buttons_Switcher
   properties {activeWidgetIndex=1}
   slot {class=VerticalBoxSlot, properties {padding=<ltrb 0 32 0 0>}}
  }
  Container_ButtonsUnavailable {
   widgetClassPath=&HorizontalBox
   bIsVariable=true
   parent=Buttons_Switcher
   widget=Container_ButtonsUnavailable
   slot {class=WidgetSwitcherSlot}
  }
  Image_Info {
   widgetClassPath=&Image
   parent=Container_ButtonsUnavailable
   widget=Image_Info
   properties {brush {
     bIsDynamicallyLoaded=false
     drawAs=Image
     imageType=FullColor
     mirroring=NoMirror
     resourceName=None
     resourceObject=/FortUILibrary/Textures/MaterialInstances/MI_UIKit_Icon_Information
     tiling=NoTile
     imageSize=<32 32>
     margin=<ltrb 0 0 0 0>
     outlineSettings {
      bUseBrushTransparency=false
      roundingType=HalfHeightRadius
      width=0
      color {colorUseRule=UseColor_Specified, specifiedColor=<rgba 0 0 0 0>}
      cornerRadii=<0 0 0 0>
     }
     tintColor {colorUseRule=UseColor_Specified, specifiedColor=<rgba 0.5365 0.5365 0.5365 1>}
     uVRegion {bIsValid=false, max=<0 0>, min=<0 0>}
    }}
   slot {class=HorizontalBoxSlot, properties {horizontalAlignment=HAlign_Left, verticalAlignment=VAlign_Center, size {sizeRule=Automatic, value=0.1}}}
  }
  CommonTextBlock_56 {
   widgetClassPath=&CommonTextBlock
   parent=Container_ButtonsUnavailable
   widget=CommonTextBlock_56
   properties {
    autoWrapText=true
    text=You can ask your parent or guardian for more time when you have less than 30 minutes left.
    colorAndOpacity {colorUseRule=UseColor_Specified, specifiedColor=<rgba 0.5365 0.5365 0.5365 1>}
    font {
     bForceMonospaced=false
     bMaterialIsStencil=false
     fontMaterial=null
     fontObject=&BurbankSmall
     letterSpacing=0
     monospacedWidth=1
     size=15
     skewAmount=0
     typefaceFontName=Bold
     outlineSettings {
      bApplyOutlineToDropShadows=false
      bMiteredCorners=false
      bSeparateFillAlpha=false
      outlineMaterial=null
      outlineSize=0
      outlineColor=<rgba 0 0 0 1>
     }
    }
    shadowOffset=<0 0>
   }
   slot {class=HorizontalBoxSlot, properties {padding=<ltrb 8 0 0 0>, size {sizeRule=Fill, value=0.9}}}
  }
  Container_Buttons {
   widgetClassPath=&VerticalBox
   bIsVariable=true
   parent=Buttons_Switcher
   widget=Container_Buttons
   slot {class=WidgetSwitcherSlot}
  }
  Button_AskMoreTime {
   widgetClassPath=/Game/UI/Frontend/MainMenu/MainMenuButton
   bIsVariable=true
   parent=Container_Buttons
   widget=Button_AskMoreTime
   properties {
    defaultText=ASK FOR MORE TIME
    holdData=&CommonUI_HoldData_Base_C
    navigation=&self:WidgetTree.Button_AskMoreTime.WidgetNavigation_1
    onButtonBaseClicked=(&self:WidgetTree.Button_AskMoreTime.HandleOnButtonClicked)
    onButtonBaseDoubleClicked=(&self:WidgetTree.Button_AskMoreTime.HandleOnButtonDoubleClicked)
    onButtonBaseHovered=(&self:WidgetTree.Button_AskMoreTime.HandleOnButtonHovered)
    onButtonBaseUnhovered=(&self:WidgetTree.Button_AskMoreTime.HandleOnButtonUnhovered)
    onSelectedChangedBase=(&self:WidgetTree.Button_AskMoreTime.HandleOnSelectedChanged)
    visibility=Visible
    defaultBrush {
     bIsDynamicallyLoaded=false
     drawAs=Image
     imageType=FullColor
     mirroring=NoMirror
     resourceName=None
     resourceObject=/FortUILibrary/Textures/MaterialInstances/MI_UIKit_Icon_Email
     tiling=NoTile
     imageSize=<32 32>
     margin=<ltrb 0 0 0 0>
     outlineSettings {
      bUseBrushTransparency=false
      roundingType=HalfHeightRadius
      width=0
      color {colorUseRule=UseColor_Specified, specifiedColor=<rgba 0 0 0 0>}
      cornerRadii=<0 0 0 0>
     }
     tintColor {colorUseRule=UseColor_Specified, specifiedColor=<rgba 1 1 1 1>}
     uVRegion {bIsValid=false, max=<0 0>, min=<0 0>}
    }
   }
   slot {class=VerticalBoxSlot, properties {padding=<ltrb 16 0 16 0>}}
  }
  Button_AddMoreTime {
   widgetClassPath=&MainMenuButton
   bIsVariable=true
   parent=Container_Buttons
   widget=Button_AddMoreTime
   properties {
    defaultText=ADD MORE TIME
    holdData=&CommonUI_HoldData_Base_C
    navigation=&self:WidgetTree.Button_AddMoreTime.WidgetNavigation_1
    onButtonBaseClicked=()
    onButtonBaseDoubleClicked=()
    onButtonBaseHovered=()
    onButtonBaseUnhovered=()
    onSelectedChangedBase=()
    visibility=Visible
   }
   slot {class=VerticalBoxSlot, properties {padding=<ltrb 16 8 16 0>}}
  }
  Box_EmailSending {
   widgetClassPath=&VerticalBox
   bIsVariable=true
   parent=Switcher_State
   widget=Box_EmailSending
   slot {class=WidgetSwitcherSlot, properties {padding=<ltrb 16 0 16 0>}}
  }
  Spacer {widgetClassPath=&Spacer, parent=Box_EmailSending, widget=Spacer, properties {size=<1 268.1173>}, slot {class=VerticalBoxSlot}}
  HorizontalBox {
   widgetClassPath=&HorizontalBox
   parent=Box_EmailSending
   widget=HorizontalBox
   slot {class=VerticalBoxSlot, properties {horizontalAlignment=HAlign_Center}}
  }
  Image_Spinner {
   widgetClassPath=&Image
   parent=HorizontalBox
   widget=Image_Spinner
   properties {
    visibility=SelfHitTestInvisible
    brush {
     bIsDynamicallyLoaded=false
     drawAs=Image
     imageType=FullColor
     mirroring=NoMirror
     resourceName=None
     resourceObject=/Game/UI/ContentOnDemand/ItemDownloadIndicator/ItemDownloadIndicator_SpinnerMaterial
     tiling=NoTile
     imageSize=<32 32>
     margin=<ltrb 0 0 0 0>
     outlineSettings {
      bUseBrushTransparency=false
      roundingType=HalfHeightRadius
      width=0
      color {colorUseRule=UseColor_Specified, specifiedColor=<rgba 0 0 0 0>}
      cornerRadii=<0 0 0 0>
     }
     tintColor {colorUseRule=UseColor_Specified, specifiedColor=<rgba 1 1 1 1>}
     uVRegion {bIsValid=false, max=<0 0>, min=<0 0>}
    }
   }
   slot {class=HorizontalBoxSlot, properties {horizontalAlignment=HAlign_Center, verticalAlignment=VAlign_Center, padding=<ltrb 0 0 8 0>}}
  }
  Text_SendingEmail {
   widgetClassPath=&CommonTextBlock
   parent=HorizontalBox
   widget=Text_SendingEmail
   properties {
    text=SENDING EMAIL
    font {
     bForceMonospaced=false
     bMaterialIsStencil=false
     fontMaterial=null
     fontObject=&Burbank
     letterSpacing=0
     monospacedWidth=1
     size=24
     skewAmount=0
     typefaceFontName=Burbank-Black
     outlineSettings {
      bApplyOutlineToDropShadows=false
      bMiteredCorners=false
      bSeparateFillAlpha=false
      outlineMaterial=null
      outlineSize=0
      outlineColor=<rgba 0 0 0 1>
     }
    }
    shadowOffset=<0 0>
   }
   slot {class=HorizontalBoxSlot}
  }
  Box_ContentEmailSent {
   widgetClassPath=&VerticalBox
   bIsVariable=true
   parent=Switcher_State
   widget=Box_ContentEmailSent
   slot {class=WidgetSwitcherSlot, properties {padding=<ltrb 16 0 16 0>}}
  }
  Spacer_151 {
   widgetClassPath=&Spacer
   parent=Box_ContentEmailSent
   widget=Spacer_151
   properties {size=<1 268.1173>}
   slot {class=VerticalBoxSlot}
  }
  HorizontalBox_1 {
   widgetClassPath=&HorizontalBox
   parent=Box_ContentEmailSent
   widget=HorizontalBox_1
   slot {class=VerticalBoxSlot, properties {horizontalAlignment=HAlign_Center}}
  }
  Image_Tick {
   widgetClassPath=&Image
   bIsVariable=true
   parent=HorizontalBox_1
   widget=Image_Tick
   properties {brush {
     bIsDynamicallyLoaded=false
     drawAs=Image
     imageType=FullColor
     mirroring=NoMirror
     resourceName=None
     resourceObject=/Game/UI/Foundation/Textures/Icons/Checks/T-Icon-Check-64
     tiling=NoTile
     imageSize=<32 32>
     margin=<ltrb 0 0 0 0>
     outlineSettings {
      bUseBrushTransparency=false
      roundingType=HalfHeightRadius
      width=0
      color {colorUseRule=UseColor_Specified, specifiedColor=<rgba 0 0 0 0>}
      cornerRadii=<0 0 0 0>
     }
     tintColor {colorUseRule=UseColor_Specified, specifiedColor=<rgba 1 1 1 1>}
     uVRegion {bIsValid=false, max=<0 0>, min=<0 0>}
    }}
   slot {class=HorizontalBoxSlot, properties {horizontalAlignment=HAlign_Center, verticalAlignment=VAlign_Center, padding=<ltrb 0 0 8 0>}}
  }
  Text_EmailSent {
   widgetClassPath=&CommonTextBlock
   bIsVariable=true
   parent=HorizontalBox_1
   widget=Text_EmailSent
   properties {
    text=EMAIL SENT
    font {
     bForceMonospaced=false
     bMaterialIsStencil=false
     fontMaterial=null
     fontObject=&Burbank
     letterSpacing=0
     monospacedWidth=1
     size=24
     skewAmount=0
     typefaceFontName=Burbank-Black
     outlineSettings {
      bApplyOutlineToDropShadows=false
      bMiteredCorners=false
      bSeparateFillAlpha=false
      outlineMaterial=null
      outlineSize=0
      outlineColor=<rgba 0 0 0 1>
     }
    }
    shadowOffset=<0 0>
   }
   slot {class=HorizontalBoxSlot}
  }
  Text_EmailDescription {
   widgetClassPath=&CommonTextBlock
   bIsVariable=true
   parent=Box_ContentEmailSent
   widget=Text_EmailDescription
   properties {
    autoWrapText=true
    justification=Center
    text="We’ve sent an email to your parent or guardian to ask for more time."
    font {
     bForceMonospaced=false
     bMaterialIsStencil=false
     fontMaterial=null
     fontObject=&BurbankSmall
     letterSpacing=0
     monospacedWidth=1
     size=18
     skewAmount=0
     typefaceFontName=Bold
     outlineSettings {
      bApplyOutlineToDropShadows=false
      bMiteredCorners=false
      bSeparateFillAlpha=false
      outlineMaterial=null
      outlineSize=0
      outlineColor=<rgba 0 0 0 1>
     }
    }
    shadowOffset=<0 0>
   }
   slot {class=VerticalBoxSlot, properties {padding=<ltrb 0 16 0 0>}}
  }
  Button_Okay {
   widgetClassPath=&MainMenuButton
   bIsVariable=true
   parent=Box_ContentEmailSent
   widget=Button_Okay
   properties {
    defaultText=Done
    holdData=&CommonUI_HoldData_Base_C
    onButtonBaseClicked=(&self:WidgetTree.Button_Okay.HandleOnButtonClicked)
    onButtonBaseDoubleClicked=(&self:WidgetTree.Button_Okay.HandleOnButtonDoubleClicked)
    onButtonBaseHovered=(&self:WidgetTree.Button_Okay.HandleOnButtonHovered)
    onButtonBaseUnhovered=(&self:WidgetTree.Button_Okay.HandleOnButtonUnhovered)
    onSelectedChangedBase=(&self:WidgetTree.Button_Okay.HandleOnSelectedChanged)
    visibility=Visible
   }
   slot {class=VerticalBoxSlot, properties {padding=<ltrb 0 48 0 0>}}
  }
 ]
}
