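# CommonActivatableWidget screen, apparently for Save the World, that shows a time limit panel over a blue background in a right-aligned, width-capped column.
# On activation it activates the nested ViewTimeLimitPanel.
# That panel's deactivation, the Back button (bound to the Back input action) and the full-screen lightbox backboard button's click all lead to deactivating the screen.
# The backboard is set HitTestInvisible.
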WidgetBlueprint /Cilantro/ViewTimeLimitPanel_STW
graphs [EventGraph {nodes [
   "On Widget Deactivated (ViewTimeLimitPanel)"() {is=title K2Node_ComponentBoundEvent}
   > DeactivateWidget()
   n1 = "On Button Clicked (Button_BackBoard)"!() -> (OutputDelegate, then, Button:Common Button Legacy Object Reference) {is=title K2Node_ComponentBoundEvent}
   DeactivateWidget(execute=[@n1.then, @n2.then])
   n2 = "On Clicked (Button_Back)"!() -> (OutputDelegate, then, Button:Common Button Base Object Reference) {is=title K2Node_ComponentBoundEvent}
   Event On Activated()
   > ActivateWidget(self=Get ViewTimeLimitPanel() -> (ViewTimeLimitPanel:View Time Limit Panel Object Reference))
  ]}]
variables []
widgets {
 info {inheritedWidgetCount=0, parentClass=&CommonActivatableWidget, rootWidgetClass=&Overlay, widgetCount=7}
 widgets [
  Overlay_0 {widgetClassPath=&Overlay, widget=Overlay_0}
  Button_BackBoard {
   widgetClassPath=&IconTextButton
   bIsVariable=true
   parent=Overlay_0
   widget=Button_BackBoard
   properties {
    bIsFocusable=false
    buttonText=""
    clickMethod=MouseDown
    onButtonBaseClicked=()
    onButtonBaseDoubleClicked=()
    onButtonBaseHovered=()
    onButtonBaseUnhovered=()
    onSelectedChangedBase=()
    pressMethod=ButtonPress
    style=/FortUILibrary/Foundation/Buttons/ButtonStyles/ButtonStyle-Lightbox.ButtonStyle-Lightbox_C
    touchMethod=Down
    visibility=HitTestInvisible
    hoveredSlateSoundOverride {resourceObject=null}
   }
   slot {class=OverlaySlot, properties {horizontalAlignment=HAlign_Fill, verticalAlignment=VAlign_Fill}}
  }
  SizeBox_105 {
   widgetClassPath=&SizeBox
   parent=Overlay_0
   widget=SizeBox_105
   properties {bOverride_MaxDesiredWidth=true, maxDesiredWidth=500}
   slot {class=OverlaySlot, properties {horizontalAlignment=HAlign_Right, verticalAlignment=VAlign_Fill}}
  }
  Overlay_1 {
   widgetClassPath=&Overlay
   parent=SizeBox_105
   widget=Overlay_1
   slot {class=SizeBoxSlot, properties {horizontalAlignment=HAlign_Right}}
  }
  Background {
   widgetClassPath=&Image
   parent=Overlay_1
   widget=Background
   properties {visibility=SelfHitTestInvisible, colorAndOpacity=<rgba 0.0194 0.0666 0.4179 1>}
   slot {class=OverlaySlot, properties {horizontalAlignment=HAlign_Fill, verticalAlignment=VAlign_Fill}}
  }
  ViewTimeLimitPanel {
   widgetClassPath=&/ViewTimeLimitPanel
   bIsVariable=true
   parent=Overlay_1
   widget=ViewTimeLimitPanel
   properties {navigation=&self:WidgetTree.ViewTimeLimitPanel.WidgetNavigation_0}
   slot {class=OverlaySlot, properties {horizontalAlignment=HAlign_Fill, verticalAlignment=VAlign_Fill}}
  }
  Button_Back {
   widgetClassPath=/FortUILibrary/UIKit/Buttons/WBP_UIKit_Button_Quiet
   bIsVariable=true
   parent=Overlay_1
   widget=Button_Back
   properties {text=Back, triggeringInputAction {dataTable=&InputActionDataTable, rowName=Back}}
   slot {class=OverlaySlot}
  }
 ]
}
