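# Add Time button derived from the UIKit regular button, with a secondary line showing remaining time.
# PreConstruct resizes the icon for compact UI, then applies background material, left alignment and fonts.
# Hover and focus show the icon and brighten both text lines; input changes reset this, though touch input always counts as hovered.
# SetSecondaryTimeText formats a Timespan as lowercase "{Time} left", or "No time left" at zero or below.
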WidgetBlueprint /Cilantro/WBP_AddTime_Button
graphs [
 EventGraph {nodes [
   n1 = Event PreConstruct()
   > Parent: PreConstruct(IsDesignTime=@n1.IsDesignTime)
   > SetCompactUIStyle() {is=title K2Node_CallFunction}
   > SetButtonVisual()
   Event Construct()
   > Parent: Construct()
   n2 = Event On Input Method Changed!() -> (OutputDelegate, then, CurrentInputType:ECommonInputType Enum) {is=title K2Node_Event}
   > Parent: On Input Method Changed!(execute, CurrentInputType:ECommonInputType Enum=@n2.CurrentInputType) -> (then) {is=title K2Node_CallParentFunction}
   > SetHoveredState!(execute, self, IsHovered:bool=false) -> (then)
   Event On Hovered()
   > n3 = Parent: On Hovered() {is=title K2Node_CallParentFunction}
   SetHoveredState(execute=[@n3.then, @n5.then], IsHovered=true)
   Event On Unhovered()
   > n4 = Parent: On Unhovered() {is=title K2Node_CallParentFunction}
   SetHoveredState(execute=[@n6.then, @n4.then], IsHovered=false)
   Event On Focused() {is=title K2Node_Event}
   > n5 = Parent: On Focused() {is=title K2Node_CallParentFunction}
   Event On Unfocused() {is=title K2Node_Event}
   > n6 = Parent: On Unfocused() {is=title K2Node_CallParentFunction}
  ]}
 SetButtonVisual {nodes [
   SetButtonVisual() {is=title K2Node_FunctionEntry}
   > SetBackgroundMaterial!(
    execute
    self:WBP UIKit Block Background Base Object Reference=Get Block_Background() -> (Block_Background:WBP UIKit Block Background With Outline Object Reference)
    Material Instance:"Material Instance Object Reference (by ref)"=Get MI_Background() -> (MI_Background:Material Instance Object Reference)
   ) -> (then)
   > SetTextAlignment()
   > SetTextFont()
   > SetHoveredState(IsHovered=Get IsButtonHovered() -> (IsButtonHovered:bool))
  ]}
 SetCompactUIStyle {nodes [
   SetCompactUIStyle() {is=title K2Node_FunctionEntry}
   > Branch(Condition=IsUsingCompactUI())
   > SetSize!(
    execute
    self:WBP UIKit Block Image Object Reference=Get Block_Icon() -> (Block_Icon:WBP UIKit Block Image Object Reference)
    Desired Size:Vector 2D Structure=(X=64,Y=64)
   ) -> (then)
  ]}
 SetHoveredState {nodes [
   SetHoveredState!() -> (then, IsHovered:bool=~) {is=title K2Node_FunctionEntry}
   n1 = OR Boolean(
    A=Get IsHovered() -> (IsHovered:bool)
    B=IsUsingInputType!(InputType:ECommonInputType Enum=Touch, PlayerController:Player Controller Object Reference=GetOwningPlayer()) -> (ReturnValue:bool)
   )
   > SetBlockVisibility!(execute, self:WBP UIKit Block Base Object Reference=Get Block_Icon(), Visible:bool=@n1) -> (then)
   > SetTextColor!(
    execute
    self
    Color:Slate Color Structure=Select(
     Option 0:Slate Color Structure=Get UnhoveredTextButtonColor() -> (UnhoveredTextButtonColor:Slate Color Structure)
     Option 1:Slate Color Structure=Get HoveredTextButtonColor() -> (HoveredTextButtonColor:Slate Color Structure)
     Index=@n1
    ) -> (ReturnValue:Slate Color Structure)
   ) -> (then)
  ]}
 SetSecondaryTimeText {nodes [
   SetSecondaryTimeText!() -> (then, Timespan:Timespan Structure) {is=title K2Node_FunctionEntry}
   n1 = TextToLower!(InText:"Text (by ref)"=GetTimespanAsText!(
     Timespan:"Timespan Structure (by ref)"=Get Timespan() -> (Timespan:Timespan Structure)
     bShowHoursWhenZero:bool=false
     bShowSeconds:bool=false
     FormatType:ETimespanAsTextFormat Enum=DaysHoursMinutesSeconds
     UnitDisplayType:ETimespanUnitDisplayFormat Enum=Full
     bUseSentenceCase:bool=false
    ) -> (ReturnValue:Text)) -> (ReturnValue:Text)
   n2 = MakeTimespan!(Days:int=0, Hours:int=0, Minutes:int=0, Seconds:int=0, Milliseconds:int=0) -> (ReturnValue:Timespan Structure)
   n3 = Get Timespan()
   > SetSecondaryText!(
    execute
    self:Fort CTAButton Object Reference
    ButtonSecondaryText:"Text (by ref)"=SelectText(
     A="LOCTEXT(\"No time left\")"
     B=Format Text!(Format="LOCTEXT(\"{Time} left\")", Time:Text=@n1) -> (Result)
     bPickA="Timespan <= Timespan"(A:Timespan Structure=@n3, B:Timespan Structure=@n2) -> (ReturnValue:bool=~) {is=title K2Node_PromotableOperator}
    )
   ) -> (then)
  ]}
 SetTextAlignment {nodes [
   SetTextAlignment() {is=title K2Node_FunctionEntry}
   > SetHorizontalAlignment(
    self:Vertical Box Slot Object Reference=[
     SlotAsVerticalBoxSlot(Widget=Get Block_SecondaryText() -> (Block_SecondaryText:WBP UIKit Block Text Object Reference))
     SlotAsVerticalBoxSlot(Widget=Get Block_Text() -> (Block_Text:WBP UIKit Block Text Object Reference))
    ]
    InHorizontalAlignment=HAlign_Left
   )
  ]}
 SetTextColor {nodes [
   n1 = SetTextColor!() -> (then, Color:Slate Color Structure) {is=title K2Node_FunctionEntry}
   > SetColorAndOpacity(
    self=[
     Get TextBlock!(self:WBP UIKit Block Text Object Reference=Get Block_SecondaryText()) -> (TextBlock:Common Text Object Reference)
     Get TextBlock(self=Get Block_Text())
    ]
    InColorAndOpacity=@n1
   )
  ]}
 SetTextFont {nodes [
   SetTextFont() {is=title K2Node_FunctionEntry}
   > SetFont!(
    execute
    self:WBP UIKit Block Text Object Reference=Get Block_Text()
    Font:"Slate Font Info Structure (by ref)"=Get PrimaryFont() -> (PrimaryFont:Slate Font Info Structure)
   ) -> (then)
   > SetFont(self=Get Block_SecondaryText(), Font=Get SecondaryFont() -> (SecondaryFont:Slate Font Info Structure))
  ]}
]
variables [
 HoveredTextButtonColor {type=SlateColor, default {colorUseRule=UseColor_Specified, specifiedColor=<rgba 1 1 1 1>}}
 IsButtonHovered {type=boolean}
 UnhoveredTextButtonColor {type=SlateColor, default {colorUseRule=UseColor_Specified, specifiedColor=<rgba 0.6376 0.6376 0.6376 1>}}
]
widgets {
 info {
  inheritedWidgetCount=0
  parentClass=/FortUILibrary/UIKit/Buttons/WBP_UIKit_Button_Regular.WBP_UIKit_Button_Regular_C
  rootWidgetClass=null
  widgetCount=0
 }
 widgets []
}
