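# Parental-controls dialog where a PIN-verified parent grants a child extra play, create or combined time.
# A switcher shows activity choice, add-time form, legacy form (when bEnableThyme is off) or day-changed notice.
# PlayerPlaytimeVM picks the starting page, builds summary text and grants the selected duration and expiry using the PIN.
# Add time validates day/time zone and past expiry before granting; gamepad focus follows the active page.
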WidgetBlueprint /Cilantro/WBP_AddTime
graphs [
 BP_GetDesiredFocusTarget {nodes [
   Get Desired Focus Target()
   > n1 = Branch(Condition=Get bEnableThyme!(self:Fort Runtime Options Object Reference=GetRuntimeOptions()) -> (bEnableThyme:bool))
   n2 = Get CommonInputSubsystem()
   > Branch(Condition="Equal (Enum)"(A=GetCurrentInputType(self=@n2), B=Gamepad))
   > n3 = Switch on Int!(
    execute:Exec
    Selection:int=GetActiveWidgetIndex!(self:Common Visibility Switcher Object Reference=Get Switcher_Content() -> (Switcher_Content:Common Visibility Switcher Object Reference)) -> (ReturnValue:int=0)
   ) -> (0:Exec, 1:Exec, 2:Exec, 3:Exec) {bHasDefaultPin=false}
   > Return Node!(execute, ReturnValue:Widget Object Reference=Get Button_PlayTime() -> (Button_PlayTime:WBP Add Time Button Object Reference))
   Return Node(execute=@n3.1, ReturnValue=Get Rotator_ExtraTime() -> (Rotator_ExtraTime:WBP Cilantro Vaue Rotator Object Reference))
   Return Node(execute=[@n3.2, @n1.else], ReturnValue=Get GrantTimeSetting() -> (GrantTimeSetting:WBP Cilantro Vaue Rotator Object Reference))
   Return Node(execute=@n3.3, ReturnValue=Get Button_Close_DayHasChanged() -> (Button_Close_DayHasChanged:WBP UIKit Button Regular Object Reference))
  ]}
 ClearSelectedRotatorIndex {nodes [
   ClearSelectedRotatorIndex() {is=title K2Node_FunctionEntry}
   > SetSelectedItem!(execute, self:WBP Cilantro Vaue Rotator Object Reference=Get Rotator_ExtraTime(), Index:int=0) -> (then)
   > SetSelectedItem(self=Get Rotator_TimeUsedBefore() -> (Rotator_TimeUsedBefore:WBP Cilantro Vaue Rotator Object Reference), Index=0)
  ]}
 DoAddTimeSettingsModify {nodes [
   n1 = DoAddTimeSettingsModify!() -> (then, Navigation:EUINavigation Enum) {is=title K2Node_FunctionEntry}
   > n2 = Switch on EUINavigation!(execute, Selection:EUINavigation Enum=@n1) -> (Left:Exec, Right:Exec, Up:Exec, Down:Exec, Next:Exec, Previous:Exec, Invalid:Exec) {is=title K2Node_SwitchEnum}
   n3 = Select(
    Option 0:WBP Cilantro Vaue Rotator Object Reference=Get Rotator_TimeUsedBefore()
    Option 1:WBP Cilantro Vaue Rotator Object Reference=Get Rotator_ExtraTime()
    Index=IsHovered(self=Get Rotator_ExtraTime())
   ) -> (ReturnValue:WBP Cilantro Vaue Rotator Object Reference)
   > n4 = ClickDecrease(self:WBP Cilantro Vaue Rotator Object Reference=@n3)
   n5 = Get Rotator_ExtraTime()
   Return Node(
    execute=[@n6.then, @n4.then]
    ReturnValue=Select(
     Option 0:Widget Object Reference=Get Rotator_TimeUsedBefore()
     Option 1:Widget Object Reference=Get Rotator_ExtraTime()
     Index=IsHovered(self=@n5)
    ) -> (ReturnValue:Widget Object Reference)
   )
   n6 = ClickIncrease(execute=@n2.Right, self:WBP Cilantro Vaue Rotator Object Reference=@n3)
  ]}
 DoGrantTimeSettingModify {nodes [
   n1 = DoGrantTimeSettingModify!() -> (then, Navigation:EUINavigation Enum) {is=title K2Node_FunctionEntry}
   > n2 = Switch on EUINavigation(Selection=@n1)
   n3 = Get GrantTimeSetting()
   > n4 = ClickDecrease(self=@n3)
   Return Node(execute=[@n5.then, @n4.then], ReturnValue=@n3)
   n5 = ClickIncrease(execute=@n2.Right, self=@n3)
  ]}
 EventGraph {nodes [
   "On Clicked (AcceptButton)"!() -> (OutputDelegate, then, Button:Common Button Base Object Reference) {is=title K2Node_ComponentBoundEvent}
   n1 = Get GrantTimeSetting()
   > GrantAdditionalPlaytime!(
    execute
    self:Player Playtime View Model Object Reference=Get PlayerPlaytimeVM() -> (PlayerPlaytimeVM:Player Playtime View Model Object Reference)
    UserPin:String=Get User Pin() -> (User Pin:String)
    SecondsToGrant:int="Get (a copy)"(
     Array:Array of Integers=Get Durations() -> (Durations:Array of Integers)
     Dimension 1=Get Selection Index!(self:WBP Cilantro Vaue Rotator Object Reference=@n1) -> (Selection Index:int=0)
    ) -> (Output:int)
    ExpirationDate:"Date Time Structure (by ref)"=UTC Now!() -> (ReturnValue:Date Time Structure)
    ActivityType:EActivityLimitType Enum=Get SelectedActivityType() -> (SelectedActivityType:EActivityLimitType Enum=None)
   ) -> (then)
   > DeactivateWidget()
   > Fire Event Element Interacted(Element=Get AcceptButton() -> (AcceptButton:WBP UIKit Button Regular Object Reference))
   n2 = "On Selection Index Changed (GrantTimeSetting)"!() -> (OutputDelegate, then, Index:int=0) {is=title K2Node_ComponentBoundEvent}
   > Branch(Condition=Is Valid Index!(TargetArray:Array of Integers=Get Durations(), IndexToTest:int=@n2.Index) -> (ReturnValue:bool))
   > "SetText (Text)"(
    self=Get Text_Summary() -> (Text_Summary:Common Text Object Reference)
    InText=GetPlaytimeAddResultText!(
     self:Player Playtime View Model Object Reference=Get PlayerPlaytimeVM()
     AddedTime:"Timespan Structure (by ref)"=GetDurationTimespan!(SelectedIndex:int=@n2.Index) -> (DurationTimespan:Timespan Structure)
     ExpirationTime:"Date Time Structure (by ref)"=UTC Now()
     ActivityType:EActivityLimitType Enum=Get SelectedActivityType() -> (SelectedActivityType=None)
    ) -> (ReturnValue:Text)
   )
   n3 = "On Clicked (CloseButton)"!() -> (OutputDelegate, then, Button:Common Button Base Object Reference) {is=title K2Node_ComponentBoundEvent}
   > Fire Event Element Interacted(Element=@n3.Button)
   > DeactivateWidget()
   n4 = "On Clicked (Button_Close_AllowMoreTime)"!() -> (OutputDelegate, then, Button:Common Button Base Object Reference) {is=title K2Node_ComponentBoundEvent}
   > n5 = Fire Event Element Interacted(Element=@n4.Button)
   DeactivateWidget(execute=[@n16.then, @n5.then])
   "On Clicked (Button_PlayTime)"!() -> (OutputDelegate, then, Button:Common Button Base Object Reference) {is=title K2Node_ComponentBoundEvent}
   > SetSwitcherState!(execute, self, ActivityType:EActivityToShow Enum=ActivityPlay) -> (then)
   "On Clicked (Button_CreateTime)"!() -> (OutputDelegate, then, Button:Common Button Base Object Reference) {is=title K2Node_ComponentBoundEvent}
   > SetSwitcherState(ActivityType=ActivityCreate)
   n6 = "On Clicked (Button_Close_AddTime)"!() -> (OutputDelegate, then, Button:Common Button Base Object Reference) {is=title K2Node_ComponentBoundEvent}
   > Fire Event Element Interacted(Element=@n6.Button)
   n7 = Get PlayerPlaytimeVM()
   > n8 = Branch(Condition="Equal (Enum)"(
     A:EActivityToShow Enum=RefreshAddMoreTimeActivities!(self:Player Playtime View Model Object Reference=@n7) -> (ReturnValue:EActivityToShow Enum=None)
     B:EActivityToShow Enum=UserSelection
    ))
   > SetSwitcherState(ActivityType=UserSelection)
   DeactivateWidget(execute=@n8.else)
   n9 = "On Selection Index Changed (Rotator_ExtraTime)"!() -> (OutputDelegate, then, Index:int=0) {is=title K2Node_ComponentBoundEvent}
   > Branch(Condition=Is Valid Index(TargetArray=Get Durations(), IndexToTest=@n9.Index))
   > n10 = UpdateExpirationTimeSelector!(execute, self, IndexToSelect:int=@n9.Index) -> (then, SelectedTime:Date Time Structure)
   > UpdateAddResultDescription!(execute, self, SelectedIndex:int=@n9.Index, ExpirationTime:"Date Time Structure (by ref)"=@n10) -> (then)
   "On Selection Index Changed (Rotator_TimeUsedBefore)"!() -> (OutputDelegate, then, Index:int=0) {is=title K2Node_ComponentBoundEvent}
   > n11 = GetSelectedExpirationTime!(execute, self) -> (then, ExpirationTime:Date Time Structure)
   > UpdateAddResultDescription(SelectedIndex=Get Selection Index(self=Get Rotator_ExtraTime()) -> (Selection Index=0), ExpirationTime=@n11)
   > UpdateErrorContainerVisibility()
   "On Clicked (Button_AddTime)"!() -> (OutputDelegate, then, Button:Common Button Base Object Reference) {is=title K2Node_ComponentBoundEvent}
   > n12 = Branch(Condition=IsAddTimeDateTimeZoneValid!(
     self:Player Playtime View Model Object Reference=Get PlayerPlaytimeVM()
     ActivationDate:"Date Time Structure (by ref)"=Get DateAtActivation() -> (DateAtActivation:Date Time Structure)
    ) -> (ReturnValue:bool))
   > UpdateErrorContainerVisibility()
   > Branch(Condition=NOT Boolean(A=GetExpirationTimeInPast!() -> (IsInPast:bool)))
   n13 = Get Rotator_ExtraTime()
   n14 = Get Rotator_TimeUsedBefore()
   > GrantAdditionalPlaytime(
    self=Get PlayerPlaytimeVM()
    UserPin=Get User Pin()
    SecondsToGrant="Get (a copy)"(Array=Get Durations(), Dimension 1=Get Selection Index(self=@n13) -> (Selection Index=0))
    ExpirationDate="Get (a ref)"(
     Array:Array of Date Time Structures=Get ExpirationOptions() -> (ExpirationOptions:Array of Date Time Structures)
     Dimension 1=Get Selection Index(self=@n14) -> (Selection Index=0)
    ) -> (Output:"Date Time Structure (by ref)")
    ActivityType=Get SelectedActivityType() -> (SelectedActivityType=None)
   )
   > Fire Event Element Interacted(Element=Get Button_AddTime() -> (Button_AddTime:WBP UIKit Button Regular Object Reference))
   > DeactivateWidget()
   SetActiveWidget(execute=@n12.else, self=Get Switcher_Content(), Widget=Get VB_DayHasChanged() -> (VB_DayHasChanged:Vertical Box Object Reference))
   > RequestRefreshFocus(self:Common Activatable Widget Object Reference)
   n15 = "On Clicked (Button_Close_DayHasChanged)"!() -> (OutputDelegate, then, Button:Common Button Base Object Reference) {is=title K2Node_ComponentBoundEvent}
   > n16 = Fire Event Element Interacted(Element=@n15.Button)
   OnInputMethodChanged!() -> (OutputDelegate, then, InputType:ECommonInputType Enum) {is=title K2Node_CustomEvent}
   > RequestRefreshFocus()
   n17 = Event OnUserPinProvided!() -> (OutputDelegate, then, UserPin:String) {is=title K2Node_Event}
   > Set User Pin(User Pin:String=@n17.UserPin) -> (Output_Get:String)
   Event On Activated()
   > Init(self:Player Playtime View Model Object Reference=Get PlayerPlaytimeVM())
   > OnActivated()
   Event Construct()
   n18 = Create Event()
   n19 = Get CommonInputSubsystem()
   > Bind Event to On Input Method Changed(self=@n19, Delegate=@n18)
   > OnInputMethodChanged!(execute, self, InputType:ECommonInputType Enum=GetCurrentInputType(self=Get CommonInputSubsystem())) -> (then) {is=title K2Node_CallFunction}
   Event Destruct()
   > Unbind Event from On Input Method Changed(self=@n19, Delegate=@n18)
  ]}
 GetDurationTimespan {nodes [
   n1 = GetDurationTimespan!() -> (then, SelectedIndex:int) {is=title K2Node_FunctionEntry}
   n2 = Get Durations()
   > Return Node!(
    execute
    DurationTimespan:Timespan Structure=MakeTimespan!(Days:int=0, Hours:int=0, Minutes:int=0, Seconds:int="Get (a copy)"(Array=@n2, Dimension 1=@n1), Milliseconds:int=0) -> (ReturnValue:Timespan Structure)
   )
  ]}
 GetExpirationTimeInPast {nodes [
   GetExpirationTimeInPast() {is=title K2Node_FunctionEntry}
   n1 = Get ExpirationOptions()
   n2 = Get Selection Index(self=Get Rotator_TimeUsedBefore()) -> (Selection Index=0)
   > Return Node!(
    execute
    IsInPast:bool="DateTime > DateTime"!(A:Date Time Structure=UTC Now(), B:Date Time Structure="Get (a ref)"(Array=@n1, Dimension 1=@n2)) -> (ReturnValue:bool)
   )
  ]}
 GetFinalAddResultDescriptionText {nodes [
   GetFinalAddResultDescriptionText!() -> (then, PlaytimeAddResultText:Text) {is=title K2Node_FunctionEntry}
   n1 = Format Text(Format="LOCTEXT(\"create time\")")
   n2 = Format Text(Format="LOCTEXT(\"play time\")")
   n3 = Format Text(Format="LOCTEXT(\"time\")")
   n4 = Get SelectedActivityType() -> (SelectedActivityType=None)
   > Return Node!(
    execute
    ReturnValue:Text=Format Text!(
     Format="LOCTEXT(\"Adding more {Activity time} will overwrite the remainder of today’s schedule. {AddResult}\")"
     Activity time:Text=Select!(None:Text=@n3, CombinedTime:Text=@n3, ActivityPlay:Text=@n2, ActivityCreate:Text=@n1, Index:EActivityLimitType Enum=@n4) -> (ReturnValue:Text)
     AddResult:Text=Get PlaytimeAddResultText() -> (PlaytimeAddResultText:Text)
    ) -> (Result)
   )
  ]}
 GetSelectedExpirationTime {nodes [
   GetSelectedExpirationTime() {is=title K2Node_FunctionEntry}
   n1 = Get Rotator_TimeUsedBefore()
   > n2 = Branch(Condition=Is Valid Index(TargetArray:Array of Date Time Structures=Get ExpirationOptions(), IndexToTest=Get Selection Index(self=@n1) -> (Selection Index=0)))
   n3 = Get Rotator_TimeUsedBefore()
   > Return Node!(
    execute
    ExpirationTime:Date Time Structure="Get (a ref)"(Array=Get ExpirationOptions(), Dimension 1=Get Selection Index(self=@n3) -> (Selection Index=0))
   )
   n4 = Get ExpirationOptions()
   Return Node(
    execute=@n2.else
    ExpirationTime="Get (a ref)"(Array=@n4, Dimension 1=Last Index!(TargetArray:Array of Date Time Structures=@n4) -> (ReturnValue=0))
   )
  ]}
 On Available Durations Updated {nodes [
   n1 = On Available Durations Updated!() -> (then, Durations:Array of Integers) {is=title K2Node_FunctionEntry}
   > n2 = Set Durations(Durations:Array of Integers=@n1) -> (Output_Get:Array of Integers)
   > n3 = For Each Loop(Array:Array of Integers=@n2) -> (Array Element:int)
   n4 = Branch(execute=@n3.Completed, Condition=Get bEnableThyme(self=GetRuntimeOptions()))
   n5 = Get L_TextValues() -> (L_TextValues:Array of Texts)
   > Init With Text Entries!(execute, self:WBP Cilantro Vaue Rotator Object Reference=Get Rotator_ExtraTime(), Entries:Array of Texts=@n5) -> (then)
   Init With Text Entries(execute=@n4.else, self=Get GrantTimeSetting(), Entries=@n5)
   Add(
    execute=@n3.LoopBody
    TargetArray:Array of Texts=Get L_TextValues()
    NewItem:"Text (by ref)"=GetTimespanAsText!(
     Timespan:"Timespan Structure (by ref)"=MakeTimespan(Days=0, Hours=0, Minutes=0, Seconds=@n3, Milliseconds=0)
     bShowHoursWhenZero:bool=false
     bShowSeconds:bool=false
     FormatType:ETimespanAsTextFormat Enum=DaysHoursMinutesSeconds
     UnitDisplayType:ETimespanUnitDisplayFormat Enum=Full
     bUseSentenceCase:bool=false
    ) -> (ReturnValue:Text)
   ) {is=title K2Node_CallArrayFunction}
   Make Array!("[0]":int=1800, "[1]":int=3600, "[2]":int=5400, "[3]":int=7200) -> (Array:Array of Integers)
  ]}
 On Time Played Updated {nodes [
   On Time Played Updated!() -> (then, Timespan:Timespan Structure) {is=title K2Node_FunctionEntry}
   n1 = GetRuntimeOptions()
   > Branch(Condition=NOT Boolean(A=Get bEnableThyme(self=@n1)))
   > "SetText (Text)"(
    self=Get Text_Description() -> (Text_Description:Common Text Object Reference)
    InText=GetPlaytimeSummaryText!(
     self:Player Playtime View Model Object Reference=Get PlayerPlaytimeVM()
     ActivityType:EActivityLimitType Enum=Get SelectedActivityType() -> (SelectedActivityType=None)
    ) -> (ReturnValue:Text)
   )
   n2 = Get Selection Index(self=Get GrantTimeSetting()) -> (Selection Index=0)
   > Branch(Condition=Is Valid Index(TargetArray:Array of Integers=Get Durations(), IndexToTest=@n2))
   > "SetText (Text)"(
    self=Get Text_Summary()
    InText=GetPlaytimeAddResultText(
     self=Get PlayerPlaytimeVM()
     AddedTime=GetDurationTimespan(SelectedIndex=@n2) {is=title K2Node_CallFunction}
     ExpirationTime=UTC Now()
     ActivityType=Get SelectedActivityType() -> (SelectedActivityType=None)
    )
   )
  ]}
 OnActivated {nodes [
   OnActivated() {is=title K2Node_FunctionEntry}
   > SetSwitcherState(ActivityType=RefreshAddMoreTimeActivities(self=Get PlayerPlaytimeVM()) -> (ReturnValue=None))
   > n1 = Branch(Condition=Get bEnableThyme(self=GetRuntimeOptions()))
   > SetVisibility(self=Get HB_Error() -> (HB_Error:Horizontal Box Object Reference), InVisibility=Collapsed)
   > n2 = Set DateAtActivation(DateAtActivation:Date Time Structure=UTC Now()) -> (Output_Get:Date Time Structure)
   Fire Event Sub View Opened!(
    execute=[@n4.then, @n2.then]
    View:Widget Object Reference=Self-Reference()
    AdditionalInfo:Flow Analytics Additional Info Structure
    bIsOverlay:bool=false
   ) -> (then)
   "SetText (Text)"(
    execute=@n1.else
    self=Get Text_Description()
    InText=GetPlaytimeSummaryText(self=Get PlayerPlaytimeVM(), ActivityType=Get SelectedActivityType() -> (SelectedActivityType=None))
   )
   n3 = Get Selection Index(self=Get GrantTimeSetting()) -> (Selection Index=0)
   > Branch(Condition=Is Valid Index(TargetArray=Get Durations(), IndexToTest=@n3))
   > n4 = "SetText (Text)"(
    self=Get Text_Summary()
    InText=GetPlaytimeAddResultText(
     self=Get PlayerPlaytimeVM()
     AddedTime=GetDurationTimespan(SelectedIndex=@n3)
     ExpirationTime=UTC Now()
     ActivityType=Get SelectedActivityType() -> (SelectedActivityType=None)
    )
   )
  ]}
 OnAvailableExpirationOptionsUpdated {nodes [
   n1 = OnAvailableExpirationOptionsUpdated!() -> (then, ExpirationTimeOptions:Localized Expiration Dates Structure) {is=title K2Node_FunctionEntry}
   n2 = Break Localized Expiration Dates!(LocalizedExpirationDates:"Localized Expiration Dates Structure (by ref)"=@n1) -> (ExpirationDates:Array of Date Time Structures, TimeZoneId:String) {is=title K2Node_BreakStruct}
   > n3 = Set ExpirationOptions(ExpirationOptions:Array of Date Time Structures=@n2) -> (Output_Get:Array of Date Time Structures)
   > n4 = For Each Loop(Array:Array of Date Time Structures=@n3) -> (Array Element:Date Time Structure)
   Init With Text Entries(execute=@n4.Completed, self=Get Rotator_TimeUsedBefore(), Entries=Get LTimeTexts() -> (LTimeTexts:Array of Texts))
   Add(
    execute=@n4.LoopBody
    TargetArray=Get LTimeTexts()
    NewItem="As Time (from UTC)"!(InDateTime:"Date Time Structure (by ref)"=@n4, InTimeZone:String=@n2.TimeZoneId, InTimeStyle:EDateTimeStyle Enum=Short) -> (ReturnValue:Text)
   )
  ]}
 SetSwitcherState {nodes [
   n1 = SetSwitcherState!() -> (then, ActivityType:EActivityToShow Enum=None) {is=title K2Node_FunctionEntry}
   > n2 = Switch on EActivityToShow!(execute, Selection:EActivityToShow Enum=@n1) -> (None:Exec, CombinedTime:Exec, ActivityPlay:Exec, ActivityCreate:Exec, UserSelection:Exec) {is=title K2Node_SwitchEnum}
   n3 = Set SelectedActivityType(execute=@n2.ActivityCreate, SelectedActivityType:EActivityLimitType Enum=ActivityCreate) -> (Output_Get:EActivityLimitType Enum=None)
   SetActiveWidget(execute=[@n7.then, @n3.then, @n6.then], self=Get Switcher_Content(), Widget=Get VB_AddTime() -> (VB_AddTime:Vertical Box Object Reference))
   > "SetText (Text)"(
    self=Get Text_Header_AddTime() -> (Text_Header_AddTime:Common Text Object Reference)
    InText=Select(
     CombinedTime="LOCTABLE(\"/Cilantro/RichTextStyles/StringTable_TimeLimitTexts\", \"AddMoreTime\")"
     ActivityPlay="LOCTABLE(\"&StringTable_TimeLimitTexts\", \"AddPlayTime\")"
     ActivityCreate="LOCTABLE(\"&StringTable_TimeLimitTexts\", \"AddMoreCreateTime\")"
     Index=Get SelectedActivityType() -> (SelectedActivityType=None)
    )
   )
   > SetSettingsText!(
    execute
    self:WBP Cilantro Vaue Rotator Object Reference=Get Rotator_ExtraTime()
    Text:Text=Select(
     CombinedTime="LOCTEXT(\"Extra time\")"
     ActivityPlay="LOCTEXT(\"Extra play time\")"
     ActivityCreate="LOCTEXT(\"Extra create time\")"
     Index=Get SelectedActivityType() -> (SelectedActivityType=None)
    )
   ) -> (then)
   > ClearSelectedRotatorIndex() {is=title K2Node_CallFunction}
   > n4 = GetSelectedExpirationTime() {is=title K2Node_CallFunction}
   > n5 = UpdateAddResultDescription(SelectedIndex=Get Selection Index(self=Get Rotator_ExtraTime()) -> (Selection Index=0), ExpirationTime=@n4)
   RequestRefreshFocus(execute=[@n5.then, @n8.then])
   n6 = Set SelectedActivityType(execute=@n2.ActivityPlay, SelectedActivityType=ActivityPlay) -> (Output_Get=None)
   Set SelectedActivityType(execute=@n2.CombinedTime, SelectedActivityType=CombinedTime) -> (Output_Get=None)
   > n7 = Branch(Condition=Get bEnableThyme(self=GetRuntimeOptions()))
   SetActiveWidget(execute=@n7.else, self=Get Switcher_Content(), Widget=Get VB_AddTimeOLD() -> (VB_AddTimeOLD:Vertical Box Object Reference))
   n8 = SetActiveWidget(execute=@n2.UserSelection, self=Get Switcher_Content(), Widget=Get VB_MoreTime() -> (VB_MoreTime:Vertical Box Object Reference))
  ]}
 UpdateAddResultDescription {nodes [
   n1 = UpdateAddResultDescription!() -> (then, SelectedIndex:int, ExpirationTime:"Date Time Structure (by ref)") {is=title K2Node_FunctionEntry}
   > n2 = GetFinalAddResultDescriptionText!(
    execute
    self
    PlaytimeAddResultText:Text=GetPlaytimeAddResultText(
     self=Get PlayerPlaytimeVM()
     AddedTime=GetDurationTimespan(SelectedIndex=@n1)
     ExpirationTime=@n1.ExpirationTime
     ActivityType=Get SelectedActivityType() -> (SelectedActivityType=None)
    )
   ) -> (then, ReturnValue:Text) {is=title K2Node_CallFunction}
   > "SetText (Text)"(self=Get Text_Desc_AddTime() -> (Text_Desc_AddTime:Common Text Object Reference), InText=@n2)
  ]}
 UpdateErrorContainerVisibility {nodes [
   UpdateErrorContainerVisibility() {is=title K2Node_FunctionEntry}
   > SetVisibility(
    self=Get HB_Error()
    InVisibility="To Visibility (Boolean)"!(
     bIsVisible:bool=GetExpirationTimeInPast() {is=title K2Node_CallFunction}
     TrueVisibility:ESlateVisibility Enum=SelfHitTestInvisible
     FalseVisibility:ESlateVisibility Enum=Collapsed
    ) -> (ReturnValue:ESlateVisibility Enum=Visible)
   )
  ]}
 UpdateExpirationTimeSelector {nodes [
   n1 = UpdateExpirationTimeSelector!() -> (then, IndexToSelect:int) {is=title K2Node_FunctionEntry}
   > n2 = Branch(Condition=Is Valid Index(TargetArray:Array of Date Time Structures=Get ExpirationOptions(), IndexToTest=@n1))
   n3 = Get Rotator_TimeUsedBefore()
   n4 = Get IndexToSelect() -> (IndexToSelect:int=0)
   > SetSelectedItem(self=@n3, Index=@n4)
   n5 = Get ExpirationOptions()
   > Return Node!(execute, SelectedTime:Date Time Structure="Get (a ref)"(Array=@n5, Dimension 1=@n4))
   n6 = Last Index(TargetArray=Get ExpirationOptions()) -> (ReturnValue=0)
   SetSelectedItem(execute=@n2.else, self=@n3, Index=@n6)
   > Return Node(SelectedTime="Get (a ref)"(Array=@n5, Dimension 1=@n6))
  ]}
]
variables [
 DateAtActivation {type=DateTime, default=0001.01.01-00.00.00}
 Durations {type="array<integer>"}
 ExpirationOptions {type="array<DateTime>", default []}
 IsPlayTimeOpened {type=boolean}
 SelectedActivityType {type=EActivityLimitType, default=None}
 User Pin {type=string, default=""}
]
widgets {
 info {
  inheritedWidgetCount=0
  namedSlotCount=1
  parentClass=/Script/CilantroUI.FortPlaytimePinActivatableWidget
  rootWidgetClass=/FortUILibrary/Systems/Dialog/WBP_UIKit_Dialog_Internal.WBP_UIKit_Dialog_Internal_C
  widgetCount=36
 }
 widgets [
  WBP_UIKit_Dialog_Base {
   widgetClassPath=/FortUILibrary/Systems/Dialog/WBP_UIKit_Dialog_Internal
   bIsVariable=true
   widget=WBP_UIKit_Dialog_Base
   properties {
    bIsFocusable=true
    designTimePreview_DialogVM=null
    navigation=&self:WidgetTree.WBP_UIKit_Dialog_Base.WidgetNavigation_0
    useText=None
    useTitle=false
    buttons {
     each Button Horizontal Fill=false
     each Button Min Width=0
     each Button Size=Medium
     focuse Button=0
     space Between Buttons=30
     cTA Buttons Types []
     group Placement {
      horizontal Alignment=HAlign_Fill
      vertical Alignment=VAlign_Bottom
      grid {column=1, column Span=0, layer=0, row=2, row Span=0, nudge=<0 0>}
      padding=<ltrb 0 0 0 0>
     }
    }
    dialogSIze {
     height Override {enable=false, value=800}
     max Aspect Ratio {enable=false, value=0}
     max Desired Height {enable=true, value=800}
     max Desired Width {enable=false, value=0}
     min Aspect Ratio {enable=false, value=0}
     min Desired Height {enable=false, value=0}
     min Desired Width {enable=false, value=0}
     width Override {enable=true, value=864}
    }
   }
  }
  Overlay_0 {widgetClassPath=&Overlay, namedSlotHost=WBP_UIKit_Dialog_Base, widget=Overlay_0}
  Switcher_Content {
   widgetClassPath=&CommonVisibilitySwitcher
   bIsVariable=true
   parent=Overlay_0
   widget=Switcher_Content
   properties {navigation=&self:WidgetTree.Switcher_Content.WidgetNavigation_0}
   slot {class=OverlaySlot, properties {horizontalAlignment=HAlign_Fill, verticalAlignment=VAlign_Fill, padding=<ltrb 80 64 80 48>}}
  }
  VB_MoreTime {
   widgetClassPath=&VerticalBox
   bIsVariable=true
   parent=Switcher_Content
   widget=VB_MoreTime
   properties {navigation=&self:WidgetTree.VB_MoreTime.WidgetNavigation_0}
   slot {class=CommonVisibilitySwitcherSlot}
  }
  HorizontalBox_1 {widgetClassPath=&HorizontalBox, parent=VB_MoreTime, widget=HorizontalBox_1, slot {class=VerticalBoxSlot}}
  Image_AllowMoreTime {
   widgetClassPath=&Image
   bIsVariable=true
   parent=HorizontalBox_1
   widget=Image_AllowMoreTime
   properties {brush {
     bIsDynamicallyLoaded=false
     drawAs=Image
     imageType=FullColor
     mirroring=NoMirror
     resourceName=None
     resourceObject=/FortUILibrary/Textures/MaterialInstances/MI_UIKit_Icon_UpdatePending
     tiling=NoTile
     imageSize=<42 42>
     margin=<ltrb 0 0 0 0>
     outlineSettings {
      bUseBrushTransparency=false
      roundingType=HalfHeightRadius
      width=0
      color {colorUseRule=UseColor_Specified, specifiedColor=<rgba 0 0 0 0>}
      cornerRadii=<0 0 0 0>
     }
     tintColor {colorUseRule=UseColor_Specified, specifiedColor=<rgba 1 1 1 1>}
     uVRegion {bIsValid=false, max=<0 0>, min=<0 0>}
    }}
   slot {class=HorizontalBoxSlot, properties {horizontalAlignment=HAlign_Left, verticalAlignment=VAlign_Center, padding=<ltrb 0 0 16 0>}}
  }
  Text_Header_AllowMoreTime {
   widgetClassPath=&CommonTextBlock
   parent=HorizontalBox_1
   widget=Text_Header_AllowMoreTime
   properties {
    autoWrapText=true
    text=Allow More Time
    font {
     bForceMonospaced=false
     bMaterialIsStencil=false
     fontMaterial=null
     fontObject=&NotoSans
     letterSpacing=0
     monospacedWidth=1
     size=24
     skewAmount=0
     typefaceFontName=NotoSans-Regular
     outlineSettings {
      bApplyOutlineToDropShadows=false
      bMiteredCorners=false
      bSeparateFillAlpha=false
      outlineMaterial=null
      outlineSize=0
      outlineColor=<rgba 0 0 0 1>
     }
    }
    shadowOffset=<0 0>
   }
   slot {class=HorizontalBoxSlot, properties {size {sizeRule=Fill, value=1}}}
  }
  Text_Description_AllowMoreTime {
   widgetClassPath=&CommonTextBlock
   bIsVariable=true
   parent=VB_MoreTime
   widget=Text_Description_AllowMoreTime
   properties {
    autoWrapText=true
    style=&TextStyle-Base_C
    text=What type of time would you like to allow?
    font {
     bForceMonospaced=false
     bMaterialIsStencil=false
     fontMaterial=null
     fontObject=&NotoSans
     letterSpacing=0
     monospacedWidth=1
     size=24
     skewAmount=0
     typefaceFontName=NotoSans-Regular
     outlineSettings {
      bApplyOutlineToDropShadows=false
      bMiteredCorners=false
      bSeparateFillAlpha=false
      outlineMaterial=null
      outlineSize=0
      outlineColor=<rgba 0 0 0 1>
     }
    }
    shadowOffset=<0 0>
   }
   slot {class=VerticalBoxSlot, properties {padding=<ltrb 6 40 6 0>}}
  }
  Button_PlayTime {
   widgetClassPath=&/WBP_AddTime_Button
   bIsVariable=true
   parent=VB_MoreTime
   widget=Button_PlayTime
   properties {
    bHideInputAction=true
    bHideInputActionWithKeyboard=true
    navigation=&self:WidgetTree.Button_PlayTime.WidgetNavigation_0
    text=Play time
   }
   slot {class=VerticalBoxSlot, properties {padding=<ltrb 8 24 8 0>}}
  }
  Button_CreateTime {
   widgetClassPath=&WBP_AddTime_Button
   bIsVariable=true
   parent=VB_MoreTime
   widget=Button_CreateTime
   properties {
    bHideInputAction=true
    bHideInputActionWithKeyboard=true
    navigation=&self:WidgetTree.Button_CreateTime.WidgetNavigation_0
    text=Create time
   }
   slot {class=VerticalBoxSlot, properties {padding=<ltrb 8 0 8 0>}}
  }
  Button_Close_AllowMoreTime {
   widgetClassPath=&WBP_UIKit_Button_Regular
   bIsVariable=true
   parent=VB_MoreTime
   widget=Button_Close_AllowMoreTime
   properties {
    bHideInputAction=true
    buttonType=Quiet
    minHeight=56
    navigation=&self:WidgetTree.Button_Close_AllowMoreTime.WidgetNavigation_0
    text=CANCEL
    triggeringInputAction {dataTable=&InputActionDataTable, rowName=Close}
   }
   slot {
    class=VerticalBoxSlot
    properties {horizontalAlignment=HAlign_Left, verticalAlignment=VAlign_Bottom, padding=<ltrb 0 64 0 0>, size {sizeRule=Fill, value=1}}
   }
  }
  VB_AddTime {
   widgetClassPath=&VerticalBox
   bIsVariable=true
   parent=Switcher_Content
   widget=VB_AddTime
   slot {class=CommonVisibilitySwitcherSlot}
  }
  HorizontalBox {widgetClassPath=&HorizontalBox, parent=VB_AddTime, widget=HorizontalBox, slot {class=VerticalBoxSlot}}
  Image_AddTime {
   widgetClassPath=&Image
   bIsVariable=true
   parent=HorizontalBox
   widget=Image_AddTime
   properties {brush {
     bIsDynamicallyLoaded=false
     drawAs=Image
     imageType=FullColor
     mirroring=NoMirror
     resourceName=None
     resourceObject=&MI_UIKit_Icon_UpdatePending
     tiling=NoTile
     imageSize=<42 42>
     margin=<ltrb 0 0 0 0>
     outlineSettings {
      bUseBrushTransparency=false
      roundingType=HalfHeightRadius
      width=0
      color {colorUseRule=UseColor_Specified, specifiedColor=<rgba 0 0 0 0>}
      cornerRadii=<0 0 0 0>
     }
     tintColor {colorUseRule=UseColor_Specified, specifiedColor=<rgba 1 1 1 1>}
     uVRegion {bIsValid=false, max=<0 0>, min=<0 0>}
    }}
   slot {class=HorizontalBoxSlot, properties {horizontalAlignment=HAlign_Left, verticalAlignment=VAlign_Center, padding=<ltrb 0 0 16 0>}}
  }
  Text_Header_AddTime {
   widgetClassPath=&CommonTextBlock
   bIsVariable=true
   parent=HorizontalBox
   widget=Text_Header_AddTime
   properties {
    autoWrapText=true
    text="Add More {PLAY} time"
    font {
     bForceMonospaced=false
     bMaterialIsStencil=false
     fontMaterial=null
     fontObject=&NotoSans
     letterSpacing=0
     monospacedWidth=1
     size=24
     skewAmount=0
     typefaceFontName=NotoSans-Regular
     outlineSettings {
      bApplyOutlineToDropShadows=false
      bMiteredCorners=false
      bSeparateFillAlpha=false
      outlineMaterial=null
      outlineSize=0
      outlineColor=<rgba 0 0 0 1>
     }
    }
    shadowOffset=<0 0>
   }
   slot {class=HorizontalBoxSlot, properties {size {sizeRule=Fill, value=1}}}
  }
  Rotator_ExtraTime {
   widgetClassPath=/Cilantro/Subwidgets/WBP_Cilantro_VaueRotator
   bIsVariable=true
   parent=VB_AddTime
   widget=Rotator_ExtraTime
   properties {
    is Show OptionsBackground=false
    navigation=&self:WidgetTree.Rotator_ExtraTime.WidgetNavigation_0
    setting Header Text="Extra {play} time"
   }
   slot {class=VerticalBoxSlot, properties {padding=<ltrb 0 87 0 0>}}
  }
  Rotator_TimeUsedBefore {
   widgetClassPath=&WBP_Cilantro_VaueRotator
   bIsVariable=true
   parent=VB_AddTime
   widget=Rotator_TimeUsedBefore
   properties {
    is Show OptionsBackground=false
    navigation=&self:WidgetTree.Rotator_TimeUsedBefore.WidgetNavigation_0
    setting Header Text=Time can be used before
   }
   slot {class=VerticalBoxSlot}
  }
  HB_Error {
   widgetClassPath=&HorizontalBox
   bIsVariable=true
   parent=VB_AddTime
   widget=HB_Error
   properties {visibility=Collapsed}
   slot {class=VerticalBoxSlot, properties {padding=<ltrb 0 26 0 0>}}
  }
  StatusIndicator {
   widgetClassPath=/FortUILibrary/UIKit/StatusIndicator/WBP_UIKit_StatusIndicator
   parent=HB_Error
   widget=StatusIndicator
   properties {type=Negative, overrideStatusIndicatorVisibility=true}
   slot {class=HorizontalBoxSlot, properties {horizontalAlignment=HAlign_Left, verticalAlignment=VAlign_Top}}
  }
  Text_Desc_Error {
   widgetClassPath=&CommonTextBlock
   parent=HB_Error
   widget=Text_Desc_Error
   properties {
    autoWrapText=true
    bIsScrollingEnabled=false
    text=This time has passed. Choose a time in the future.
    colorAndOpacity {colorUseRule=UseColor_Specified, specifiedColor=<rgba 1 0.147 0.1329 1>}
    font {
     bForceMonospaced=false
     bMaterialIsStencil=false
     fontMaterial=null
     fontObject=&NotoSans
     letterSpacing=0
     monospacedWidth=1
     size=24
     skewAmount=0
     typefaceFontName=NotoSans-Regular
     outlineSettings {
      bApplyOutlineToDropShadows=false
      bMiteredCorners=false
      bSeparateFillAlpha=false
      outlineMaterial=null
      outlineSize=0
      outlineColor=<rgba 0 0 0 1>
     }
    }
    shadowOffset=<0 0>
   }
   slot {class=HorizontalBoxSlot, properties {verticalAlignment=VAlign_Center, padding=<ltrb 10 0 0 0>, size {sizeRule=Fill, value=1}}}
  }
  Text_Desc_AddTime {
   widgetClassPath=&CommonTextBlock
   bIsVariable=true
   parent=VB_AddTime
   widget=Text_Desc_AddTime
   properties {
    autoWrapText=true
    bAutoCollapseWithEmptyText=true
    text=Your child will have 45 mins to use before 7:30 PM today.
    font {
     bForceMonospaced=false
     bMaterialIsStencil=false
     fontMaterial=null
     fontObject=&NotoSans
     letterSpacing=0
     monospacedWidth=1
     size=24
     skewAmount=0
     typefaceFontName=NotoSans-Regular
     outlineSettings {
      bApplyOutlineToDropShadows=false
      bMiteredCorners=false
      bSeparateFillAlpha=false
      outlineMaterial=null
      outlineSize=0
      outlineColor=<rgba 0 0 0 1>
     }
    }
    shadowOffset=<0 0>
   }
   slot {class=VerticalBoxSlot, properties {padding=<ltrb 0 40 0 0>, size {sizeRule=Fill, value=1}}}
  }
  GridPanel_0 {
   widgetClassPath=&GridPanel
   parent=VB_AddTime
   widget=GridPanel_0
   properties {bOverride_Cursor=true}
   slot {class=VerticalBoxSlot, properties {horizontalAlignment=HAlign_Left, padding=<ltrb 0 64 0 0>}}
  }
  Button_AddTime {
   widgetClassPath=&WBP_UIKit_Button_Regular
   bIsVariable=true
   parent=GridPanel_0
   widget=Button_AddTime
   properties {bHideInputAction=true, navigation=&self:WidgetTree.Button_AddTime.WidgetNavigation_0, text=Add time}
   slot {class=GridSlot}
  }
  Button_Close_AddTime {
   widgetClassPath=&WBP_UIKit_Button_Regular
   bIsVariable=true
   parent=GridPanel_0
   widget=Button_Close_AddTime
   properties {
    bHideInputAction=true
    buttonType=Quiet
    minHeight=56
    navigation=&self:WidgetTree.Button_Close_AddTime.WidgetNavigation_0
    text=Back
    triggeringInputAction {dataTable=&InputActionDataTable, rowName=Close}
   }
   slot {class=GridSlot, properties {column=1, padding=<ltrb 24 0 0 0>}}
  }
  VB_AddTimeOLD {
   widgetClassPath=&VerticalBox
   bIsVariable=true
   parent=Switcher_Content
   widget=VB_AddTimeOLD
   slot {class=CommonVisibilitySwitcherSlot}
  }
  Text_Header {
   widgetClassPath=&CommonTextBlock
   parent=VB_AddTimeOLD
   widget=Text_Header
   properties {
    style=&TextStyle-Base_C
    text=ADD MORE TIME
    font {
     bForceMonospaced=false
     bMaterialIsStencil=false
     fontMaterial=null
     fontObject=&NotoSans
     letterSpacing=0
     monospacedWidth=1
     size=24
     skewAmount=0
     typefaceFontName=NotoSans-Regular
     outlineSettings {
      bApplyOutlineToDropShadows=false
      bMiteredCorners=false
      bSeparateFillAlpha=false
      outlineMaterial=null
      outlineSize=0
      outlineColor=<rgba 0 0 0 1>
     }
    }
    shadowOffset=<0 0>
   }
   slot {class=VerticalBoxSlot}
  }
  Text_Description {
   widgetClassPath=&CommonTextBlock
   bIsVariable=true
   parent=VB_AddTimeOLD
   widget=Text_Description
   properties {
    autoWrapText=true
    style=&TextStyle-Base_C
    text=Your child has played for 2 hours today and has no playtime left. How much more time would you like to add?
    font {
     bForceMonospaced=false
     bMaterialIsStencil=false
     fontMaterial=null
     fontObject=&NotoSans
     letterSpacing=0
     monospacedWidth=1
     size=24
     skewAmount=0
     typefaceFontName=NotoSans-Regular
     outlineSettings {
      bApplyOutlineToDropShadows=false
      bMiteredCorners=false
      bSeparateFillAlpha=false
      outlineMaterial=null
      outlineSize=0
      outlineColor=<rgba 0 0 0 1>
     }
    }
    shadowOffset=<0 0>
   }
   slot {class=VerticalBoxSlot, properties {padding=<ltrb 6 24 6 0>}}
  }
  GrantTimeSetting {
   widgetClassPath=&WBP_Cilantro_VaueRotator
   bIsVariable=true
   parent=VB_AddTimeOLD
   widget=GrantTimeSetting
   properties {navigation=&self:WidgetTree.GrantTimeSetting.WidgetNavigation_0}
   slot {class=VerticalBoxSlot, properties {padding=<ltrb 0 32 0 0>}}
  }
  Text_Summary {
   widgetClassPath=&CommonTextBlock
   bIsVariable=true
   parent=VB_AddTimeOLD
   widget=Text_Summary
   properties {
    autoWrapText=true
    style=&TextStyle-Base_C
    text=You child will have a total of 30 minutes of play time.
    font {
     bForceMonospaced=false
     bMaterialIsStencil=false
     fontMaterial=null
     fontObject=&NotoSans
     letterSpacing=0
     monospacedWidth=1
     size=24
     skewAmount=0
     typefaceFontName=NotoSans-Regular
     outlineSettings {
      bApplyOutlineToDropShadows=false
      bMiteredCorners=false
      bSeparateFillAlpha=false
      outlineMaterial=null
      outlineSize=0
      outlineColor=<rgba 0 0 0 1>
     }
    }
    shadowOffset=<0 0>
   }
   slot {class=VerticalBoxSlot, properties {padding=<ltrb 6 24 6 0>}}
  }
  HorizontalBox_0 {
   widgetClassPath=&HorizontalBox
   parent=VB_AddTimeOLD
   widget=HorizontalBox_0
   slot {class=VerticalBoxSlot, properties {verticalAlignment=VAlign_Bottom, padding=<ltrb 0 64 0 0>}}
  }
  AcceptButton {
   widgetClassPath=&WBP_UIKit_Button_Regular
   bIsVariable=true
   parent=HorizontalBox_0
   widget=AcceptButton
   properties {minHeight=56, text=DONE, triggeringInputAction {dataTable=&InputActionDataTable, rowName=Accept}}
   slot {class=HorizontalBoxSlot, properties {verticalAlignment=VAlign_Center, padding=<ltrb 0 0 24 0>, size {sizeRule=Fill, value=1}}}
  }
  CloseButton {
   widgetClassPath=&WBP_UIKit_Button_Regular
   bIsVariable=true
   parent=HorizontalBox_0
   widget=CloseButton
   properties {buttonType=Quiet, minHeight=56, text=CANCEL, triggeringInputAction {dataTable=&InputActionDataTable, rowName=Close}}
   slot {class=HorizontalBoxSlot, properties {size {sizeRule=Fill, value=1}}}
  }
  VB_DayHasChanged {
   widgetClassPath=&VerticalBox
   bIsVariable=true
   parent=Switcher_Content
   widget=VB_DayHasChanged
   slot {class=CommonVisibilitySwitcherSlot}
  }
  Text_Header_DayHasChanged {
   widgetClassPath=&CommonTextBlock
   parent=VB_DayHasChanged
   widget=Text_Header_DayHasChanged
   properties {
    autoWrapText=true
    bIsScrollingEnabled=false
    text=DAY HAS CHANGED
    font {
     bForceMonospaced=false
     bMaterialIsStencil=false
     fontMaterial=null
     fontObject=&NotoSans
     letterSpacing=0
     monospacedWidth=1
     size=24
     skewAmount=0
     typefaceFontName=NotoSans-Regular
     outlineSettings {
      bApplyOutlineToDropShadows=false
      bMiteredCorners=false
      bSeparateFillAlpha=false
      outlineMaterial=null
      outlineSize=0
      outlineColor=<rgba 0 0 0 1>
     }
    }
    shadowOffset=<0 0>
   }
   slot {class=VerticalBoxSlot}
  }
  Text_Desc_DayHasChanged {
   widgetClassPath=&CommonTextBlock
   parent=VB_DayHasChanged
   widget=Text_Desc_DayHasChanged
   properties {
    autoWrapText=true
    bIsScrollingEnabled=false
    text=The day changed while you were adding time. Add time again.
    font {
     bForceMonospaced=false
     bMaterialIsStencil=false
     fontMaterial=null
     fontObject=&NotoSans
     letterSpacing=0
     monospacedWidth=1
     size=24
     skewAmount=0
     typefaceFontName=NotoSans-Regular
     outlineSettings {
      bApplyOutlineToDropShadows=false
      bMiteredCorners=false
      bSeparateFillAlpha=false
      outlineMaterial=null
      outlineSize=0
      outlineColor=<rgba 0 0 0 1>
     }
    }
    shadowOffset=<0 0>
   }
   slot {class=VerticalBoxSlot, properties {padding=<ltrb 0 48 0 0>}}
  }
  Button_Close_DayHasChanged {
   widgetClassPath=&WBP_UIKit_Button_Regular
   bIsVariable=true
   parent=VB_DayHasChanged
   widget=Button_Close_DayHasChanged
   properties {
    bHideInputAction=true
    buttonType=Quiet
    minHeight=56
    navigation=&self:WidgetTree.Button_Close_DayHasChanged.WidgetNavigation_0
    text=Close
    triggeringInputAction {dataTable=&InputActionDataTable, rowName=Close}
   }
   slot {class=VerticalBoxSlot, properties {horizontalAlignment=HAlign_Left, padding=<ltrb 0 64 0 0>}}
  }
 ]
}
