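# Modal dialog that lets a player ask a parent or guardian by email for more play or create time.
# On activation, the PlayerPlaytime view model picks the activity, either sending the request immediately or showing Play/Create choice buttons.
# Request status events swap between collapsible sending, sent and error panels, with sent text tailored to the activity.
# Panel changes and input-method changes refresh CommonUI focus, targeting the visible panel's button.
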WidgetBlueprint /Cilantro/WBP_AskForMoreTime
graphs [
 BP_GetDesiredFocusTarget {nodes [
   Get Desired Focus Target()
   > n1 = Branch(Condition=IsVisible(self=Get VB_AskForMoreTime() -> (VB_AskForMoreTime:Vertical Box Object Reference)))
   > Return Node!(execute, ReturnValue:Widget Object Reference=Get Button_PlayTime() -> (Button_PlayTime:WBP Add Time Button Object Reference))
   n2 = Branch(execute=@n1.else, Condition=IsVisible(self=Get VB_EmailSent() -> (VB_EmailSent:Vertical Box Object Reference)))
   > Return Node(ReturnValue=Get Button_Close_EmailSent() -> (Button_Close_EmailSent:WBP UIKit Button Regular Object Reference))
   Return Node(execute=@n2.else, ReturnValue=Get Button_Close_Failed() -> (Button_Close_Failed:WBP UIKit Button Regular Object Reference))
  ]}
 EventGraph {nodes [
   "On Clicked (Button_PlayTime)"!() -> (OutputDelegate, then, Button:Common Button Base Object Reference) {is=title K2Node_ComponentBoundEvent}
   > Set ActivityToShow(ActivityToShow:EActivityToShow Enum=ActivityPlay) -> (Output_Get:EActivityToShow Enum=None)
   > RequestAdditionalPlaytime!(
    execute
    self:Player Playtime View Model Object Reference=Get PlayerPlaytimeVM() -> (PlayerPlaytimeVM:Player Playtime View Model Object Reference)
    ActivityType:EActivityLimitType Enum=ActivityPlay
   ) -> (then)
   "On Clicked (Button_CreateTime)"!() -> (OutputDelegate, then, Button:Common Button Base Object Reference) {is=title K2Node_ComponentBoundEvent}
   > Set ActivityToShow(ActivityToShow=ActivityCreate) -> (Output_Get=None)
   > RequestAdditionalPlaytime(self=Get PlayerPlaytimeVM(), ActivityType=ActivityCreate)
   n1 = "On Clicked (Button_Close_EmailSent)"!() -> (OutputDelegate, then, Button:Common Button Base Object Reference) {is=title K2Node_ComponentBoundEvent}
   DeactivateWidget(execute=[@n1.then, @n2.then, @n3.then])
   n2 = "On Clicked (Button_Close_AskForMoreTime)"!() -> (OutputDelegate, then, Button:Common Button Base Object Reference) {is=title K2Node_ComponentBoundEvent}
   n3 = "On Clicked (Button_Close_Failed)"!() -> (OutputDelegate, then, Button:Common Button Base Object Reference) {is=title K2Node_ComponentBoundEvent}
   n4 = OnRequestAdditionalPlaytimeStatusUpdated!() -> (OutputDelegate, then, RequestStatus:EPlaytimeRequestState Enum=None) {is=title K2Node_CustomEvent}
   > n5 = Switch on EPlaytimeRequestState!(execute, Selection:EPlaytimeRequestState Enum=@n4.RequestStatus) -> (None:Exec, Waiting:Exec, Success:Exec, Failure:Exec) {is=title K2Node_SwitchEnum}
   SetVisibility(execute=@n5.Failure, self=Get VB_ErrorState() -> (VB_ErrorState:Vertical Box Object Reference), InVisibility=SelfHitTestInvisible)
   > n6 = SetVisibility(
    self=[Get VB_SendingEmail() -> (VB_SendingEmail:Vertical Box Object Reference), Get VB_EmailSent(), Get VB_AskForMoreTime()]
    InVisibility=Collapsed
   )
   RequestRefreshFocus(execute=[@n7.then, @n6.then], self:Common Activatable Widget Object Reference)
   SetVisibility(execute=@n5.Success, self=Get VB_EmailSent(), InVisibility=SelfHitTestInvisible)
   > SetVisibility(self=[Get VB_ErrorState(), Get VB_SendingEmail(), Get VB_AskForMoreTime()], InVisibility=Collapsed)
   > n7 = SetDescriptionText!(execute, self, ActivityToShow:EActivityToShow Enum=Get ActivityToShow() -> (ActivityToShow:EActivityToShow Enum=None)) -> (then)
   SetVisibility(execute=@n5.Waiting, self=Get VB_SendingEmail(), InVisibility=SelfHitTestInvisible)
   > SetVisibility(self=[Get VB_EmailSent(), Get VB_AskForMoreTime(), Get VB_ErrorState()], InVisibility=Collapsed)
   OnInputMethodChanged!() -> (OutputDelegate, then, NewInputType:ECommonInputType Enum) {is=title K2Node_CustomEvent}
   > RequestRefreshFocus()
   Event Construct()
   n8 = Create Event()
   n9 = Get PlayerPlaytimeVM()
   > Bind Event to On Request Additional Playtime Status Update(self:Player Playtime View Model Object Reference=@n9, Delegate=@n8) {is=title K2Node_AddDelegate}
   n10 = Create Event()
   n11 = Get CommonInputSubsystem()
   > Bind Event to On Input Method Changed(self=@n11, Delegate=@n10)
   > OnInputMethodChanged!(execute, self, NewInputType:ECommonInputType Enum=GetCurrentInputType(self=Get CommonInputSubsystem())) -> (then) {is=title K2Node_CallFunction}
   Event On Activated()
   > Init(self:Player Playtime View Model Object Reference=Get PlayerPlaytimeVM())
   > SetModalStates!(
    execute
    self
    ActivityType:EActivityToShow Enum=RefreshAskMoreTimeActivities!(self:Player Playtime View Model Object Reference=Get PlayerPlaytimeVM()) -> (ReturnValue:EActivityToShow Enum=None)
   ) -> (then)
   Event Destruct()
   > Unbind Event from On Request Additional Playtime Status Update(self:Player Playtime View Model Object Reference=@n9, Delegate=@n8) {is=title K2Node_RemoveDelegate}
   > Unbind Event from On Input Method Changed(self=@n11, Delegate=@n10)
  ]}
 SetDescriptionText {nodes [
   n1 = SetDescriptionText!() -> (then, ActivityToShow:EActivityToShow Enum=None) {is=title K2Node_FunctionEntry}
   > "SetText (Text)"(
    self=Get Text_Description_Email() -> (Text_Description_Email:Common Text Object Reference)
    InText=Select!(
     None:Text
     CombinedTime:Text="LOCTEXT(\"We’ve sent an email to your parent or guardian to ask for more time.\")"
     ActivityPlay:Text="LOCTEXT(\"We\\'ve sent an email to your parent or guardian to ask for more play time.\")"
     ActivityCreate:Text="LOCTEXT(\"We\\'ve sent an email to your parent or guardian to ask for more create time.\")"
     UserSelection:Text
     Index:EActivityToShow Enum=@n1
    ) -> (ReturnValue:Text)
   )
  ]}
 SetModalStates {nodes [
   n1 = SetModalStates!() -> (then, ActivityType:EActivityToShow Enum=None) {is=title K2Node_FunctionEntry}
   > n2 = Set ActivityToShow(ActivityToShow=@n1) -> (Output_Get=None)
   > n3 = Switch on EActivityToShow!(execute, Selection:EActivityToShow Enum=@n2) -> (None:Exec, CombinedTime:Exec, ActivityPlay:Exec, ActivityCreate:Exec, UserSelection:Exec) {is=title K2Node_SwitchEnum}
   n4 = Set ActivityToShow(
    execute=@n3.UserSelection
    ActivityToShow=CheckBlockedStateForUserSelection!(self:Player Playtime View Model Object Reference=Get PlayerPlaytimeVM()) -> (ReturnValue:EActivityToShow Enum=None)
   ) -> (Output_Get=None)
   > n5 = Switch on EActivityToShow(Selection=@n4)
   SetVisibility(execute=@n5.UserSelection, self=Get VB_AskForMoreTime(), InVisibility=SelfHitTestInvisible)
   > SetVisibility(self=[Get VB_EmailSent(), Get VB_SendingEmail()], InVisibility=Collapsed)
   > RequestRefreshFocus()
   RequestAdditionalPlaytime(
    execute=[@n5.ActivityCreate, @n5.ActivityPlay, @n5.CombinedTime, @n3.ActivityCreate, @n3.ActivityPlay, @n3.CombinedTime]
    self=Get PlayerPlaytimeVM()
    ActivityType=Select(
     None:EActivityLimitType Enum=None
     CombinedTime:EActivityLimitType Enum=CombinedTime
     ActivityPlay:EActivityLimitType Enum=ActivityPlay
     ActivityCreate:EActivityLimitType Enum=ActivityCreate
     UserSelection:EActivityLimitType Enum=None
     Index=Get ActivityToShow() -> (ActivityToShow=None)
    ) -> (ReturnValue:EActivityLimitType Enum)
   )
  ]}
]
variables [
 Activity Type {type=EActivityLimitType, default=MAX}
 ActivityToShow {type=EActivityToShow, default=None}
]
widgets {
 info {
  inheritedWidgetCount=0
  namedSlotCount=1
  parentClass=/Script/CilantroUI.FortPlaytimePinActivatableWidget
  rootWidgetClass=/FortUILibrary/Systems/Dialog/WBP_UIKit_Dialog_Internal.WBP_UIKit_Dialog_Internal_C
  widgetCount=28
 }
 widgets [
  WBP_UIKit_Dialog_Base {
   widgetClassPath=/FortUILibrary/Systems/Dialog/WBP_UIKit_Dialog_Internal
   bIsVariable=true
   widget=WBP_UIKit_Dialog_Base
   properties {
    bIsFocusable=true
    designTimePreview_DialogVM=null
    useText=None
    useTitle=false
    buttons {
     each Button Horizontal Fill=false
     each Button Min Width=0
     each Button Size=Medium
     focuse Button=0
     space Between Buttons=30
     cTA Buttons Types []
     group Placement {
      horizontal Alignment=HAlign_Fill
      vertical Alignment=VAlign_Bottom
      grid {column=1, column Span=0, layer=0, row=2, row Span=0, nudge=<0 0>}
      padding=<ltrb 0 0 0 0>
     }
    }
    dialogSIze {
     height Override {enable=false, value=800}
     max Aspect Ratio {enable=false, value=0}
     max Desired Height {enable=true, value=800}
     max Desired Width {enable=false, value=0}
     min Aspect Ratio {enable=false, value=0}
     min Desired Height {enable=false, value=0}
     min Desired Width {enable=false, value=0}
     width Override {enable=true, value=864}
    }
   }
  }
  Overlay_0 {widgetClassPath=&Overlay, namedSlotHost=WBP_UIKit_Dialog_Base, widget=Overlay_0}
  GridPanel_48 {
   widgetClassPath=&GridPanel
   parent=Overlay_0
   widget=GridPanel_48
   properties {columnFill [1], rowFill [1]}
   slot {class=OverlaySlot, properties {horizontalAlignment=HAlign_Fill, verticalAlignment=VAlign_Fill, padding=<ltrb 80 64 80 80>}}
  }
  VB_AskForMoreTime {widgetClassPath=&VerticalBox, bIsVariable=true, parent=GridPanel_48, widget=VB_AskForMoreTime, slot {class=GridSlot}}
  HorizontalBox_1 {widgetClassPath=&HorizontalBox, parent=VB_AskForMoreTime, widget=HorizontalBox_1, slot {class=VerticalBoxSlot}}
  Image_AskForMoreTime {
   widgetClassPath=&Image
   bIsVariable=true
   parent=HorizontalBox_1
   widget=Image_AskForMoreTime
   properties {brush {
     bIsDynamicallyLoaded=false
     drawAs=Image
     imageType=FullColor
     mirroring=NoMirror
     resourceName=None
     resourceObject=/FortUILibrary/Textures/MaterialInstances/MI_UIKit_Icon_Email
     tiling=NoTile
     imageSize=<48 48>
     margin=<ltrb 0 0 0 0>
     outlineSettings {
      bUseBrushTransparency=false
      roundingType=HalfHeightRadius
      width=0
      color {colorUseRule=UseColor_Specified, specifiedColor=<rgba 0 0 0 0>}
      cornerRadii=<0 0 0 0>
     }
     tintColor {colorUseRule=UseColor_Specified, specifiedColor=<rgba 1 1 1 1>}
     uVRegion {bIsValid=false, max=<0 0>, min=<0 0>}
    }}
   slot {class=HorizontalBoxSlot, properties {horizontalAlignment=HAlign_Left, verticalAlignment=VAlign_Center, padding=<ltrb 0 0 13 0>}}
  }
  Text_Header {
   widgetClassPath=&CommonTextBlock
   parent=HorizontalBox_1
   widget=Text_Header
   properties {
    autoWrapText=true
    bIsScrollingEnabled=false
    text=Ask for More Time
    font {
     bForceMonospaced=false
     bMaterialIsStencil=false
     fontMaterial=null
     fontObject=&NotoSans
     letterSpacing=0
     monospacedWidth=1
     size=24
     skewAmount=0
     typefaceFontName=NotoSans-Regular
     outlineSettings {
      bApplyOutlineToDropShadows=false
      bMiteredCorners=false
      bSeparateFillAlpha=false
      outlineMaterial=null
      outlineSize=0
      outlineColor=<rgba 0 0 0 1>
     }
    }
    shadowOffset=<0 0>
   }
   slot {class=HorizontalBoxSlot, properties {size {sizeRule=Fill, value=1}}}
  }
  Text_Description {
   widgetClassPath=&CommonTextBlock
   parent=VB_AskForMoreTime
   widget=Text_Description
   properties {
    autoWrapText=true
    bIsScrollingEnabled=false
    text=What type of time would you like to ask for?
    font {
     bForceMonospaced=false
     bMaterialIsStencil=false
     fontMaterial=null
     fontObject=&NotoSans
     letterSpacing=0
     monospacedWidth=1
     size=24
     skewAmount=0
     typefaceFontName=NotoSans-Regular
     outlineSettings {
      bApplyOutlineToDropShadows=false
      bMiteredCorners=false
      bSeparateFillAlpha=false
      outlineMaterial=null
      outlineSize=0
      outlineColor=<rgba 0 0 0 1>
     }
    }
    shadowOffset=<0 0>
   }
   slot {class=VerticalBoxSlot, properties {padding=<ltrb 8 40 8 0>}}
  }
  Button_PlayTime {
   widgetClassPath=&/WBP_AddTime_Button
   bIsVariable=true
   parent=VB_AskForMoreTime
   widget=Button_PlayTime
   properties {
    bHideInputAction=true
    bHideInputActionWithKeyboard=true
    navigation=&self:WidgetTree.Button_PlayTime.WidgetNavigation_0
    text=Play time
   }
   slot {class=VerticalBoxSlot, properties {padding=<ltrb 8 24 8 0>}}
  }
  Button_CreateTime {
   widgetClassPath=&WBP_AddTime_Button
   bIsVariable=true
   parent=VB_AskForMoreTime
   widget=Button_CreateTime
   properties {
    bHideInputAction=true
    bHideInputActionWithKeyboard=true
    navigation=&self:WidgetTree.Button_CreateTime.WidgetNavigation_0
    text=Create time
   }
   slot {class=VerticalBoxSlot, properties {padding=<ltrb 8 0 8 0>}}
  }
  Text_Description_1 {
   widgetClassPath=&CommonTextBlock
   parent=VB_AskForMoreTime
   widget=Text_Description_1
   properties {
    autoWrapText=true
    bIsScrollingEnabled=false
    text="We’ll email your parent asking for more time."
    colorAndOpacity {colorUseRule=UseColor_Specified, specifiedColor=<rgba 0.5776 0.5776 0.5776 1>}
    font {
     bForceMonospaced=false
     bMaterialIsStencil=false
     fontMaterial=null
     fontObject=&NotoSans
     letterSpacing=0
     monospacedWidth=1
     size=24
     skewAmount=0
     typefaceFontName=NotoSans-Regular
     outlineSettings {
      bApplyOutlineToDropShadows=false
      bMiteredCorners=false
      bSeparateFillAlpha=false
      outlineMaterial=null
      outlineSize=0
      outlineColor=<rgba 0 0 0 1>
     }
    }
    shadowOffset=<0 0>
   }
   slot {class=VerticalBoxSlot, properties {padding=<ltrb 8 24 8 0>}}
  }
  Button_Close_AskForMoreTime {
   widgetClassPath=&WBP_UIKit_Button_Regular
   bIsVariable=true
   parent=VB_AskForMoreTime
   widget=Button_Close_AskForMoreTime
   properties {
    bHideInputAction=true
    buttonType=Quiet
    minHeight=56
    navigation=&self:WidgetTree.Button_Close_AskForMoreTime.WidgetNavigation_0
    text=CANCEL
   }
   slot {class=VerticalBoxSlot, properties {horizontalAlignment=HAlign_Left, verticalAlignment=VAlign_Bottom, padding=<ltrb 0 64 0 0>}}
  }
  VB_SendingEmail {
   widgetClassPath=&VerticalBox
   bIsVariable=true
   parent=GridPanel_48
   widget=VB_SendingEmail
   properties {visibility=Collapsed}
   slot {class=GridSlot}
  }
  HorizontalBox {widgetClassPath=&HorizontalBox, parent=VB_SendingEmail, widget=HorizontalBox, slot {class=VerticalBoxSlot}}
  WBP_UIKit_Throbber {
   widgetClassPath=/FortUILibrary/UIKit/Throbber/WBP_UIKit_Throbber
   parent=HorizontalBox
   widget=WBP_UIKit_Throbber
   properties {displayBackground=false}
   slot {class=HorizontalBoxSlot}
  }
  Text_Header_1 {
   widgetClassPath=&CommonTextBlock
   parent=HorizontalBox
   widget=Text_Header_1
   properties {
    autoWrapText=true
    text=Sending Email
    font {
     bForceMonospaced=false
     bMaterialIsStencil=false
     fontMaterial=null
     fontObject=&NotoSans
     letterSpacing=0
     monospacedWidth=1
     size=24
     skewAmount=0
     typefaceFontName=NotoSans-Regular
     outlineSettings {
      bApplyOutlineToDropShadows=false
      bMiteredCorners=false
      bSeparateFillAlpha=false
      outlineMaterial=null
      outlineSize=0
      outlineColor=<rgba 0 0 0 1>
     }
    }
    shadowOffset=<0 0>
   }
   slot {class=HorizontalBoxSlot, properties {padding=<ltrb 13 0 0 0>, size {sizeRule=Fill, value=1}}}
  }
  VB_EmailSent {
   widgetClassPath=&VerticalBox
   bIsVariable=true
   parent=GridPanel_48
   widget=VB_EmailSent
   properties {visibility=Collapsed}
   slot {class=GridSlot}
  }
  HorizontalBox_2 {widgetClassPath=&HorizontalBox, parent=VB_EmailSent, widget=HorizontalBox_2, slot {class=VerticalBoxSlot}}
  Image_95 {
   widgetClassPath=&Image
   parent=HorizontalBox_2
   widget=Image_95
   properties {brush {
     bIsDynamicallyLoaded=false
     drawAs=Image
     imageType=FullColor
     mirroring=NoMirror
     resourceName=None
     resourceObject=/FortUILibrary/Textures/MaterialInstances/MI_UIKit_Icon_Check
     tiling=NoTile
     imageSize=<48 48>
     margin=<ltrb 0 0 0 0>
     outlineSettings {
      bUseBrushTransparency=false
      roundingType=HalfHeightRadius
      width=0
      color {colorUseRule=UseColor_Specified, specifiedColor=<rgba 0 0 0 0>}
      cornerRadii=<0 0 0 0>
     }
     tintColor {colorUseRule=UseColor_Specified, specifiedColor=<rgba 1 1 1 1>}
     uVRegion {bIsValid=false, max=<0 0>, min=<0 0>}
    }}
   slot {class=HorizontalBoxSlot, properties {horizontalAlignment=HAlign_Left, verticalAlignment=VAlign_Center}}
  }
  Text_Header_2 {
   widgetClassPath=&CommonTextBlock
   parent=HorizontalBox_2
   widget=Text_Header_2
   properties {
    autoWrapText=true
    text="Email Sent!"
    font {
     bForceMonospaced=false
     bMaterialIsStencil=false
     fontMaterial=null
     fontObject=&NotoSans
     letterSpacing=0
     monospacedWidth=1
     size=24
     skewAmount=0
     typefaceFontName=NotoSans-Regular
     outlineSettings {
      bApplyOutlineToDropShadows=false
      bMiteredCorners=false
      bSeparateFillAlpha=false
      outlineMaterial=null
      outlineSize=0
      outlineColor=<rgba 0 0 0 1>
     }
    }
    shadowOffset=<0 0>
   }
   slot {class=HorizontalBoxSlot, properties {padding=<ltrb 13 0 0 0>, size {sizeRule=Fill, value=1}}}
  }
  Text_Description_Email {
   widgetClassPath=&CommonTextBlock
   bIsVariable=true
   parent=VB_EmailSent
   widget=Text_Description_Email
   properties {
    autoWrapText=true
    bIsScrollingEnabled=false
    text="We've sent an email to your parent or guardian to ask for more [play|create|or nothing if it's for both] time.\r\nYou can still create your own islands in Creative."
    font {
     bForceMonospaced=false
     bMaterialIsStencil=false
     fontMaterial=null
     fontObject=&NotoSans
     letterSpacing=0
     monospacedWidth=1
     size=24
     skewAmount=0
     typefaceFontName=NotoSans-Regular
     outlineSettings {
      bApplyOutlineToDropShadows=false
      bMiteredCorners=false
      bSeparateFillAlpha=false
      outlineMaterial=null
      outlineSize=0
      outlineColor=<rgba 0 0 0 1>
     }
    }
    shadowOffset=<0 0>
   }
   slot {class=VerticalBoxSlot, properties {padding=<ltrb 0 40 0 0>}}
  }
  Button_Close_EmailSent {
   widgetClassPath=&WBP_UIKit_Button_Regular
   bIsVariable=true
   parent=VB_EmailSent
   widget=Button_Close_EmailSent
   properties {bHideInputAction=true, minHeight=56, navigation=&self:WidgetTree.Button_Close_EmailSent.WidgetNavigation_0, text=Done}
   slot {class=VerticalBoxSlot, properties {horizontalAlignment=HAlign_Left, verticalAlignment=VAlign_Bottom, padding=<ltrb 0 64 0 0>}}
  }
  VB_ErrorState {
   widgetClassPath=&VerticalBox
   bIsVariable=true
   parent=GridPanel_48
   widget=VB_ErrorState
   properties {visibility=Collapsed}
   slot {class=GridSlot}
  }
  HorizontalBox_3 {widgetClassPath=&HorizontalBox, parent=VB_ErrorState, widget=HorizontalBox_3, slot {class=VerticalBoxSlot}}
  Image {
   widgetClassPath=&Image
   parent=HorizontalBox_3
   widget=Image
   properties {brush {
     bIsDynamicallyLoaded=false
     drawAs=Image
     imageType=FullColor
     mirroring=NoMirror
     resourceName=None
     resourceObject=/FortUILibrary/Textures/MaterialInstances/MI_UIKit_Icon_UpdatePending
     tiling=NoTile
     imageSize=<48 48>
     margin=<ltrb 0 0 0 0>
     outlineSettings {
      bUseBrushTransparency=false
      roundingType=HalfHeightRadius
      width=0
      color {colorUseRule=UseColor_Specified, specifiedColor=<rgba 0 0 0 0>}
      cornerRadii=<0 0 0 0>
     }
     tintColor {colorUseRule=UseColor_Specified, specifiedColor=<rgba 1 1 1 1>}
     uVRegion {bIsValid=false, max=<0 0>, min=<0 0>}
    }}
   slot {class=HorizontalBoxSlot, properties {horizontalAlignment=HAlign_Left, verticalAlignment=VAlign_Center}}
  }
  Text_Header_3 {
   widgetClassPath=&CommonTextBlock
   parent=HorizontalBox_3
   widget=Text_Header_3
   properties {
    autoWrapText=true
    text=failed to send email
    font {
     bForceMonospaced=false
     bMaterialIsStencil=false
     fontMaterial=null
     fontObject=&NotoSans
     letterSpacing=0
     monospacedWidth=1
     size=24
     skewAmount=0
     typefaceFontName=NotoSans-Regular
     outlineSettings {
      bApplyOutlineToDropShadows=false
      bMiteredCorners=false
      bSeparateFillAlpha=false
      outlineMaterial=null
      outlineSize=0
      outlineColor=<rgba 0 0 0 1>
     }
    }
    shadowOffset=<0 0>
   }
   slot {class=HorizontalBoxSlot, properties {padding=<ltrb 13 0 0 0>, size {sizeRule=Fill, value=1}}}
  }
  Text_Description_Failed {
   widgetClassPath=&CommonTextBlock
   bIsVariable=true
   parent=VB_ErrorState
   widget=Text_Description_Failed
   properties {
    autoWrapText=true
    bIsScrollingEnabled=false
    text=Please try again.
    font {
     bForceMonospaced=false
     bMaterialIsStencil=false
     fontMaterial=null
     fontObject=&NotoSans
     letterSpacing=0
     monospacedWidth=1
     size=24
     skewAmount=0
     typefaceFontName=NotoSans-Regular
     outlineSettings {
      bApplyOutlineToDropShadows=false
      bMiteredCorners=false
      bSeparateFillAlpha=false
      outlineMaterial=null
      outlineSize=0
      outlineColor=<rgba 0 0 0 1>
     }
    }
    shadowOffset=<0 0>
   }
   slot {class=VerticalBoxSlot, properties {padding=<ltrb 0 40 0 0>}}
  }
  Button_Close_Failed {
   widgetClassPath=&WBP_UIKit_Button_Regular
   bIsVariable=true
   parent=VB_ErrorState
   widget=Button_Close_Failed
   properties {
    bHideInputAction=true
    buttonType=Quiet
    minHeight=56
    navigation=&self:WidgetTree.Button_Close_Failed.WidgetNavigation_0
    text=Close
   }
   slot {class=VerticalBoxSlot, properties {horizontalAlignment=HAlign_Left, verticalAlignment=VAlign_Bottom, padding=<ltrb 0 64 0 0>}}
  }
 ]
}
