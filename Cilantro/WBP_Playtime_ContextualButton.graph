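# WBP_Playtime_ContextualButton is a Playtime contextual button derived from the UIKit generic button, with no widgets or variables of its own.
# OnInitialized and the input-method-changed event both call ToggleTouchStyle, whose graph holds only an entry node, so it currently does nothing.
# PreConstruct, Construct, hover and unhover only forward to the parent.
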WidgetBlueprint /Cilantro/WBP_Playtime_ContextualButton
graphs [
 EventGraph {nodes [
   n1 = Event PreConstruct()
   > Parent: PreConstruct(IsDesignTime=@n1.IsDesignTime)
   Event Construct()
   > Parent: Construct()
   n2 = Event OnInitialized()
   ToggleTouchStyle(execute=[@n2.then, @n3.then])
   Event On Hovered()
   > Parent: On Hovered() {is=title K2Node_CallParentFunction}
   Event On Unhovered()
   > Parent: On Unhovered() {is=title K2Node_CallParentFunction}
   n3 = Event On Input Method Changed!() -> (OutputDelegate, then, CurrentInputType:ECommonInputType Enum) {is=title K2Node_Event}
  ]}
 ToggleTouchStyle {nodes [ToggleTouchStyle() {is=title K2Node_FunctionEntry}]}
]
variables []
widgets {
 info {
  inheritedWidgetCount=0
  parentClass=/FortUILibrary/UIKit/Buttons/WBP_UIKit_Button_Generic.WBP_UIKit_Button_Generic_C
  rootWidgetClass=null
  widgetCount=0
 }
 widgets []
}
