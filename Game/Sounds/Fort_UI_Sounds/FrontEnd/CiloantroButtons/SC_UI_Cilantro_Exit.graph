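# Sound cue for the Cilantro front-end UI exit sound, named for an exit action among the button sounds.
# A Mixer feeds a single wave player for Cilantro_UI_Exit through its second input, leaving the first input empty.
# The cue plays at reduced volume in the Master sound class, with base submix off and a send to Fortnite_Final.
# Its own concurrency override allows one instance, stopping the oldest, and virtualization is disabled.
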SoundCue /Game/Sounds/Fort_UI_Sounds/FrontEnd/CiloantroButtons/SC_UI_Cilantro_Exit
root=Mixer_0
nodes [
 {id=Mixer_0, children [null,WavePlayer_1], properties {inputVolume [2,1]}}
 {id=WavePlayer_1, properties {soundWaveAssetPtr=&/Cilantro_UI_Exit}}
]
properties {
 bEnableBaseSubmix=false
 bOverrideConcurrency=true
 maxDistance=2097152
 soundClassObject=/Game/Sounds/FortSoundClassesAndModes/Master
 virtualizationMode=Disabled
 volumeMultiplier=0.1
 concurrencyOverrides {maxCount=1, resolutionRule=StopOldest, retriggerTime=1}
 soundSubmixSends [{minSendLevel=0, soundSubmix=/Game/Sounds/SubMixes/Fortnite_Final}]
}
