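# Sound cue for the Cilantro Start UI sound, filed with the front-end button sounds.
# A single wave player plays Cilantro_UI_Start at reduced volume, with no randomization or layering.
# Its own concurrency override allows one instance and stops the oldest on retrigger, with virtualization disabled.
# It uses the Master sound class and skips the base submix, sending to Fortnite_Final instead.
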SoundCue /Game/Sounds/Fort_UI_Sounds/FrontEnd/CiloantroButtons/SC_UI_Cilantro_Start
root=WavePlayer_5
nodes [{id=WavePlayer_5, properties {soundWaveAssetPtr=&/Cilantro_UI_Start}}]
properties {
 bEnableBaseSubmix=false
 bOverrideConcurrency=true
 maxDistance=2097152
 soundClassObject=/Game/Sounds/FortSoundClassesAndModes/Master
 virtualizationMode=Disabled
 volumeMultiplier=0.3
 attenuationOverrides {falloffDistance=4000, spatializationAlgorithm=SPATIALIZATION_HRTF, stereoSpread=0, attenuationShapeExtents {x=0}}
 concurrencyOverrides {maxCount=1, resolutionRule=StopOldest}
 soundSubmixSends [{minSendLevel=0, soundSubmix=/Game/Sounds/SubMixes/Fortnite_Final}]
}
