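# SC_UI_Cilantro_Timer_Tick is a front-end UI sound cue for a Cilantro timer tick.
# A Mixer layers the main tick wave at full volume with a quieter clock tick wave, and the cue lowers the combined level.
# It overrides concurrency to a small voice cap that stops the oldest instance, with virtualization disabled.
# It uses the Master sound class, disables the base submix and sends to the Fortnite_Final submix.
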SoundCue /Game/Sounds/Fort_UI_Sounds/FrontEnd/CiloantroButtons/SC_UI_Cilantro_Timer_Tick
root=Mixer_0
nodes [
 {id=Mixer_0, children [WavePlayer_1,WavePlayer_0], properties {inputVolume [1,0.2]}}
 {id=WavePlayer_0, properties {soundWaveAssetPtr=&/Cilantro_UI_Timer_Tick_Clock}}
 {id=WavePlayer_1, properties {soundWaveAssetPtr=&/Cilantro_UI_Timer_Tick}}
]
properties {
 bEnableBaseSubmix=false
 bOverrideConcurrency=true
 maxDistance=2097152
 soundClassObject=/Game/Sounds/FortSoundClassesAndModes/Master
 virtualizationMode=Disabled
 volumeMultiplier=0.221
 concurrencyOverrides {maxCount=2, resolutionRule=StopOldest}
 soundSubmixSends [{minSendLevel=0, soundSubmix=/Game/Sounds/SubMixes/Fortnite_Final}]
}
