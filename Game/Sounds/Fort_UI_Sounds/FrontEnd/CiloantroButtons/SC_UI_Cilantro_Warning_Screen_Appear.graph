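# Sound cue for the front-end UI sound that plays when the Cilantro warning screen appears.
# A single wave player feeds the output directly at reduced volume, under the Master sound class.
# A concurrency override allows one instance at a time, stopping the oldest, and virtualization is disabled.
# The base submix is off, with a submix send to Fortnite_Final configured instead.
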SoundCue /Game/Sounds/Fort_UI_Sounds/FrontEnd/CiloantroButtons/SC_UI_Cilantro_Warning_Screen_Appear
root=WavePlayer_1
nodes [{id=WavePlayer_1, properties {soundWaveAssetPtr=&/Cilantro_UI_Warning_Screen_Appear}}]
properties {
 bEnableBaseSubmix=false
 bOverrideConcurrency=true
 maxDistance=2097152
 soundClassObject=/Game/Sounds/FortSoundClassesAndModes/Master
 virtualizationMode=Disabled
 volumeMultiplier=0.25
 concurrencyOverrides {maxCount=1, resolutionRule=StopOldest}
 soundSubmixSends [{minSendLevel=0, soundSubmix=/Game/Sounds/SubMixes/Fortnite_Final}]
}
