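# WBP_PlaytimeBlock is Fortnite's playtime screen widget, covering time-limit blocked, idle warning and sleep mode states.
# OnPlaytimeStateChanged switches state overlays, info text, focus, sounds and control bus mixes, with timer-driven material animation for the idle llama.
# PlayerPlaytimeVM drives context menus for asking for or adding time, while request status updates show email progress.
# Layout adapts to console, compact UI, input type and the Thyme option.
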WidgetBlueprint /Cilantro/WBP_PlaytimeBlock
animations [
 Anim_IdleLama {
  binding_count=1
  length_seconds=2.5
  object=&self:Anim_IdleLama
  sequence {
   bindings [Idle_Lama {
     id=#1
     possessed_class=Image
     tracks [{
       type=MovieSceneWidgetMaterialTrack
       sections [{
         type=MovieSceneParameterSection
         channels [
          AnimRatio {type=MovieSceneScriptingFloatChannel, keys [[0,0],[30,0.8462],[54,0.8462],[87,0],[102,0],[150,1]]}
          OverrideAnim {type=MovieSceneScriptingFloatChannel, const [1,1]}
         ]
        }]
      }]
    }]
   display_rate {numerator=60}
  }
 }
 Anim_OverrideLamaOutro {
  binding_count=1
  length_seconds=0.3
  object=&self:Anim_OverrideLamaOutro
  sequence {
   bindings [Idle_Lama {
     id=#2
     possessed_class=Image
     tracks [{
       type=MovieSceneWidgetMaterialTrack
       sections [{
         type=MovieSceneParameterSection
         channels [
          OverrideAnim {type=MovieSceneScriptingFloatChannel, keys [[0,1],[18,0]]}
          AnimRatio {type=MovieSceneScriptingFloatChannel, const [1,1]}
         ]
        }]
      }]
    }]
   display_rate {numerator=60}
  }
 }
]
graphs [
 ActivateControlBusMix {nodes [
   n1 = ActivateControlBusMix!() -> (then, Mix:Sound Control Bus Mix Object Reference) {is=title K2Node_FunctionEntry}
   > n2 = Branch(Condition=Is Control Bus Mix Active!(Mix:Sound Control Bus Mix Object Reference=@n1) -> (ReturnValue:bool))
   Activate Control Bus Mix!(execute=@n2.else, Mix:Sound Control Bus Mix Object Reference=@n1) -> (then)
  ]}
 BP_GetDesiredFocusTarget {nodes [
   Get Desired Focus Target()
   n1 = GetActiveWidget!(self:Common Visibility Switcher Object Reference=Get StateSwitcher() -> (StateSwitcher:Common Visibility Switcher Object Reference)) -> (ReturnValue:Widget Object Reference)
   > n2 = Branch(Condition="Equal (Object)"(A=@n1, B=Get BlockState() -> (BlockState:Overlay Object Reference)))
   > Return Node!(
    execute
    ReturnValue:Widget Object Reference=Get ContextMenu_NeedMoreTime() -> (ContextMenu_NeedMoreTime:WBP UIKit Context Menu Anchor Button Object Reference)
   )
   n3 = Branch(execute=@n2.else, Condition="Equal (Object)"(A=@n1, B=Get SleepState() -> (SleepState:Overlay Object Reference)))
   > Return Node(ReturnValue=Get Button_Resume() -> (Button_Resume:WBP UIKit Button Regular Object Reference))
   Return Node(execute=@n3.else, ReturnValue=Self-Reference())
  ]}
 ConfigureConsoleOrCompactUILayout {nodes [
   ConfigureConsoleOrCompactUILayout() {is=title K2Node_FunctionEntry}
   > Branch(Condition=OR Boolean(A=IsUsingCompactUI(), B=IsConsolePlatform!() -> (ReturnValue:bool)))
   > RemoveFromParent(self=[
     Get Button_Exit_Thyme() -> (Button_Exit_Thyme:WBP UIKit Button Regular Object Reference)
     Get Button_Exit() -> (Button_Exit:WBP UIKit Button Regular Object Reference)
     Get Button_Close() -> (Button_Close:WBP UIKit Button Regular Object Reference)
    ])
   > n1 = GetDynamicMaterial(self=Get Idle_Lama() -> (Idle_Lama:Image Object Reference))
   > SetScalarParameterValue(self=@n1, ParameterName=AnimRatio, Value=0)
  ]}
 DeactivateControlBusMixes {nodes [
   DeactivateControlBusMixes() {is=title K2Node_FunctionEntry}
   n1 = Get Control_Bus_Mix_Sleep_Enabled() -> (Control_Bus_Mix_Sleep_Enabled:Sound Control Bus Mix Object Reference)
   n2 = Get Control_Bus_Mix_Warning() -> (Control_Bus_Mix_Warning:Sound Control Bus Mix Object Reference)
   > n3 = Branch(Condition=OR Boolean(A=Is Control Bus Mix Active(Mix=@n1), B=Is Control Bus Mix Active(Mix=@n2)))
   > n4 = PlaySound2D(Sound=Get ExitSound() -> (ExitSound:Sound Base Object Reference))
   Set CanAudioPlay(execute=[@n4.then, @n3.else], CanAudioPlay:bool=false) -> (Output_Get:bool)
   > n5 = Sequence()
   n6 = Get Control_Bus_Mix_Warning()
   > Branch(Condition=Is Control Bus Mix Active(Mix=@n6))
   > Deactivate Control Bus Mix!(execute, Mix:Sound Control Bus Mix Object Reference=@n6) -> (then)
   n7 = Get Control_Bus_Mix_Sleep_Enabled()
   Branch(execute=@n5.then_1, Condition=Is Control Bus Mix Active(Mix=@n7))
   > Deactivate Control Bus Mix(Mix=@n7)
  ]}
 EventGraph {nodes [
   n1 = DelayFinishes() {is=title K2Node_CustomEvent}
   n2 = Set Timer by Event(Delegate=@n1, Time=3, bLooping=false)
   > Set LamaTimerHandle(LamaTimerHandle:Timer Handle Structure=@n2) -> (Output_Get:Timer Handle Structure)
   n3 = "On Clicked (Button_Exit)"!() -> (OutputDelegate, then, Button:Common Button Base Object Reference) {is=title K2Node_ComponentBoundEvent}
   > n4 = Fire Event Element Interacted(Element=@n3.Button)
   CloseGame(
    execute=[@n12.then, @n4.then, @n10.then]
    self:Player Playtime View Model Object Reference=Get PlayerPlaytimeVM() -> (PlayerPlaytimeVM:Player Playtime View Model Object Reference)
   )
   > DeactivateControlBusMixes() {is=title K2Node_CallFunction}
   n5 = "On Clicked (Button_Resume)"!() -> (OutputDelegate, then, Button:Common Button Base Object Reference) {is=title K2Node_ComponentBoundEvent}
   > n6 = Fire Event Element Interacted(Element=@n5.Button)
   TryClose!(execute=[@n6.then, @n8.then], self:Fort Playtime Block Object Reference, bForceClose:bool=false) -> (then)
   > DeactivateControlBusMixes()
   n7 = "On Pressed (Button_KeepPlaying)"!() -> (OutputDelegate, then, Button:Common Button Base Object Reference) {is=title K2Node_ComponentBoundEvent}
   > n8 = Fire Event Element Interacted(Element=@n7.Button)
   n9 = "On Clicked (Button_Close)"!() -> (OutputDelegate, then, Button:Common Button Base Object Reference) {is=title K2Node_ComponentBoundEvent}
   > n10 = Fire Event Element Interacted(Element=@n9.Button)
   n11 = "On Clicked (Button_Exit_Thyme)"!() -> (OutputDelegate, then, Button:Common Button Base Object Reference) {is=title K2Node_ComponentBoundEvent}
   > n12 = Fire Event Element Interacted(Element=@n11.Button)
   n13 = "On Clicked (Button_OutOfTime)"!() -> (OutputDelegate, then, Button:Common Button Base Object Reference) {is=title K2Node_ComponentBoundEvent}
   > Fire Event Element Interacted(Element=@n13.Button)
   > TryClose(bForceClose=false)
   n14 = UpdateKeepPlayingButton!() -> (OutputDelegate, then, bNewInputType:ECommonInputType Enum) {is=title K2Node_CustomEvent}
   > SetText!(
    execute
    self:Fort CTAButton Object Reference=Get Button_KeepPlaying() -> (Button_KeepPlaying:WBP UIKit Button Regular Object Reference)
    ButtonText:"Text (by ref)"=Select!(
     MouseAndKeyboard:"Text (by ref)"=MakeLiteralText(Value="LOCTEXT(\"Press any key to keep playing\")")
     Gamepad:"Text (by ref)"=MakeLiteralText(Value="LOCTEXT(\"Press any button to keep playing\")")
     Touch:"Text (by ref)"=MakeLiteralText(Value="LOCTEXT(\"Tap anywhere to keep playing\")")
     Count:"Text (by ref)"
     Index:ECommonInputType Enum=@n14.bNewInputType
    ) -> (ReturnValue:"Text (by ref)")
   ) -> (then)
   n15 = On Request Playtime Status Update!() -> (OutputDelegate, then, RequestStatus:EPlaytimeRequestState Enum=None) {is=title K2Node_CustomEvent}
   > n16 = Set Request Status(Request Status:EPlaytimeRequestState Enum=@n15.RequestStatus) -> (Output_Get:EPlaytimeRequestState Enum=None)
   > n17 = Switch on EPlaytimeRequestState!(execute, Selection:EPlaytimeRequestState Enum=@n16) -> (None:Exec, Waiting:Exec, Success:Exec, Failure:Exec) {is=title K2Node_SwitchEnum}
   On Request State Failed(execute=@n17.Failure)
   LogString!(
    execute=@n17.None
    Instance Log Level:BlueprintLogLevel Enum=NewEnumerator4
    LogString:String="Request state is none!"
    PrintToScreen:bool=true
    Instigator:Object Reference
   ) -> (then)
   On Request State Success(execute=@n17.Success)
   On Request State Wait(execute=@n17.Waiting)
   n18 = OnLamaTimer() {is=title K2Node_CustomEvent}
   > PlaySound2D(Sound=Get LlamaSound() -> (LlamaSound:Sound Base Object Reference))
   > PlayAnimation(InAnimation=Get Anim_IdleLama() -> (Anim_IdleLama:Widget Animation Object Reference))
   n19 = StartLamaIntro() {is=title K2Node_CustomEvent}
   > n20 = Set Timer by Event(Delegate=@n18, Time=8, bLooping=false)
   > Set LamaTimerHandle(LamaTimerHandle=@n20)
   Event OnReturnFocusFromTakeover() {is=title K2Node_Event}
   > SetFocus(self=Get ContextMenu_More() -> (ContextMenu_More:WBP UIKit Context Menu Anchor Button Object Reference))
   Event OnInitialized()
   > BindToAnimationFinished(Animation=Get Anim_IdleLama(), Delegate=IdleLamaAnimFinished() {is=title K2Node_CustomEvent})
   > SetCompactUIStyle() {is=title K2Node_CallFunction}
   > Set bIsEnableThyme(bIsEnableThyme:bool=Get bEnableThyme!(self:Fort Runtime Options Object Reference=GetRuntimeOptions()) -> (bEnableThyme:bool)) -> (Output_Get:bool)
   Event On Deactivated()
   > Clear and Invalidate Timer by Handle(Handle=Get LamaTimerHandle() -> (LamaTimerHandle:Timer Handle Structure))
   > DeactivateControlBusMixes()
   Event On Activated()
   > Fire Event Sub View Opened!(execute, View:Widget Object Reference=Self-Reference(), AdditionalInfo:Flow Analytics Additional Info Structure, bIsOverlay:bool=false) -> (then)
   Event Destruct()
   > Set Control_Bus_Mix_Warning(Control_Bus_Mix_Warning:Sound Control Bus Mix Object Reference) -> (Output_Get:Sound Control Bus Mix Object Reference)
   > Set Control_Bus_Mix_Sleep_Enabled(Control_Bus_Mix_Sleep_Enabled:Sound Control Bus Mix Object Reference) -> (Output_Get:Sound Control Bus Mix Object Reference)
   n21 = Get CommonInputSubsystem()
   > Unbind Event from On Input Method Changed(self=@n21, Delegate=@n14)
   n22 = Create Event()
   > Unbind Event from On Menu Open Changed(
    self:Menu Anchor Object Reference=Get Anchor!(self:UIKit Context Menu Anchor Object Reference=Get ContextMenu_NeedMoreTime()) -> (Anchor:Menu Anchor Object Reference)
    Delegate=@n22
   ) {is=title K2Node_RemoveDelegate}
   n23 = Create Event()
   > Unbind Event from On Menu Open Changed(self=Get Anchor(self=Get ContextMenu_More()), Delegate=@n23)
   > Parent: Destruct()
   Event Construct()
   > Init Context Menu()
   > SetStatusIcon!(execute, self, IconTexture:Object Reference=Get ClockIconTexture() -> (ClockIconTexture:Object Reference)) -> (then)
   > Assign On Input Method Changed!(execute:Exec, self:Common Input Object Reference=@n21, Delegate:"Delegate (by ref)"=@n14) -> (then:Exec) {is=title K2Node_AssignDelegate}
   > UpdateKeepPlayingButton!(execute, self, bNewInputType:ECommonInputType Enum=GetCurrentInputType(self=@n21)) -> (then) {is=title K2Node_CallFunction}
   > ConfigureConsoleOrCompactUILayout() {is=title K2Node_CallFunction}
   > n24 = GetDynamicMaterial(self=Get Idle_Lama())
   > SetScalarParameterValue(self=@n24, ParameterName=AnimRatio, Value=0)
   > Bind Event to On Menu Open Changed(self:Menu Anchor Object Reference=Get Anchor(self=Get ContextMenu_NeedMoreTime()), Delegate=@n22) {is=title K2Node_AddDelegate}
   > Bind Event to On Menu Open Changed(self=Get Anchor(self=Get ContextMenu_More()), Delegate=@n23)
   > UpdateVisual()
   BindToAnimationFinished(Animation=Get Anim_OverrideLamaOutro() -> (Anim_OverrideLamaOutro:Widget Animation Object Reference), Delegate=@n19)
   PlayAnimation(execute=@n1.then, InAnimation=Get Anim_OverrideLamaOutro())
  ]}
 GetAddMoreTimeName {nodes [
   GetAddMoreTimeName() {is=title K2Node_FunctionEntry}
   n1 = Get PlayerPlaytimeVM()
   > Return Node!(
    execute
    AddMoreTimeName:Text=Select!(
     None:Text="LOCTABLE(\"/Cilantro/RichTextStyles/StringTable_TimeLimitTexts\", \"AddMoreTime\")"
     CombinedTime:Text="LOCTABLE(\"&StringTable_TimeLimitTexts\", \"AddMoreTime\")"
     ActivityPlay:Text="LOCTABLE(\"&StringTable_TimeLimitTexts\", \"AddPlayTime\")"
     ActivityCreate:Text="LOCTABLE(\"&StringTable_TimeLimitTexts\", \"AddMoreCreateTime\")"
     UserSelection:Text="LOCTABLE(\"&StringTable_TimeLimitTexts\", \"AddMoreTime\")"
     Index:EActivityToShow Enum=RefreshAddMoreTimeActivities!(self:Player Playtime View Model Object Reference=@n1) -> (ReturnValue:EActivityToShow Enum=None)
    ) -> (ReturnValue:Text)
   )
  ]}
 Init Context Menu {nodes [
   Init Context Menu() {is=title K2Node_FunctionEntry}
   > n1 = Sequence()
   > n2 = Is Valid(InputObject=Get NeedMoreTime_Context() -> (NeedMoreTime_Context:UIKit Context Menu Action Context Object Reference))
   n3 = Construct Object from Class(execute=@n2.Is Not Valid, Class=/Script/FortniteUI.UIKitContextMenuActionContext) -> (ReturnValue:UIKit Context Menu Action Context Object Reference)
   > n4 = Set NeedMoreTime_Context(NeedMoreTime_Context:UIKit Context Menu Action Context Object Reference=@n3) -> (Output_Get:UIKit Context Menu Action Context Object Reference)
   > SetContext!(
    execute
    self:WBP UIKit Context Menu Anchor Button Base Object Reference=Get ContextMenu_NeedMoreTime()
    InteractionContext_Data:UIKit Context Menu Action Context Object Reference=@n4
   ) -> (then)
   > RefreshNeedMoreTime!(
    execute
    self
    ExcludeAskTime:bool=false
    AskMoreTimeName:Text="LOCTABLE(\"&StringTable_TimeLimitTexts\", \"AskMoreTime\")"
    AddMoreTimeName:Text="LOCTABLE(\"&StringTable_TimeLimitTexts\", \"AddMoreTime\")"
   ) -> (then)
   n5 = Is Valid(exec=@n1.then_1, InputObject=Get MoreOptions_Context() -> (MoreOptions_Context:UIKit Context Menu Action Context Object Reference))
   n6 = Construct Object from Class(execute=@n5.Is Not Valid, Class=&UIKitContextMenuActionContext)
   > n7 = Set MoreOptions_Context(MoreOptions_Context:UIKit Context Menu Action Context Object Reference=@n6) -> (Output_Get:UIKit Context Menu Action Context Object Reference)
   > AddInteraction!(
    execute
    self:UIKit Context Menu Action Context Object Reference=@n7
    InActionInfo_DisplayName:Text="LOCTEXT(\"Voice Reporting\")"
    InActionInfo_IconTexture:Texture 2D Soft Object Reference
    InActionInfo_IconMaterial:Material Instance Soft Object Reference
    InActionInfo_SortPriority:int=1000
    InActionInfo_bIsAvailable:bool=True
    InActionInfo_bIsEnabled:bool=True
    InActionInfo_bIsSelected:bool=False
    InActionInfo_bAddDivider:bool=False
    InActionInfo_bPositionDividerAfter:bool=False
    InActionInfo_bCloseMenuOnSelect:bool=False
    InActionInfo_InteractionNameID:Name=None
    InActionInfo_Interaction:Data Table Row Handle Structure=(DataTable=None,RowName="")
    InActionInfo_PressedDelegate:Delegate="(null).None"
    InActionInfo_ReleasedDelegate:Delegate="(null).None"
    InInteractionDelegate:"Delegate (by ref)"=Create Event()
   ) -> (then)
   > AddInteraction(
    self=[@n7, @n7]
    InActionInfo_DisplayName="LOCTEXT(\"Reporting\")"
    InActionInfo_SortPriority=1000
    InActionInfo_bIsAvailable=True
    InActionInfo_bIsEnabled=True
    InActionInfo_bIsSelected=False
    InActionInfo_bAddDivider=False
    InActionInfo_bPositionDividerAfter=False
    InActionInfo_bCloseMenuOnSelect=False
    InActionInfo_InteractionNameID=None
    InActionInfo_Interaction=(DataTable=None,RowName="")
    InActionInfo_PressedDelegate="(null).None"
    InActionInfo_ReleasedDelegate="(null).None"
    InInteractionDelegate=Create Event()
   )
   > AddInteraction(
    self=@n7
    InActionInfo_DisplayName="LOCTEXT(\"Parental Controls\")"
    InActionInfo_SortPriority=1000
    InActionInfo_bIsAvailable=True
    InActionInfo_bIsEnabled=True
    InActionInfo_bIsSelected=False
    InActionInfo_bAddDivider=False
    InActionInfo_bPositionDividerAfter=False
    InActionInfo_bCloseMenuOnSelect=False
    InActionInfo_InteractionNameID=None
    InActionInfo_Interaction=(DataTable=None,RowName="")
    InActionInfo_PressedDelegate="(null).None"
    InActionInfo_ReleasedDelegate="(null).None"
    InInteractionDelegate=Create Event()
   )
   > AddInteraction(
    self=@n7
    InActionInfo_DisplayName="LOCTEXT(\"Support\")"
    InActionInfo_SortPriority=1000
    InActionInfo_bIsAvailable=True
    InActionInfo_bIsEnabled=True
    InActionInfo_bIsSelected=False
    InActionInfo_bAddDivider=False
    InActionInfo_bPositionDividerAfter=False
    InActionInfo_bCloseMenuOnSelect=False
    InActionInfo_InteractionNameID=None
    InActionInfo_Interaction=(DataTable=None,RowName="")
    InActionInfo_PressedDelegate="(null).None"
    InActionInfo_ReleasedDelegate="(null).None"
    InInteractionDelegate=Create Event()
   )
   > AddInteraction(
    self=@n7
    InActionInfo_DisplayName="LOCTEXT(\"Refunds\")"
    InActionInfo_SortPriority=1000
    InActionInfo_bIsAvailable=True
    InActionInfo_bIsEnabled=True
    InActionInfo_bIsSelected=False
    InActionInfo_bAddDivider=False
    InActionInfo_bPositionDividerAfter=False
    InActionInfo_bCloseMenuOnSelect=False
    InActionInfo_InteractionNameID=None
    InActionInfo_Interaction=(DataTable=None,RowName="")
    InActionInfo_PressedDelegate="(null).None"
    InActionInfo_ReleasedDelegate="(null).None"
    InInteractionDelegate=Create Event()
   )
   > AddInteraction(
    self=@n7
    InActionInfo_DisplayName="LOCTEXT(\"Legal\")"
    InActionInfo_SortPriority=1000
    InActionInfo_bIsAvailable=True
    InActionInfo_bIsEnabled=True
    InActionInfo_bIsSelected=False
    InActionInfo_bAddDivider=False
    InActionInfo_bPositionDividerAfter=False
    InActionInfo_bCloseMenuOnSelect=False
    InActionInfo_InteractionNameID=None
    InActionInfo_Interaction=(DataTable=None,RowName="")
    InActionInfo_PressedDelegate="(null).None"
    InActionInfo_ReleasedDelegate="(null).None"
    InInteractionDelegate=Create Event()
   )
   > SetContext(self=Get ContextMenu_More(), InteractionContext_Data=Get MoreOptions_Context())
  ]}
 Nav_ButtonExit_Left {nodes [
   Nav_ButtonExit_Left!() -> (then, Navigation:EUINavigation Enum) {is=title K2Node_FunctionEntry}
   > n1 = Branch(Condition=IsVisible(self=Get Button_Exit_Thyme()))
   > Return Node!(execute, ReturnValue:Widget Object Reference=Get Button_Exit_Thyme())
   Return Node(execute=@n1.else, ReturnValue=Get Button_OutOfTime() -> (Button_OutOfTime:WBP UIKit Button Loud Object Reference))
  ]}
 Nav_ButtonExit_Up {nodes [
   Nav_ButtonExit_Up!() -> (then, Navigation:EUINavigation Enum) {is=title K2Node_FunctionEntry}
   > Branch(Condition=IsVisible(self=Get Button_Exit()))
   > Return Node(ReturnValue=Get Button_Exit())
  ]}
 On Request State Failed {nodes [
   On Request State Failed() {is=title K2Node_FunctionEntry}
   > SetActiveWidget(
    self:Widget Switcher Object Reference=Get IconSwitcher_Status() -> (IconSwitcher_Status:Common Activatable Widget Switcher Object Reference)
    Widget=Get Image_StatusIcon() -> (Image_StatusIcon:Image Object Reference)
   )
   > SetStatusIcon(IconTexture=Get ClockIconTexture())
   > "SetText (Text)"(self=Get Text_Status() -> (Text_Status:Common Text Object Reference), InText="LOCTEXT(\"FAILED TO SEND AN EMAIL. PLEASE TRY AGAIN.\")")
  ]}
 On Request State Success {nodes [
   On Request State Success() {is=title K2Node_FunctionEntry}
   > SetActiveWidget(self=Get IconSwitcher_Status(), Widget=Get Image_StatusIcon())
   > SetStatusIcon(IconTexture=Get TickIconTexture() -> (TickIconTexture:Object Reference))
   > "SetText (Text)"(self=Get Text_Status(), InText="LOCTEXT(\"EMAIL SENT!\")")
  ]}
 On Request State Wait {nodes [
   On Request State Wait() {is=title K2Node_FunctionEntry}
   > SetActiveWidget(self=Get IconSwitcher_Status(), Widget=Get Throbber_EmailSent() -> (Throbber_EmailSent:WBP UIKit Throbber Object Reference))
   > "SetText (Text)"(self=Get Text_Status(), InText="LOCTEXT(\"SENDING EMAIL\")")
   > PlaySound2D(Sound=Get EmailSentSound() -> (EmailSentSound:Sound Base Object Reference))
  ]}
 On Sleep Time Updated {nodes [
   n1 = On Sleep Time Updated!() -> (then, Elapsed Time:double=~) {is=title K2Node_FunctionEntry}
   > n2 = Set L_SleepTimeFactor(L_SleepTimeFactor:double=SafeDivide(A=@n1, B=Get L_SleepTimeInterval() -> (L_SleepTimeInterval:double))) -> (Output_Get:double)
   > SetProgress!(
    execute
    self:WBP UIKit Progress Bar Base Object Reference=Get ProgressBar_SleepTime() -> (ProgressBar_SleepTime:WBP UIKit Progress Bar Object Reference)
    Progress:double=float - float(A=1, B=@n2)
   ) -> (then)
   > Branch(Condition=Get CanAudioPlay() -> (CanAudioPlay:bool))
   > PlaySound2D(Sound=Get TimerSound() -> (TimerSound:Sound Base Object Reference))
  ]}
 OnAddMoreTimePressed {nodes [
   OnAddMoreTimePressed!() -> (then, InteractionNameID:Name) {is=title K2Node_FunctionEntry}
   > OpenPinTimeGrant!(execute, self:Player Playtime View Model Object Reference=Get PlayerPlaytimeVM(), Context:User Widget Object Reference=Self-Reference()) -> (then)
   > n1 = Cast To FortPlayerController(Object=GetOwningPlayer())
   > "Fire Event Element Interacted (Non-Localized)"!(
    execute
    FortPC:Fort Player Controller Object Reference=@n1
    ElementName:String=ContextOption_AddMoreTime
    AdditionalInfo:Flow Analytics Additional Info Structure
   ) -> (then)
  ]}
 OnAskForMoreTimeUpdated {nodes [
   n1 = OnAskForMoreTimeUpdated!() -> (then, ActivityToShow:EActivityToShow Enum=None) {is=title K2Node_FunctionEntry}
   > n2 = GetAddMoreTimeName!(execute, self) -> (then, AddMoreTimeName:Text) {is=title K2Node_CallFunction}
   > Set AddTimeText(AddTimeText:Text=@n2) -> (Output_Get:Text)
   > Is Valid(InputObject=Get NeedMoreTime_Context())
   > n3 = Switch on EActivityToShow!(execute, Selection:EActivityToShow Enum=@n1) -> (None:Exec, CombinedTime:Exec, ActivityPlay:Exec, ActivityCreate:Exec, UserSelection:Exec) {is=title K2Node_SwitchEnum}
   Set AskTimeText(execute=@n3.ActivityCreate, AskTimeText:Text="LOCTABLE(\"&StringTable_TimeLimitTexts\", \"AskCreateTime\")") -> (Output_Get:Text)
   n4 = Get PlayerPlaytimeVM()
   > n5 = Set ExcludeAskTime(ExcludeAskTime:bool=IsAskForMoreCreateTimeBlocked!(self:Player Playtime View Model Object Reference=@n4) -> (ReturnValue:bool)) -> (Output_Get:bool)
   RefreshNeedMoreTime(
    execute=[@n9.then, @n7.then, @n5.then, @n11.then]
    ExcludeAskTime=Get ExcludeAskTime() -> (ExcludeAskTime:bool)
    AskMoreTimeName=Get AskTimeText() -> (AskTimeText:Text)
    AddMoreTimeName=Get AddTimeText() -> (AddTimeText:Text)
   )
   n6 = Set AskTimeText(execute=@n3.ActivityPlay, AskTimeText="LOCTABLE(\"&StringTable_TimeLimitTexts\", \"AskPlayTime\")")
   n7 = Set ExcludeAskTime(
    execute=[@n8.then, @n6.then]
    ExcludeAskTime=IsAskForMorePlayTimeBlocked!(self:Player Playtime View Model Object Reference=@n4) -> (ReturnValue:bool)
   )
   n8 = Set AskTimeText(execute=@n3.CombinedTime, AskTimeText="LOCTABLE(\"&StringTable_TimeLimitTexts\", \"AskMoreTime\")")
   n9 = Set ExcludeAskTime(execute=@n3.None, ExcludeAskTime=true)
   n10 = CheckBlockedStateForUserSelection!(self:Player Playtime View Model Object Reference=Get PlayerPlaytimeVM()) -> (ReturnValue:EActivityToShow Enum=None)
   Set AskTimeText(
    execute=@n3.UserSelection
    AskTimeText=Select(
     None="LOCTABLE(\"&StringTable_TimeLimitTexts\", \"AskMoreTime\")"
     CombinedTime="LOCTABLE(\"&StringTable_TimeLimitTexts\", \"AskMoreTime\")"
     ActivityPlay="LOCTABLE(\"&StringTable_TimeLimitTexts\", \"AskPlayTime\")"
     ActivityCreate="LOCTABLE(\"&StringTable_TimeLimitTexts\", \"AskCreateTime\")"
     UserSelection="LOCTABLE(\"&StringTable_TimeLimitTexts\", \"AskMoreTime\")"
     Index=@n10
    )
   )
   > n11 = Set ExcludeAskTime(ExcludeAskTime="Equal (Enum)"(A:EActivityToShow Enum=@n10, B:EActivityToShow Enum=None))
  ]}
 OnAskMoreTimePressed {nodes [
   OnAskMoreTimePressed!() -> (then, InteractionNameID:Name) {is=title K2Node_FunctionEntry}
   > OpenRequestMoreTimeDialog!(
    execute
    self:Player Playtime View Model Object Reference=Get PlayerPlaytimeVM()
    ContextWidget:User Widget Object Reference=Self-Reference()
   ) -> (then)
   > n1 = Cast To FortPlayerController(Object=GetOwningPlayer())
   > "Fire Event Element Interacted (Non-Localized)"(FortPC=@n1, ElementName=ContextOption_AskMoreTime)
  ]}
 OnLegalClicked {nodes [
   OnLegalClicked!() -> (then, InteractionNameID:Name) {is=title K2Node_FunctionEntry}
   > DisplayLegalModal(self:Fort Playtime Block Object Reference)
   > n1 = Cast To FortPlayerController(Object=GetOwningPlayer())
   > "Fire Event Element Interacted (Non-Localized)"(FortPC=@n1, ElementName=ContextOption_Legal)
  ]}
 OnMoreOpen {nodes [
   n1 = OnMoreOpen!() -> (then, bIsOpen:bool=~) {is=title K2Node_FunctionEntry}
   > Branch(Condition=@n1)
   > Fire Event Element Interacted(Element=Get ContextMenu_More())
  ]}
 OnNeedMoreTimeOpen {nodes [
   n1 = OnNeedMoreTimeOpen!() -> (then, bIsOpen:bool=~) {is=title K2Node_FunctionEntry}
   > Branch(Condition=@n1)
   > Fire Event Element Interacted(Element=Get ContextMenu_NeedMoreTime())
  ]}
 OnParentalControlsClicked {nodes [
   OnParentalControlsClicked!() -> (then, InteractionNameID:Name) {is=title K2Node_FunctionEntry}
   > OpenParentalControls(self:Fort Playtime Block Object Reference)
   > n1 = Cast To FortPlayerController(Object=GetOwningPlayer())
   > "Fire Event Element Interacted (Non-Localized)"(FortPC=@n1, ElementName=ContextOption_ParentalControls)
  ]}
 OnPlaytimeStateChanged {nodes [
   OnPlaytimeStateChanged!() -> (then, BlockState:EPlaytimeState Enum=Unblocked) {is=title K2Node_FunctionEntry}
   > Set Visibility(
    Visibility=Select!(
     Unblocked:ESlateVisibility Enum=Collapsed
     Blocked:ESlateVisibility Enum=Collapsed
     Idle:ESlateVisibility Enum=SelfHitTestInvisible
     Sleep:ESlateVisibility Enum=SelfHitTestInvisible
     PlayTime_Blocked:ESlateVisibility Enum=Collapsed
     CreateTime_Blocked:ESlateVisibility Enum=Collapsed
     Uninitialized:ESlateVisibility Enum=Collapsed
     Index:EPlaytimeState Enum=Get BlockState() -> (BlockState:EPlaytimeState Enum=Unblocked)
    ) -> (ReturnValue:ESlateVisibility Enum)
    self=Get SafeZone_SettingsReminder() -> (SafeZone_SettingsReminder:Safe Zone Object Reference)
   )
   > n1 = Switch on EPlaytimeState!(execute, Selection:EPlaytimeState Enum=Get BlockState() -> (BlockState=Unblocked)) -> (Unblocked:Exec, Blocked:Exec, Idle:Exec, Sleep:Exec, PlayTime_Blocked:Exec, CreateTime_Blocked:Exec, Uninitialized:Exec) {is=title K2Node_SwitchEnum}
   n2 = Get BlockState() -> (BlockState:Overlay Object Reference)
   n3 = Get StateSwitcher()
   SetActiveWidget(execute=@n1.Blocked, self:Common Visibility Switcher Object Reference=@n3, Widget=@n2)
   > Fire Event Sub View Opened(View=@n2, bIsOverlay=false)
   > SetProgress(self=Get ProgressBar_SleepTime(), Progress=0)
   > SetFocus(self=Get ContextMenu_NeedMoreTime())
   n4 = Get IdleState() -> (IdleState:Overlay Object Reference)
   SetActiveWidget(execute=@n1.Idle, self=@n3, Widget=@n4)
   > Fire Event Sub View Opened(View=@n4, bIsOverlay=false)
   > StartLamaIntro() {is=title K2Node_CallFunction}
   n5 = Get PlayerPlaytimeVM()
   n6 = IsReportingPlaytime!(self:Player Playtime View Model Object Reference=@n5) -> (ReturnValue:bool)
   n7 = IsPlaytimeLimitsEnabled!(self:Player Playtime View Model Object Reference=@n5) -> (ReturnValue:bool)
   n8 = SelectText(
    A="LOCTEXT(\"<img id=\\\"Voice\\\"/> <Bold>Voice chat</>, <img id=\\\"Playtime\\\"/> <Bold>Time Limits</> and <img id=\\\"Reports\\\"/> <Bold>Time Reports</> will be paused when Fortnite enters sleep mode.\")"
    B="LOCTEXT(\"<img id=\\\"Voice\\\"/> <Bold>Voice chat</> and  <img id=\\\"Reports\\\"/> <Bold>Time Reports</> will be paused when Fortnite enters sleep mode.\")"
    bPickA=@n7
   )
   > SetText!(
    execute
    self:Rich Text Block Object Reference=Get Text_Idle_Infos() -> (Text_Idle_Infos:Common Rich Text Block Object Reference)
    InText:"Text (by ref)"=SelectText(
     A=SelectText(
      A=@n8
      B="LOCTEXT(\"<img id=\\\"Voice\\\"/> <Bold>Voice chat</> and <img id=\\\"Playtime\\\"/> <Bold>Time Limits</> will be paused when Fortnite enters sleep mode.\")"
      bPickA=@n6
     )
     B="LOCTEXT(\"<img id=\\\"Voice\\\"/> <Bold>Voice chat</> will be paused when Fortnite enters Sleep Mode\")"
     bPickA=OR Boolean(A=@n6, B=@n7)
    )
   ) -> (then)
   > PlaySound2D(Sound=Get AppearSound() -> (AppearSound:Sound Base Object Reference))
   > ActivateControlBusMix!(execute, self, Mix:Sound Control Bus Mix Object Reference=Get Control_Bus_Mix_Warning()) -> (then) {is=title K2Node_CallFunction}
   > Set CanAudioPlay(CanAudioPlay=true)
   n9 = Get SleepState()
   SetActiveWidget(execute=@n1.Sleep, self=@n3, Widget=@n9)
   > Fire Event Sub View Opened(View=@n9, bIsOverlay=false)
   > Clear and Invalidate Timer by Handle(Handle=Get LamaTimerHandle())
   > PlayAnimationForward(InAnimation=Get Anim_OverrideLamaOutro())
   > PlaySound2D(Sound=Get StartSleepSound() -> (StartSleepSound:Sound Base Object Reference))
   > SetFocus(self=Get Button_Resume())
   > ActivateControlBusMix(Mix=Get Control_Bus_Mix_Sleep_Enabled())
   n10 = Get PlayerPlaytimeVM()
   n11 = IsReportingPlaytime(self=@n10)
   n12 = IsPlaytimeLimitsEnabled(self=@n10)
   n13 = SelectText(
    A="LOCTEXT(\"<img id=\\\"Voice\\\"/> <Bold>Voice chat</>, <img id=\\\"Playtime\\\"/> <Bold>Time Limits</> and <img id=\\\"Reports\\\"/> <Bold>Time Reports</> are paused and will resume when you start playing again.\")"
    B="LOCTEXT(\"<img id=\\\"Voice\\\"/> <Bold>Voice chat</> and  <img id=\\\"Reports\\\"/> <Bold>Time Reports</> are paused and will resume when you start playing again.\")"
    bPickA=@n12
   )
   > SetText(
    self=Get Text_Sleep_Infos() -> (Text_Sleep_Infos:Common Rich Text Block Object Reference)
    InText=SelectText(
     A=SelectText(
      A=@n13
      B="LOCTEXT(\"<img id=\\\"Voice\\\"/> <Bold>Voice chat</> and <img id=\\\"Playtime\\\"/> <Bold>Time Limits</> are paused and will resume when you start playing again.\")"
      bPickA=@n11
     )
     B="LOCTEXT(\"<img id=\\\"Voice\\\"/> <Bold>Voice chat</> is paused and will resume when you start playing again.\")"
     bPickA=OR Boolean(A=@n11, B=@n12)
    )
   )
   TryClose(execute=@n1.Unblocked, bForceClose=false)
   > DeactivateControlBusMixes()
  ]}
 OnRefundsClicked {nodes [
   OnRefundsClicked!() -> (then, InteractionNameID:Name) {is=title K2Node_FunctionEntry}
   > OpenReturnPanel(self:Fort Playtime Block Object Reference)
   > n1 = Cast To FortPlayerController(Object=GetOwningPlayer())
   > "Fire Event Element Interacted (Non-Localized)"(FortPC=@n1, ElementName=ContextOption_Refunds)
  ]}
 OnReportingClicked {nodes [
   OnReportingClicked!() -> (then, InteractionNameID:Name) {is=title K2Node_FunctionEntry}
   > DisplayReporting(self:Fort Playtime Block Object Reference)
   > n1 = Cast To FortPlayerController(Object=GetOwningPlayer())
   > "Fire Event Element Interacted (Non-Localized)"(FortPC=@n1, ElementName=ContextOption_Reporting)
  ]}
 OnSupportClicked {nodes [
   OnSupportClicked!() -> (then, InteractionNameID:Name) {is=title K2Node_FunctionEntry}
   > LaunchSupportURL(self:Fort Playtime Block Object Reference)
   > n1 = Cast To FortPlayerController(Object=GetOwningPlayer())
   > "Fire Event Element Interacted (Non-Localized)"(FortPC=@n1, ElementName=ContextOption_Support)
  ]}
 OnVoiceReportingClicked {nodes [
   OnVoiceReportingClicked!() -> (then, InteractionNameID:Name) {is=title K2Node_FunctionEntry}
   > DisplayVoiceReporting(self:Fort Playtime Block Object Reference)
   > n1 = Cast To FortPlayerController(Object=GetOwningPlayer())
   > "Fire Event Element Interacted (Non-Localized)"(FortPC=@n1, ElementName=ContextOption_VoiceReporting)
  ]}
 RefreshNeedMoreTime {nodes [
   RefreshNeedMoreTime!() -> (
    then
    ExcludeAskTime:bool=~
    AskMoreTimeName:Text="LOCTABLE(\"&StringTable_TimeLimitTexts\", \"AskMoreTime\")"
    AddMoreTimeName:Text="LOCTABLE(\"&StringTable_TimeLimitTexts\", \"AddMoreTime\")"
   ) {is=title K2Node_FunctionEntry}
   > RemoveAllInteractions(self:UIKit Context Menu Action Context Object Reference=Get NeedMoreTime_Context())
   n1 = Get bIsEnableThyme() -> (bIsEnableThyme:bool)
   > AddInteraction(
    self=Get NeedMoreTime_Context()
    InActionInfo_DisplayName=Get AskMoreTimeName() -> (AskMoreTimeName:Text)
    InActionInfo_IconTexture=Select!(
     Option 0:Texture 2D Soft Object Reference
     Option 1:Texture 2D Soft Object Reference=/FortUILibrary/Textures/Icons/T_UI_Icon_UIKit_Email
     Index:bool=@n1
    ) -> (ReturnValue:Texture 2D Soft Object Reference)
    InActionInfo_IconMaterial=Select(
     Option 0:Material Instance Soft Object Reference
     Option 1:Material Instance Soft Object Reference=/Cilantro/MaterialInstances/MI_UI_Playtime_ContextMenuListEntry_Icon
     Index=@n1
    ) -> (ReturnValue:Material Instance Soft Object Reference)
    InActionInfo_SortPriority=999
    InActionInfo_bIsAvailable=True
    InActionInfo_bIsEnabled=NOT Boolean(A=Get ExcludeAskTime())
    InActionInfo_bIsSelected=False
    InActionInfo_bAddDivider=False
    InActionInfo_bPositionDividerAfter=False
    InActionInfo_bCloseMenuOnSelect=False
    InActionInfo_InteractionNameID=Get AskForMoreTimeInteractionId() -> (AskForMoreTimeInteractionId:Name=None)
    InActionInfo_Interaction=(DataTable=None,RowName="")
    InActionInfo_PressedDelegate="(null).None"
    InActionInfo_ReleasedDelegate="(null).None"
    InInteractionDelegate=Create Event()
   )
   n2 = Get bIsEnableThyme()
   > AddInteraction(
    self=Get NeedMoreTime_Context()
    InActionInfo_DisplayName=Get AddMoreTimeName() -> (AddMoreTimeName:Text)
    InActionInfo_IconTexture=Select(
     Option 0:Texture 2D Soft Object Reference
     Option 1:Texture 2D Soft Object Reference=/FortUILibrary/Textures/Icons/T_UI_Icon_UIKit_Locked
     Index=@n2
    ) -> (ReturnValue:Texture 2D Soft Object Reference)
    InActionInfo_IconMaterial=Select(
     Option 0:Material Instance Soft Object Reference
     Option 1:Material Instance Soft Object Reference=&MI_UI_Playtime_ContextMenuListEntry_Icon
     Index=@n2
    ) -> (ReturnValue:Material Instance Soft Object Reference)
    InActionInfo_SortPriority=1000
    InActionInfo_bIsAvailable=True
    InActionInfo_bIsEnabled=True
    InActionInfo_bIsSelected=False
    InActionInfo_bAddDivider=False
    InActionInfo_bPositionDividerAfter=False
    InActionInfo_bCloseMenuOnSelect=False
    InActionInfo_InteractionNameID=Get AddMoreTimeInteractionId() -> (AddMoreTimeInteractionId:Name=None)
    InActionInfo_Interaction=(DataTable=None,RowName="")
    InActionInfo_PressedDelegate="(null).None"
    InActionInfo_ReleasedDelegate="(null).None"
    InInteractionDelegate=Create Event()
   )
  ]}
 SetActivityState {nodes [
   SetActivityState!() -> (then, OutOfTimeActivityType:EActivityToShow Enum=None) {is=title K2Node_FunctionEntry}
   > SetVisibility(
    self=Get Button_OutOfTime()
    InVisibility=Select!(
     None:ESlateVisibility Enum=Collapsed
     CombinedTime:ESlateVisibility Enum=Collapsed
     ActivityPlay:ESlateVisibility Enum=SelfHitTestInvisible
     ActivityCreate:ESlateVisibility Enum=SelfHitTestInvisible
     UserSelection:ESlateVisibility Enum=Collapsed
     Index:EActivityToShow Enum=Get OutOfTimeActivityType() -> (OutOfTimeActivityType:EActivityToShow Enum=None)
    ) -> (ReturnValue:ESlateVisibility Enum)
   )
   > SetVisibility(
    self=Get Button_Exit_Thyme()
    InVisibility=Select(
     None=SelfHitTestInvisible
     CombinedTime=SelfHitTestInvisible
     ActivityPlay=Collapsed
     ActivityCreate=Collapsed
     UserSelection=SelfHitTestInvisible
     Index=Get OutOfTimeActivityType() -> (OutOfTimeActivityType=None)
    )
   )
   > SetText!(
    execute
    self:Fort CTAButton Object Reference=Get Button_OutOfTime()
    ButtonText:"Text (by ref)"=Select(
     None:"Text (by ref)"
     CombinedTime:"Text (by ref)"
     ActivityPlay:"Text (by ref)"=Format Text(Format="LOCTEXT(\"Go To Creative\")")
     ActivityCreate:"Text (by ref)"=Format Text(Format="LOCTEXT(\"Go To Discover\")")
     UserSelection:"Text (by ref)"
     Index=Get OutOfTimeActivityType() -> (OutOfTimeActivityType=None)
    ) -> (ReturnValue:"Text (by ref)")
   ) -> (then)
   > Branch(Condition="Equal (Enum)"(A:EPlaytimeRequestState Enum=Get Request Status() -> (Request Status:EPlaytimeRequestState Enum=None), B:EPlaytimeRequestState Enum=None))
   > "SetText (Text)"(
    self=Get Text_Status()
    InText=Select(
     None:Text="LOCTEXT(\"Your Time is up\")"
     CombinedTime:Text="LOCTEXT(\"Your Time is up\")"
     ActivityPlay:Text="LOCTEXT(\"Your Play Time is Up\")"
     ActivityCreate:Text="LOCTEXT(\"Your Create Time is Up\")"
     UserSelection:Text="LOCTEXT(\"Your Time is up\")"
     Index=Get OutOfTimeActivityType() -> (OutOfTimeActivityType=None)
    ) -> (ReturnValue:Text)
   )
   > "SetText (Text)"(
    self=Get Text_Message() -> (Text_Message:Common Text Object Reference)
    InText=GetNextPeriodText!(self:Player Playtime View Model Object Reference=Get PlayerPlaytimeVM()) -> (ReturnValue:Text)
   )
  ]}
 SetCompactUIStyle {nodes [
   SetCompactUIStyle() {is=title K2Node_FunctionEntry}
   > Branch(Condition=IsUsingCompactUI())
   > Set Min Height!(execute, Min Height:int=80, self:WBP UIKit Context Menu Anchor Button Base Object Reference=Get ContextMenu_NeedMoreTime()) -> (then, Output_Get:int=0)
   n1 = Get ContextMenu_More()
   > Set Min Height(Min Height=80, self=@n1) -> (Output_Get=0)
   > Set Min Width!(execute, Min Width:int=80, self:WBP UIKit Context Menu Anchor Button Base Object Reference=@n1) -> (then, Output_Get:int=0)
   > SetPadding!(
    execute
    self:User Widget Object Reference=[Get Button_OutOfTime(), Get Button_Exit_Thyme(), Get ContextMenu_NeedMoreTime()]
    InPadding_Left:float=0
    InPadding_Top:float=0
    InPadding_Right:float=20
    InPadding_Bottom:float=0
   ) -> (then)
  ]}
 SetStatusIcon {nodes [
   n1 = SetStatusIcon!() -> (then, IconTexture:Object Reference) {is=title K2Node_FunctionEntry}
   n2 = Get!(self:Image Object Reference=Get Image_StatusIcon()) -> (
    Brush_TintColor:Slate Color Structure
    Brush_DrawAs:ESlateBrushDrawType Enum
    Brush_Tiling:ESlateBrushTileType Enum
    Brush_Mirroring:ESlateBrushMirrorType Enum
    Brush_ImageSize:Vector 2D Structure
    Brush_Margin:Margin Structure
    Brush_ResourceObject:Object Reference
    Brush_OutlineSettings:Slate Brush Outline Settings Structure
   )
   > SetBrush!(
    execute
    self=Get Image_StatusIcon()
    InBrush_TintColor:Slate Color Structure=@n2
    InBrush_DrawAs:ESlateBrushDrawType Enum=@n2.Brush_DrawAs
    InBrush_Tiling:ESlateBrushTileType Enum=@n2.Brush_Tiling
    InBrush_Mirroring:ESlateBrushMirrorType Enum=@n2.Brush_Mirroring
    InBrush_ImageSize:Vector 2D Structure=@n2.Brush_ImageSize
    InBrush_Margin:Margin Structure=@n2.Brush_Margin
    InBrush_ResourceObject:Object Reference=@n1
    InBrush_OutlineSettings:Slate Brush Outline Settings Structure=@n2.Brush_OutlineSettings
   ) -> (then)
  ]}
 UpdateVisual {nodes [
   UpdateVisual() {is=title K2Node_FunctionEntry}
   > SetVisibility(
    self=Get Button_Exit()
    InVisibility=Select!(Option 0:ESlateVisibility Enum=SelfHitTestInvisible, Option 1:ESlateVisibility Enum=Collapsed, Index:bool=Get bIsEnableThyme()) -> (ReturnValue:ESlateVisibility Enum)
   )
   > SetVisibility(self=Get Button_Exit_Thyme(), InVisibility=Select(Option 0=Collapsed, Option 1=SelfHitTestInvisible, Index=Get bIsEnableThyme()))
  ]}
]
variables [
 AddMoreTimeInteractionId {type=string, default=AddMoreTime}
 AppearSound {type=SoundBase, default=/Game/Sounds/Fort_UI_Sounds/FrontEnd/CiloantroButtons/SC_UI_Cilantro_Warning_Screen_Appear}
 AskForMoreTimeInteractionId {type=string, default=AskForMoreTime}
 bIsEnableThyme {type=boolean}
 CanAudioPlay {type=boolean}
 ClockIconTexture {type=Object, default=/FortUILibrary/Textures/MaterialInstances/MI_UIKit_Icon_UpdatePending}
 Control_Bus_Mix_Sleep_Enabled {type=SoundControlBusMix, default=/Cilantro/Sounds/CBM_Cilantro_Enabled}
 Control_Bus_Mix_Warning {type=SoundControlBusMix, default=/Cilantro/Sounds/CBM_Cilantro_Warning}
 EmailSentSound {type=SoundBase, default=/Game/Sounds/Fort_UI_Sounds/FrontEnd/Fort_UI_Email_Sent_Cue}
 ExitSound {type=SoundBase, default=/Game/Sounds/Fort_UI_Sounds/FrontEnd/CiloantroButtons/SC_UI_Cilantro_Exit}
 In Action Info Icon Texture {type=Texture2D, default=&T_UI_Icon_UIKit_Email}
 LamaTimerHandle {type=TimerHandle}
 LlamaSound {type=SoundBase, default=/Game/Sounds/Fort_UI_Sounds/FrontEnd/CiloantroButtons/SC_UI_Cilantro_Llama}
 MoreOptions_Context {type=UIKitContextMenuActionContext, default=null}
 NeedMoreTime_Context {type=UIKitContextMenuActionContext, default=null}
 Request Status {type=EPlaytimeRequestState, default=None}
 StartSleepSound {type=SoundBase, default=/Game/Sounds/Fort_UI_Sounds/FrontEnd/CiloantroButtons/SC_UI_Cilantro_Start}
 TickIconTexture {type=Object, default=/FortUILibrary/Textures/MaterialInstances/MI_UIKit_Icon_Check}
 TimerSound {type=SoundBase, default=/Game/Sounds/Fort_UI_Sounds/FrontEnd/CiloantroButtons/SC_UI_Cilantro_Timer_Tick}
]
widgets {
 info {inheritedWidgetCount=1, parentClass=/Script/CilantroUI.FortPlaytimeBlock, rootWidgetClass=&Overlay, widgetCount=44}
 widgets [
  Overlay_0 {widgetClassPath=&Overlay, widget=Overlay_0}
  StateSwitcher {
   widgetClassPath=&CommonVisibilitySwitcher
   bIsVariable=true
   parent=Overlay_0
   widget=StateSwitcher
   properties {clipping=ClipToBounds}
   slot {class=OverlaySlot, properties {horizontalAlignment=HAlign_Fill, verticalAlignment=VAlign_Fill}}
  }
  BlockState {widgetClassPath=&Overlay, bIsVariable=true, parent=StateSwitcher, widget=BlockState, slot {class=CommonVisibilitySwitcherSlot}}
  Background_Block {
   widgetClassPath=&Image
   parent=BlockState
   widget=Background_Block
   properties {
    visibility=SelfHitTestInvisible
    brush {
     bIsDynamicallyLoaded=false
     drawAs=Image
     imageType=FullColor
     mirroring=NoMirror
     resourceName=None
     resourceObject=/Cilantro/MaterialInstances/MI_UI_PlaytimeControl_Background_TimeUp
     tiling=NoTile
     imageSize=<32 32>
     margin=<ltrb 0 0 0 0>
     outlineSettings {
      bUseBrushTransparency=false
      roundingType=HalfHeightRadius
      width=0
      color {colorUseRule=UseColor_Specified, specifiedColor=<rgba 0 0 0 0>}
      cornerRadii=<0 0 0 0>
     }
     tintColor {colorUseRule=UseColor_Specified, specifiedColor=<rgba 1 1 1 1>}
     uVRegion {bIsValid=false, max=<0 0>, min=<0 0>}
    }
   }
   slot {class=OverlaySlot, properties {horizontalAlignment=HAlign_Fill, verticalAlignment=VAlign_Fill}}
  }
  WBP_CaptureForPostBufferUpdate {
   widgetClassPath=&WBP_CaptureForPostBufferUpdate
   parent=BlockState
   slot {class=OverlaySlot, properties {horizontalAlignment=HAlign_Fill, verticalAlignment=VAlign_Fill}}
  }
  SZ_Block {
   widgetClassPath=&SafeZone
   bIsVariable=true
   parent=BlockState
   widget=SZ_Block
   slot {class=OverlaySlot, properties {horizontalAlignment=HAlign_Fill, verticalAlignment=VAlign_Fill}}
  }
  Overlay_2 {widgetClassPath=&Overlay, parent=SZ_Block, widget=Overlay_2, slot {class=SafeZoneSlot}}
  HorizontalBox_0 {
   widgetClassPath=&HorizontalBox
   parent=Overlay_2
   widget=HorizontalBox_0
   slot {class=OverlaySlot, properties {verticalAlignment=VAlign_Bottom, padding=<ltrb 136 0 0 48>}}
  }
  Button_OutOfTime {
   widgetClassPath=/FortUILibrary/UIKit/Buttons/WBP_UIKit_Button_Loud
   bIsVariable=true
   parent=HorizontalBox_0
   widget=Button_OutOfTime
   properties {
    bHideInputAction=true
    bHideInputActionWithKeyboard=true
    minWidth=224
    navigation=&self:WidgetTree.Button_OutOfTime.WidgetNavigation_0
    text=Go to discover
    visibility=Collapsed
   }
   slot {class=HorizontalBoxSlot, properties {padding=<ltrb 0 0 12 0>}}
  }
  Button_Exit_Thyme {
   widgetClassPath=&WBP_UIKit_Button_Regular
   bIsVariable=true
   parent=HorizontalBox_0
   widget=Button_Exit_Thyme
   properties {
    bHideInputAction=true
    minWidth=224
    navigation=&self:WidgetTree.Button_Exit_Thyme.WidgetNavigation_0
    text=Close Fortnite
   }
   slot {class=HorizontalBoxSlot, properties {horizontalAlignment=HAlign_Left, padding=<ltrb 0 0 12 0>}}
  }
  ContextMenu_NeedMoreTime {
   widgetClassPath=/FortUILibrary/UIKit/ContextMenu/WBP_UIKit_ContextMenuAnchorButton
   bIsVariable=true
   parent=HorizontalBox_0
   widget=ContextMenu_NeedMoreTime
   properties {
    bIsFocusable=true
    buttonIcon=&T_UI_Icon_UIKit_Email
    buttonText=Need more time?
    closeMenuOnContextClicked=true
    menuPlacement=MenuPlacement_AboveAnchor
    min Width=248
    mobileFontSizeMultiplier=1.17
    navigation=&self:WidgetTree.ContextMenu_NeedMoreTime.WidgetNavigation_0
    button Text Font {
     bForceMonospaced=false
     bMaterialIsStencil=false
     fontMaterial=/FortUILibrary/UIKit/Buttons/MaterialInstances/CTAButtons/MI_UIKit_TextRegular
     fontObject=&HeadingNow
     letterSpacing=48
     monospacedWidth=1
     size=18
     skewAmount=0
     typefaceFontName=76 Bold
     outlineSettings {
      bApplyOutlineToDropShadows=false
      bMiteredCorners=false
      bSeparateFillAlpha=false
      outlineMaterial=null
      outlineSize=0
      outlineColor=<rgba 0 0 0 1>
     }
    }
    mobileButtonPadding=<ltrb 24 0 24 0>
   }
   slot {class=HorizontalBoxSlot, properties {horizontalAlignment=HAlign_Left, padding=<ltrb 0 0 12 0>}}
  }
  ContextMenu_More {
   widgetClassPath=&WBP_UIKit_ContextMenuAnchorButton
   bIsVariable=true
   parent=HorizontalBox_0
   widget=ContextMenu_More
   properties {
    bIsFocusable=true
    button Background Material=/FortUILibrary/UIKit/Buttons/MaterialInstances/CTAButtons/MI_UIKit_BgQuiet
    buttonClass=&/WBP_Playtime_ContextualButton.WBP_Playtime_ContextualButton_C
    buttonIcon=/FortUILibrary/Textures/Icons/T_UI_Icon_UIKit_ContextMenu_SDF
    buttonText=""
    closeMenuOnContextClicked=true
    menuPlacement=MenuPlacement_AboveAnchor
    min Height=56
    min Width=56
    mobileMinHeight=90
    mobileMinWidth=90
    navigation=&self:WidgetTree.ContextMenu_More.WidgetNavigation_0
    button Padding=<ltrb 0 0 0 0>
   }
   slot {class=HorizontalBoxSlot, properties {horizontalAlignment=HAlign_Left, verticalAlignment=VAlign_Center}}
  }
  InfoContainer {
   widgetClassPath=&VerticalBox
   parent=Overlay_2
   widget=InfoContainer
   slot {class=OverlaySlot, properties {verticalAlignment=VAlign_Center, padding=<ltrb 136 0 0 0>}}
  }
  StatusContainer {widgetClassPath=&HorizontalBox, parent=InfoContainer, widget=StatusContainer, slot {class=VerticalBoxSlot}}
  IconSwitcher_Status {
   widgetClassPath=&CommonActivatableWidgetSwitcher
   bIsVariable=true
   parent=StatusContainer
   widget=IconSwitcher_Status
   slot {class=HorizontalBoxSlot}
  }
  Image_StatusIcon {
   widgetClassPath=&Image
   bIsVariable=true
   parent=IconSwitcher_Status
   widget=Image_StatusIcon
   properties {
    visibility=SelfHitTestInvisible
    brush {
     bIsDynamicallyLoaded=false
     drawAs=Image
     imageType=FullColor
     mirroring=NoMirror
     resourceName=None
     resourceObject=&MI_UIKit_Icon_UpdatePending
     tiling=NoTile
     imageSize=<48 48>
     margin=<ltrb 0 0 0 0>
     outlineSettings {
      bUseBrushTransparency=false
      roundingType=HalfHeightRadius
      width=0
      color {colorUseRule=UseColor_Specified, specifiedColor=<rgba 0 0 0 0>}
      cornerRadii=<0 0 0 0>
     }
     tintColor {colorUseRule=UseColor_Specified, specifiedColor=<rgba 1 1 1 1>}
     uVRegion {bIsValid=false, max=<0 0>, min=<0 0>}
    }
   }
   slot {class=WidgetSwitcherSlot, properties {horizontalAlignment=HAlign_Left, verticalAlignment=VAlign_Center}}
  }
  Throbber_EmailSent {
   widgetClassPath=/FortUILibrary/UIKit/Throbber/WBP_UIKit_Throbber
   bIsVariable=true
   parent=IconSwitcher_Status
   widget=Throbber_EmailSent
   slot {class=WidgetSwitcherSlot, properties {horizontalAlignment=HAlign_Left, verticalAlignment=VAlign_Center}}
  }
  Text_Status {
   widgetClassPath=&CommonTextBlock
   bIsVariable=true
   parent=StatusContainer
   widget=Text_Status
   properties {
    style=&TextStyle-Base_C
    text="YOUR PLAY TIME IS UP!"
    font {
     bForceMonospaced=false
     bMaterialIsStencil=false
     fontMaterial=null
     fontObject=&NotoSans
     letterSpacing=0
     monospacedWidth=1
     size=24
     skewAmount=0
     typefaceFontName=NotoSans-Regular
     outlineSettings {
      bApplyOutlineToDropShadows=false
      bMiteredCorners=false
      bSeparateFillAlpha=false
      outlineMaterial=null
      outlineSize=0
      outlineColor=<rgba 0 0 0 1>
     }
    }
    shadowOffset=<0 0>
   }
   slot {class=HorizontalBoxSlot, properties {verticalAlignment=VAlign_Center, padding=<ltrb 16 0 0 0>}}
  }
  Text_Message {
   widgetClassPath=&CommonTextBlock
   bIsVariable=true
   parent=InfoContainer
   widget=Text_Message
   properties {
    autoWrapText=true
    style=&TextStyle-Base_C
    text=Your time to play Fortnite is up. Please ask your parent when you can play again.
    wrapTextAt=792
    font {
     bForceMonospaced=false
     bMaterialIsStencil=false
     fontMaterial=null
     fontObject=&NotoSans
     letterSpacing=0
     monospacedWidth=1
     size=24
     skewAmount=0
     typefaceFontName=NotoSans-Regular
     outlineSettings {
      bApplyOutlineToDropShadows=false
      bMiteredCorners=false
      bSeparateFillAlpha=false
      outlineMaterial=null
      outlineSize=0
      outlineColor=<rgba 0 0 0 1>
     }
    }
    shadowOffset=<0 0>
   }
   slot {class=VerticalBoxSlot, properties {padding=<ltrb 0 16 0 0>, size {sizeRule=Automatic, value=0.8}}}
  }
  Button_Exit {
   widgetClassPath=&WBP_UIKit_Button_Regular
   bIsVariable=true
   parent=InfoContainer
   widget=Button_Exit
   properties {minWidth=224, navigation=&self:WidgetTree.Button_Exit.WidgetNavigation_0, text=Close Fortnite}
   slot {class=VerticalBoxSlot, properties {horizontalAlignment=HAlign_Left, padding=<ltrb 0 32 0 0>}}
  }
  IdleState {widgetClassPath=&Overlay, bIsVariable=true, parent=StateSwitcher, widget=IdleState, slot {class=CommonVisibilitySwitcherSlot}}
  Background_Idle {
   widgetClassPath=&Image
   parent=IdleState
   widget=Background_Idle
   properties {
    visibility=SelfHitTestInvisible
    brush {
     bIsDynamicallyLoaded=false
     drawAs=Image
     imageType=FullColor
     mirroring=NoMirror
     resourceName=None
     resourceObject=/Cilantro/MaterialInstances/MI_UI_PlaytimeControl_Background_Idle
     tiling=NoTile
     imageSize=<32 32>
     margin=<ltrb 0 0 0 0>
     outlineSettings {
      bUseBrushTransparency=false
      roundingType=HalfHeightRadius
      width=0
      color {colorUseRule=UseColor_Specified, specifiedColor=<rgba 0 0 0 0>}
      cornerRadii=<0 0 0 0>
     }
     tintColor {colorUseRule=UseColor_Specified, specifiedColor=<rgba 1 1 1 1>}
     uVRegion {bIsValid=false, max=<0 0>, min=<0 0>}
    }
   }
   slot {class=OverlaySlot, properties {horizontalAlignment=HAlign_Fill, verticalAlignment=VAlign_Fill}}
  }
  FortVisualAttachment_0 {
   widgetClassPath=&FortVisualAttachment
   parent=IdleState
   widget=FortVisualAttachment_0
   properties {bOverride_HeightOverride=true, bOverride_WidthOverride=true, heightOverride=512, widthOverride=512, contentAnchor=<0.5 0.5>}
   slot {class=OverlaySlot, properties {horizontalAlignment=HAlign_Right, verticalAlignment=VAlign_Center}}
  }
  Idle_Lama {
   widgetClassPath=&Image
   bIsVariable=true
   parent=FortVisualAttachment_0
   widget=Idle_Lama
   properties {
    visibility=SelfHitTestInvisible
    brush {
     bIsDynamicallyLoaded=false
     drawAs=Image
     imageType=FullColor
     mirroring=NoMirror
     resourceName=None
     resourceObject=/Cilantro/MaterialInstances/MI_UI_PlaytimeControl_Lama
     tiling=NoTile
     imageSize=<32 32>
     margin=<ltrb 0 0 0 0>
     outlineSettings {
      bUseBrushTransparency=false
      roundingType=HalfHeightRadius
      width=0
      color {colorUseRule=UseColor_Specified, specifiedColor=<rgba 0 0 0 0>}
      cornerRadii=<0 0 0 0>
     }
     tintColor {colorUseRule=UseColor_Specified, specifiedColor=<rgba 1 1 1 1>}
     uVRegion {bIsValid=false, max=<0 0>, min=<0 0>}
    }
   }
   slot {class=SizeBoxSlot}
  }
  WBP_CaptureForPostBufferUpdate_1 {
   widgetClassPath=&WBP_CaptureForPostBufferUpdate
   parent=IdleState
   widget=WBP_CaptureForPostBufferUpdate_1
   slot {class=OverlaySlot, properties {horizontalAlignment=HAlign_Fill, verticalAlignment=VAlign_Fill}}
  }
  Idle_InfoContainer {
   widgetClassPath=&VerticalBox
   parent=IdleState
   widget=Idle_InfoContainer
   slot {class=OverlaySlot, properties {horizontalAlignment=HAlign_Center, verticalAlignment=VAlign_Center}}
  }
  Text_Status_1 {
   widgetClassPath=&CommonTextBlock
   bIsVariable=true
   parent=Idle_InfoContainer
   widget=Text_Status_1
   properties {
    style=&TextStyle-Base_C
    text=Are you Still There?
    font {
     bForceMonospaced=false
     bMaterialIsStencil=false
     fontMaterial=null
     fontObject=&NotoSans
     letterSpacing=0
     monospacedWidth=1
     size=24
     skewAmount=0
     typefaceFontName=NotoSans-Regular
     outlineSettings {
      bApplyOutlineToDropShadows=false
      bMiteredCorners=false
      bSeparateFillAlpha=false
      outlineMaterial=null
      outlineSize=0
      outlineColor=<rgba 0 0 0 1>
     }
    }
    shadowOffset=<0 0>
   }
   slot {class=VerticalBoxSlot, properties {horizontalAlignment=HAlign_Center, verticalAlignment=VAlign_Top}}
  }
  Text_Message_1 {
   widgetClassPath=&CommonTextBlock
   bIsVariable=true
   parent=Idle_InfoContainer
   widget=Text_Message_1
   properties {
    autoWrapText=true
    justification=Center
    style=&TextStyle-Base_C
    text=Fortnite will enter sleep mode soon.
    wrapTextAt=752
    font {
     bForceMonospaced=false
     bMaterialIsStencil=false
     fontMaterial=null
     fontObject=&NotoSans
     letterSpacing=0
     monospacedWidth=1
     size=24
     skewAmount=0
     typefaceFontName=NotoSans-Regular
     outlineSettings {
      bApplyOutlineToDropShadows=false
      bMiteredCorners=false
      bSeparateFillAlpha=false
      outlineMaterial=null
      outlineSize=0
      outlineColor=<rgba 0 0 0 1>
     }
    }
    shadowOffset=<0 0>
   }
   slot {class=VerticalBoxSlot, properties {verticalAlignment=VAlign_Top, padding=<ltrb 0 16 0 0>, size {sizeRule=Automatic, value=0.8}}}
  }
  Button_KeepPlaying {
   widgetClassPath=&WBP_UIKit_Button_Regular
   bIsVariable=true
   parent=Idle_InfoContainer
   widget=Button_KeepPlaying
   properties {minWidth=544, text=PRESS ANY BUTTON TO KEEP PLAYING}
   slot {class=VerticalBoxSlot, properties {horizontalAlignment=HAlign_Center, padding=<ltrb 0 32 0 0>}}
  }
  ProgressBar_SleepTime {
   widgetClassPath=/FortUILibrary/UIKit/ProgressBar/WBP_UIKit_ProgressBar
   bIsVariable=true
   parent=Idle_InfoContainer
   widget=ProgressBar_SleepTime
   properties {
    glowIntensity=1
    height=16
    progressValue=0
    useCircular=false
    useGlowColor=false
    width=640
    colorA=<rgba 0 0.8714 0.9823 1>
    colorB=<rgba 0.9823 0.9823 0.9823 1>
    emptyColor=<rgba 0 0 0 0.5>
    glowColor=<rgba 1 0 0 1>
   }
   slot {class=VerticalBoxSlot, properties {horizontalAlignment=HAlign_Center, verticalAlignment=VAlign_Top, padding=<ltrb 0 24 0 0>}}
  }
  Text_Idle_Infos {
   widgetClassPath=&CommonRichTextBlock
   bIsVariable=true
   parent=Idle_InfoContainer
   widget=Text_Idle_Infos
   properties {
    autoWrapText=true
    justification=Center
    text="<img id=\"Voice\"/> <Bold>Voice chat</> and <img id=\"Playtime\"/> <Bold>Play time limits</> will be paused when Fortnite enters sleep mode."
    textStyleSet=/Cilantro/RichTextStyles/DT_RichText_Cilantro_InfoStyle
    wrapTextAt=752
    decoratorClasses [/Cilantro/RichTextStyles/BP_UI_Cilantro_RichTextBlockDecorator.BP_UI_Cilantro_RichTextBlockDecorator_C]
    defaultTextStyleOverride {
     overflowPolicy=Clip
     transformPolicy=None
     colorAndOpacity {colorUseRule=UseColor_Specified, specifiedColor=<rgba 0.5776 0.5776 0.5776 1>}
     font {
      bForceMonospaced=false
      bMaterialIsStencil=false
      fontMaterial=null
      fontObject=&HeadingNow
      letterSpacing=10
      monospacedWidth=1
      size=15.75
      skewAmount=0
      typefaceFontName=74 Regular
      outlineSettings {
       bApplyOutlineToDropShadows=false
       bMiteredCorners=false
       bSeparateFillAlpha=false
       outlineMaterial=null
       outlineSize=0
       outlineColor=<rgba 0 0 0 1>
      }
     }
     highlightColor {colorUseRule=UseColor_Specified, specifiedColor=<rgba 0 0 0 1>}
     highlightShape {
      bIsDynamicallyLoaded=false
      drawAs=Image
      imageType=NoImage
      mirroring=NoMirror
      resourceName=None
      resourceObject=null
      tiling=NoTile
      imageSize=<32 32>
      margin=<ltrb 0 0 0 0>
      outlineSettings {
       bUseBrushTransparency=false
       roundingType=HalfHeightRadius
       width=0
       color {colorUseRule=UseColor_Specified, specifiedColor=<rgba 0 0 0 0>}
       cornerRadii=<0 0 0 0>
      }
      tintColor {colorUseRule=UseColor_Specified, specifiedColor=<rgba 1 1 1 1>}
      uVRegion {bIsValid=false, max=<0 0>, min=<0 0>}
     }
     selectedBackgroundColor {colorUseRule=UseColor_Foreground, specifiedColor=<rgba 1 0 1 1>}
     shadowColorAndOpacity=<rgba 0 0 0 1>
     shadowOffset=<0 0>
     strikeBrush {
      bIsDynamicallyLoaded=false
      drawAs=Image
      imageType=NoImage
      mirroring=NoMirror
      resourceName=None
      resourceObject=null
      tiling=NoTile
      imageSize=<32 32>
      margin=<ltrb 0 0 0 0>
      outlineSettings {
       bUseBrushTransparency=false
       roundingType=HalfHeightRadius
       width=0
       color {colorUseRule=UseColor_Specified, specifiedColor=<rgba 0 0 0 0>}
       cornerRadii=<0 0 0 0>
      }
      tintColor {colorUseRule=UseColor_Specified, specifiedColor=<rgba 1 1 1 1>}
      uVRegion {bIsValid=false, max=<0 0>, min=<0 0>}
     }
     underlineBrush {
      bIsDynamicallyLoaded=false
      drawAs=Image
      imageType=NoImage
      mirroring=NoMirror
      resourceName=None
      resourceObject=null
      tiling=NoTile
      imageSize=<32 32>
      margin=<ltrb 0 0 0 0>
      outlineSettings {
       bUseBrushTransparency=false
       roundingType=HalfHeightRadius
       width=0
       color {colorUseRule=UseColor_Specified, specifiedColor=<rgba 0 0 0 0>}
       cornerRadii=<0 0 0 0>
      }
      tintColor {colorUseRule=UseColor_Specified, specifiedColor=<rgba 1 1 1 1>}
      uVRegion {bIsValid=false, max=<0 0>, min=<0 0>}
     }
    }
   }
   slot {class=VerticalBoxSlot, properties {verticalAlignment=VAlign_Top, padding=<ltrb 0 32 0 0>}}
  }
  SleepState {widgetClassPath=&Overlay, bIsVariable=true, parent=StateSwitcher, widget=SleepState, slot {class=CommonVisibilitySwitcherSlot}}
  Background_Sleep {
   widgetClassPath=&Image
   parent=SleepState
   widget=Background_Sleep
   properties {
    visibility=SelfHitTestInvisible
    brush {
     bIsDynamicallyLoaded=false
     drawAs=Image
     imageType=FullColor
     mirroring=NoMirror
     resourceName=None
     resourceObject=/Cilantro/MaterialInstances/MI_UI_PlaytimeControl_Background_Sleep
     tiling=NoTile
     imageSize=<32 32>
     margin=<ltrb 0 0 0 0>
     outlineSettings {
      bUseBrushTransparency=false
      roundingType=HalfHeightRadius
      width=0
      color {colorUseRule=UseColor_Specified, specifiedColor=<rgba 0 0 0 0>}
      cornerRadii=<0 0 0 0>
     }
     tintColor {colorUseRule=UseColor_Specified, specifiedColor=<rgba 1 1 1 1>}
     uVRegion {bIsValid=false, max=<0 0>, min=<0 0>}
    }
   }
   slot {class=OverlaySlot, properties {horizontalAlignment=HAlign_Fill, verticalAlignment=VAlign_Fill}}
  }
  WBP_CaptureForPostBufferUpdate_2 {
   widgetClassPath=&WBP_CaptureForPostBufferUpdate
   parent=SleepState
   widget=WBP_CaptureForPostBufferUpdate_2
   slot {class=OverlaySlot, properties {horizontalAlignment=HAlign_Fill, verticalAlignment=VAlign_Fill}}
  }
  VerticalBox_0 {
   widgetClassPath=&VerticalBox
   parent=SleepState
   widget=VerticalBox_0
   slot {class=OverlaySlot, properties {horizontalAlignment=HAlign_Center, verticalAlignment=VAlign_Center}}
  }
  Text_Status_2 {
   widgetClassPath=&CommonTextBlock
   bIsVariable=true
   parent=VerticalBox_0
   widget=Text_Status_2
   properties {
    style=&TextStyle-Base_C
    text=Fortnite is in Sleep Mode
    font {
     bForceMonospaced=false
     bMaterialIsStencil=false
     fontMaterial=null
     fontObject=&NotoSans
     letterSpacing=0
     monospacedWidth=1
     size=24
     skewAmount=0
     typefaceFontName=NotoSans-Regular
     outlineSettings {
      bApplyOutlineToDropShadows=false
      bMiteredCorners=false
      bSeparateFillAlpha=false
      outlineMaterial=null
      outlineSize=0
      outlineColor=<rgba 0 0 0 1>
     }
    }
    shadowOffset=<0 0>
   }
   slot {class=VerticalBoxSlot, properties {horizontalAlignment=HAlign_Center, verticalAlignment=VAlign_Top}}
  }
  Button_Resume {
   widgetClassPath=&WBP_UIKit_Button_Regular
   bIsVariable=true
   parent=VerticalBox_0
   widget=Button_Resume
   properties {minWidth=196, navigation=&self:WidgetTree.Button_Resume.WidgetNavigation_0, text=Keep Playing}
   slot {class=VerticalBoxSlot, properties {horizontalAlignment=HAlign_Center, padding=<ltrb 0 24 0 0>}}
  }
  Button_Close {
   widgetClassPath=&WBP_UIKit_Button_Regular
   bIsVariable=true
   parent=VerticalBox_0
   widget=Button_Close
   properties {minWidth=196, navigation=&self:WidgetTree.Button_Close.WidgetNavigation_0, text=Close Fortnite}
   slot {class=VerticalBoxSlot, properties {horizontalAlignment=HAlign_Center, padding=<ltrb 0 24 0 0>}}
  }
  Text_Sleep_Infos {
   widgetClassPath=&CommonRichTextBlock
   bIsVariable=true
   parent=VerticalBox_0
   widget=Text_Sleep_Infos
   properties {
    autoWrapText=true
    justification=Center
    text="<img id=\"Voice\"/> <Bold>Voice chat</> and <img id=\"Playtime\"/> <Bold>Play time limits</> are paused while Fortnite is in Sleep Mode. They will resume when you start playing again."
    textStyleSet=&DT_RichText_Cilantro_InfoStyle
    wrapTextAt=752
    decoratorClasses [&BP_UI_Cilantro_RichTextBlockDecorator_C]
    defaultTextStyleOverride {
     overflowPolicy=Clip
     transformPolicy=None
     colorAndOpacity {colorUseRule=UseColor_Specified, specifiedColor=<rgba 0.5776 0.5776 0.5776 1>}
     font {
      bForceMonospaced=false
      bMaterialIsStencil=false
      fontMaterial=null
      fontObject=&HeadingNow
      letterSpacing=10
      monospacedWidth=1
      size=15.75
      skewAmount=0
      typefaceFontName=74 Regular
      outlineSettings {
       bApplyOutlineToDropShadows=false
       bMiteredCorners=false
       bSeparateFillAlpha=false
       outlineMaterial=null
       outlineSize=0
       outlineColor=<rgba 0 0 0 1>
      }
     }
     highlightColor {colorUseRule=UseColor_Specified, specifiedColor=<rgba 0 0 0 1>}
     highlightShape {
      bIsDynamicallyLoaded=false
      drawAs=Image
      imageType=NoImage
      mirroring=NoMirror
      resourceName=None
      resourceObject=null
      tiling=NoTile
      imageSize=<32 32>
      margin=<ltrb 0 0 0 0>
      outlineSettings {
       bUseBrushTransparency=false
       roundingType=HalfHeightRadius
       width=0
       color {colorUseRule=UseColor_Specified, specifiedColor=<rgba 0 0 0 0>}
       cornerRadii=<0 0 0 0>
      }
      tintColor {colorUseRule=UseColor_Specified, specifiedColor=<rgba 1 1 1 1>}
      uVRegion {bIsValid=false, max=<0 0>, min=<0 0>}
     }
     selectedBackgroundColor {colorUseRule=UseColor_Foreground, specifiedColor=<rgba 1 0 1 1>}
     shadowColorAndOpacity=<rgba 0 0 0 1>
     shadowOffset=<0 0>
     strikeBrush {
      bIsDynamicallyLoaded=false
      drawAs=Image
      imageType=NoImage
      mirroring=NoMirror
      resourceName=None
      resourceObject=null
      tiling=NoTile
      imageSize=<32 32>
      margin=<ltrb 0 0 0 0>
      outlineSettings {
       bUseBrushTransparency=false
       roundingType=HalfHeightRadius
       width=0
       color {colorUseRule=UseColor_Specified, specifiedColor=<rgba 0 0 0 0>}
       cornerRadii=<0 0 0 0>
      }
      tintColor {colorUseRule=UseColor_Specified, specifiedColor=<rgba 1 1 1 1>}
      uVRegion {bIsValid=false, max=<0 0>, min=<0 0>}
     }
     underlineBrush {
      bIsDynamicallyLoaded=false
      drawAs=Image
      imageType=NoImage
      mirroring=NoMirror
      resourceName=None
      resourceObject=null
      tiling=NoTile
      imageSize=<32 32>
      margin=<ltrb 0 0 0 0>
      outlineSettings {
       bUseBrushTransparency=false
       roundingType=HalfHeightRadius
       width=0
       color {colorUseRule=UseColor_Specified, specifiedColor=<rgba 0 0 0 0>}
       cornerRadii=<0 0 0 0>
      }
      tintColor {colorUseRule=UseColor_Specified, specifiedColor=<rgba 1 1 1 1>}
      uVRegion {bIsValid=false, max=<0 0>, min=<0 0>}
     }
    }
   }
   slot {class=VerticalBoxSlot, properties {verticalAlignment=VAlign_Top, padding=<ltrb 0 32 0 0>}}
  }
  SafeZone_SettingsReminder {
   widgetClassPath=&SafeZone
   bIsVariable=true
   parent=Overlay_0
   widget=SafeZone_SettingsReminder
   properties {padTop=false, visibility=Collapsed}
   slot {class=OverlaySlot, properties {horizontalAlignment=HAlign_Center, verticalAlignment=VAlign_Bottom, padding=<ltrb 0 0 0 48>}}
  }
  O_SettingsReminder {widgetClassPath=&Overlay, parent=SafeZone_SettingsReminder, widget=O_SettingsReminder, slot {class=SafeZoneSlot}}
  WBP_UIKit_Backplate {
   widgetClassPath=&WBP_UIKit_Backplate
   parent=O_SettingsReminder
   widget=WBP_UIKit_Backplate
   properties {brightness=Dark}
   slot {class=OverlaySlot, properties {horizontalAlignment=HAlign_Fill, verticalAlignment=VAlign_Fill}}
  }
  Text_SettingsReminder {
   widgetClassPath=&CommonRichTextBlock
   parent=O_SettingsReminder
   widget=Text_SettingsReminder
   properties {
    text="You can configure Sleep Mode in <img id=\"Settings\"/> <Bold>Settings</> under the Game tab."
    textStyleSet=&DT_RichText_Cilantro_InfoStyle
    decoratorClasses [&BP_UI_Cilantro_RichTextBlockDecorator_C]
    defaultTextStyleOverride {
     overflowPolicy=Clip
     transformPolicy=None
     colorAndOpacity {colorUseRule=UseColor_Specified, specifiedColor=<rgba 0.5776 0.5776 0.5776 1>}
     font {
      bForceMonospaced=false
      bMaterialIsStencil=false
      fontMaterial=null
      fontObject=&HeadingNow
      letterSpacing=10
      monospacedWidth=1
      size=15.75
      skewAmount=0
      typefaceFontName=74 Regular
      outlineSettings {
       bApplyOutlineToDropShadows=false
       bMiteredCorners=false
       bSeparateFillAlpha=false
       outlineMaterial=null
       outlineSize=0
       outlineColor=<rgba 0 0 0 1>
      }
     }
     highlightColor {colorUseRule=UseColor_Specified, specifiedColor=<rgba 0 0 0 1>}
     highlightShape {
      bIsDynamicallyLoaded=false
      drawAs=Image
      imageType=NoImage
      mirroring=NoMirror
      resourceName=None
      resourceObject=null
      tiling=NoTile
      imageSize=<32 32>
      margin=<ltrb 0 0 0 0>
      outlineSettings {
       bUseBrushTransparency=false
       roundingType=HalfHeightRadius
       width=0
       color {colorUseRule=UseColor_Specified, specifiedColor=<rgba 0 0 0 0>}
       cornerRadii=<0 0 0 0>
      }
      tintColor {colorUseRule=UseColor_Specified, specifiedColor=<rgba 1 1 1 1>}
      uVRegion {bIsValid=false, max=<0 0>, min=<0 0>}
     }
     selectedBackgroundColor {colorUseRule=UseColor_Foreground, specifiedColor=<rgba 1 0 1 1>}
     shadowColorAndOpacity=<rgba 0 0 0 1>
     shadowOffset=<0 0>
     strikeBrush {
      bIsDynamicallyLoaded=false
      drawAs=Image
      imageType=NoImage
      mirroring=NoMirror
      resourceName=None
      resourceObject=null
      tiling=NoTile
      imageSize=<32 32>
      margin=<ltrb 0 0 0 0>
      outlineSettings {
       bUseBrushTransparency=false
       roundingType=HalfHeightRadius
       width=0
       color {colorUseRule=UseColor_Specified, specifiedColor=<rgba 0 0 0 0>}
       cornerRadii=<0 0 0 0>
      }
      tintColor {colorUseRule=UseColor_Specified, specifiedColor=<rgba 1 1 1 1>}
      uVRegion {bIsValid=false, max=<0 0>, min=<0 0>}
     }
     underlineBrush {
      bIsDynamicallyLoaded=false
      drawAs=Image
      imageType=NoImage
      mirroring=NoMirror
      resourceName=None
      resourceObject=null
      tiling=NoTile
      imageSize=<32 32>
      margin=<ltrb 0 0 0 0>
      outlineSettings {
       bUseBrushTransparency=false
       roundingType=HalfHeightRadius
       width=0
       color {colorUseRule=UseColor_Specified, specifiedColor=<rgba 0 0 0 0>}
       cornerRadii=<0 0 0 0>
      }
      tintColor {colorUseRule=UseColor_Specified, specifiedColor=<rgba 1 1 1 1>}
      uVRegion {bIsValid=false, max=<0 0>, min=<0 0>}
     }
    }
   }
   slot {class=OverlaySlot, properties {horizontalAlignment=HAlign_Fill, verticalAlignment=VAlign_Center, padding=<ltrb 24 16 24 16>}}
  }
  WidgetStack_Takeover {
   widgetClassPath=/Script/CommonUI.CommonActivatableWidgetStack
   bInherited=true
   bIsVariable=true
   parent=Overlay_0
   widget=WidgetStack_Takeover
   slot {class=OverlaySlot, properties {horizontalAlignment=HAlign_Fill, verticalAlignment=VAlign_Fill}}
  }
 ]
}
