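# Playtime schedule widget that lists each day's time limit and time windows, with an optional warning that days without a window aren't playable.
# OnScheduleTextsUpdated rebuilds ScheduleEntryBox with a header row (custom title or "SCHEDULE"), then one WBP_PlaytimeScheduleEntry per data entry.
# Each entry's time windows are joined with line breaks through TimeWindowText, which is cleared per entry.
# On initialization, the bEnableThyme runtime option controls divider visibility.
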WidgetBlueprint /Cilantro/WBP_PlaytimeSchedule
graphs [
 EventGraph {nodes [
   n1 = Event OnScheduleTextsUpdated!() -> (
    OutputDelegate
    then
    UpdatedScheduleEntries:Array of Playtime Schedule Entry Data Structures
    CustomTitle:"Text (by ref)"
    bShowDisclaimer:bool
   ) {is=title K2Node_Event}
   > Reset!(
    execute
    self:Dynamic Entry Box Object Reference=Get ScheduleEntryBox() -> (ScheduleEntryBox:Fort Dynamic Entry Box Object Reference)
    bDeleteWidgets:bool=false
   ) -> (then)
   > SetDisclaimerVisibility!(execute, self, IsVisible:bool=@n1.bShowDisclaimer) -> (then)
   > n2 = Sequence()
   > n3 = Create Entry!(execute, self:Dynamic Entry Box Object Reference=Get ScheduleEntryBox()) -> (then, ReturnValue:User Widget Object Reference)
   > n4 = Cast To WBP_PlaytimeScheduleEntry!(execute, Object=@n3) -> (then, CastFailed, AsWBP Playtime Schedule Entry:WBP Playtime Schedule Entry Object Reference) {is=title K2Node_DynamicCast}
   > SetTexts!(
    execute
    self:WBP Playtime Schedule Entry Object Reference=@n4
    Day:Text=SelectText(A="LOCTEXT(\"<Strong>SCHEDULE</>\")", B=@n1.CustomTitle, bPickA=TextIsEmpty(InText=@n1.CustomTitle))
    AllowedTime:Text="LOCTEXT(\"<Strong>TIME LIMIT</>\")"
    TimeWindows:Text="LOCTEXT(\"<Strong>TIME WINDOW</>\")"
   ) -> (then)
   n5 = For Each Loop(Exec=@n2.then_1, Array:Array of Playtime Schedule Entry Data Structures=@n1.UpdatedScheduleEntries) -> (Array Element:Playtime Schedule Entry Data Structure)
   > Set TimeWindowText(TimeWindowText:Text) -> (Output_Get:Text)
   n6 = Break Playtime Schedule Entry Data!(PlaytimeScheduleEntryData:"Playtime Schedule Entry Data Structure (by ref)"=@n5) -> (DayText:Text, AllowedTimeText:Text, TimeWindowTexts:Array of Texts) {is=title K2Node_BreakStruct}
   > n7 = For Each Loop(Array:Array of Texts=@n6.TimeWindowTexts) -> (Array Element:Text)
   n8 = Create Entry(execute=@n7.Completed, self=Get ScheduleEntryBox())
   > n9 = Cast To WBP_PlaytimeScheduleEntry(Object=@n8)
   > SetTexts(self=@n9, Day=@n6, AllowedTime=@n6.AllowedTimeText, TimeWindows=Get TimeWindowText() -> (TimeWindowText:Text))
   n10 = Get TimeWindowText()
   Set TimeWindowText(
    execute=@n7.LoopBody
    TimeWindowText=SelectText(A=@n7, B=Format Text!(Format="INVTEXT(\"{1}\\r\\n{2}\")", 1:Text=@n10, 2:Text=@n7) -> (Result), bPickA=TextIsEmpty(InText=@n10))
   )
   Event OnInitialized()
   n11 = GetRuntimeOptions()
   > SetVisibility(
    self=Get Image_Divider() -> (Image_Divider:Image Object Reference)
    InVisibility=Select(
     Option 0=Collapsed
     Option 1=SelfHitTestInvisible
     Index=Get bEnableThyme!(self:Fort Runtime Options Object Reference=@n11) -> (bEnableThyme:bool)
    )
   )
  ]}
 SetDisclaimerVisibility {nodes [
   n1 = SetDisclaimerVisibility!() -> (then, IsVisible:bool=~) {is=title K2Node_FunctionEntry}
   > SetVisibility(
    self=Get Text_Disclaimer() -> (Text_Disclaimer:Common Rich Text Block Object Reference)
    InVisibility="To Visibility (Boolean)"!(bIsVisible:bool=@n1, TrueVisibility:ESlateVisibility Enum=SelfHitTestInvisible, FalseVisibility:ESlateVisibility Enum=Collapsed) -> (ReturnValue:ESlateVisibility Enum=Visible)
   )
  ]}
]
variables [TimeWindowText {type=string, default=""}]
widgets {
 info {inheritedWidgetCount=0, parentClass=/Script/CilantroUI.FortPlaytimeScheduleWidget, rootWidgetClass=&GridPanel, widgetCount=4}
 widgets [
  GP_Content {widgetClassPath=&GridPanel, widget=GP_Content, properties {columnFill [1], rowFill [0,1]}}
  Image_Divider {
   widgetClassPath=&Image
   bIsVariable=true
   parent=GP_Content
   widget=Image_Divider
   properties {
    brush {
     bIsDynamicallyLoaded=false
     drawAs=Image
     imageType=NoImage
     mirroring=NoMirror
     resourceName=None
     resourceObject=null
     tiling=NoTile
     imageSize=<32 2>
     margin=<ltrb 0 0 0 0>
     outlineSettings {
      bUseBrushTransparency=false
      roundingType=HalfHeightRadius
      width=0
      color {colorUseRule=UseColor_Specified, specifiedColor=<rgba 0 0 0 0>}
      cornerRadii=<0 0 0 0>
     }
     tintColor {colorUseRule=UseColor_Specified, specifiedColor=<rgba 1 1 1 1>}
     uVRegion {bIsValid=false, max=<0 0>, min=<0 0>}
    }
    colorAndOpacity=<rgba 1 1 1 0.6>
   }
   slot {class=GridSlot, properties {verticalAlignment=VAlign_Top, padding=<ltrb 0 0 0 20>}}
  }
  ScheduleEntryBox {
   widgetClassPath=/Script/FortniteUI.FortDynamicEntryBox
   bIsVariable=true
   parent=GP_Content
   widget=ScheduleEntryBox
   properties {
    entryBoxType=Vertical
    entryWidgetClass=&/WBP_PlaytimeScheduleEntry.WBP_PlaytimeScheduleEntry_C
    numDesignerPreviewEntries=5
    entrySpacing=<0 10>
   }
   slot {class=GridSlot, properties {row=1, rowSpan=1}}
  }
  Text_Disclaimer {
   widgetClassPath=&CommonRichTextBlock
   bIsVariable=true
   parent=GP_Content
   widget=Text_Disclaimer
   properties {
    autoWrapText=true
    bIsScrollingEnabled=false
    bOverrideDefaultStyle=true
    text="<Icon.Warning/> <DefaultAttention>Days with a time limit but no time window set won't be playable. Add a time window to allow play.</>"
    textStyleSet=/FortniteCoreUI/Settings/SettingsDescriptionStyles
    defaultTextStyleOverride {
     overflowPolicy=Clip
     transformPolicy=None
     colorAndOpacity {colorUseRule=UseColor_Specified, specifiedColor=<rgba 1 1 1 1>}
     font {
      bForceMonospaced=false
      bMaterialIsStencil=false
      fontMaterial=null
      fontObject=&BurbankSmall
      letterSpacing=0
      monospacedWidth=1
      size=15
      skewAmount=0
      typefaceFontName=Bold
      outlineSettings {
       bApplyOutlineToDropShadows=false
       bMiteredCorners=false
       bSeparateFillAlpha=false
       outlineMaterial=null
       outlineSize=0
       outlineColor=<rgba 0 0 0 1>
      }
     }
     highlightColor {colorUseRule=UseColor_Specified, specifiedColor=<rgba 0 0 0 1>}
     highlightShape {
      bIsDynamicallyLoaded=false
      drawAs=Image
      imageType=NoImage
      mirroring=NoMirror
      resourceName=None
      resourceObject=null
      tiling=NoTile
      imageSize=<32 32>
      margin=<ltrb 0 0 0 0>
      outlineSettings {
       bUseBrushTransparency=false
       roundingType=HalfHeightRadius
       width=0
       color {colorUseRule=UseColor_Specified, specifiedColor=<rgba 0 0 0 0>}
       cornerRadii=<0 0 0 0>
      }
      tintColor {colorUseRule=UseColor_Specified, specifiedColor=<rgba 1 1 1 1>}
      uVRegion {bIsValid=false, max=<0 0>, min=<0 0>}
     }
     selectedBackgroundColor {colorUseRule=UseColor_Foreground, specifiedColor=<rgba 1 0 1 1>}
     shadowColorAndOpacity=<rgba 0 0 0 1>
     shadowOffset=<0 0>
     strikeBrush {
      bIsDynamicallyLoaded=false
      drawAs=Image
      imageType=NoImage
      mirroring=NoMirror
      resourceName=None
      resourceObject=null
      tiling=NoTile
      imageSize=<32 32>
      margin=<ltrb 0 0 0 0>
      outlineSettings {
       bUseBrushTransparency=false
       roundingType=HalfHeightRadius
       width=0
       color {colorUseRule=UseColor_Specified, specifiedColor=<rgba 0 0 0 0>}
       cornerRadii=<0 0 0 0>
      }
      tintColor {colorUseRule=UseColor_Specified, specifiedColor=<rgba 1 1 1 1>}
      uVRegion {bIsValid=false, max=<0 0>, min=<0 0>}
     }
     underlineBrush {
      bIsDynamicallyLoaded=false
      drawAs=Image
      imageType=NoImage
      mirroring=NoMirror
      resourceName=None
      resourceObject=null
      tiling=NoTile
      imageSize=<32 32>
      margin=<ltrb 0 0 0 0>
      outlineSettings {
       bUseBrushTransparency=false
       roundingType=HalfHeightRadius
       width=0
       color {colorUseRule=UseColor_Specified, specifiedColor=<rgba 0 0 0 0>}
       cornerRadii=<0 0 0 0>
      }
      tintColor {colorUseRule=UseColor_Specified, specifiedColor=<rgba 1 1 1 1>}
      uVRegion {bIsValid=false, max=<0 0>, min=<0 0>}
     }
    }
   }
   slot {class=GridSlot, properties {row=2, padding=<ltrb 0 30 0 0>}}
  }
 ]
}
