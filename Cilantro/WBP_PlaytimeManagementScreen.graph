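# Fortnite parental-control Time Limits screen, a child widget of FortPlaytimeManagementScreen for playtime settings and customized schedules.
# A switcher toggles settings and schedule panels beside detail and schedule widgets; Exit returns from the schedule or closes the screen.
# An Allow Creative Time Mode dialog reports Thyme enablement, deferring screen exit in the playtime-blocked flow.
# Layout adapts to compact UI and the bEnableThyme runtime option.
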WidgetBlueprint /Cilantro/WBP_PlaytimeManagementScreen
graphs [
 EventGraph {nodes [
   "On Clicked (Button_CloseScreen)"!() -> (OutputDelegate, then, Button:Common Button Base Object Reference) {is=title K2Node_ComponentBoundEvent}
   n1 = Get Switcher_SubViews() -> (Switcher_SubViews:Common Activatable Widget Switcher Object Reference)
   > n2 = Branch(Condition="Equal (Object)"(
     A=Get Panel_PlaytimeScheduleSettings() -> (Panel_PlaytimeScheduleSettings:Fort Settings Panel Object Reference)
     B=GetActiveWidget!(self:Widget Switcher Object Reference=@n1) -> (ReturnValue:Widget Object Reference)
    ))
   > ShowPlaytimeSettings()
   ExitScreen(execute=@n2.else, self:Fort Playtime Management Screen Object Reference)
   ShowPlaytimeSettings() {is=title K2Node_CustomEvent}
   n3 = Get Panel_PlaytimeSettings() -> (Panel_PlaytimeSettings:Fort Settings Panel Object Reference)
   n4 = Get Switcher_SubViews()
   > SetActiveWidget(self:Widget Switcher Object Reference=@n4, Widget=@n3)
   > RefreshPlaytimeSettings(self:Fort Playtime Management Screen Object Reference)
   > "SetText (Text)"(self=Get ScreenTitle() -> (ScreenTitle:Common Text Object Reference), InText="LOCTEXT(\"Time Limits\")")
   > Fire Event Sub View Opened!(execute, View:Widget Object Reference=@n3, AdditionalInfo:Flow Analytics Additional Info Structure, bIsOverlay:bool=false) -> (then)
   ShowPlaytimeScheduleSettings() {is=title K2Node_CustomEvent}
   n5 = Get Panel_PlaytimeScheduleSettings()
   > SetActiveWidget(self=@n4, Widget=@n5)
   > RefreshPlaytimeScheduleSettings(self:Fort Playtime Management Screen Object Reference)
   > "SetText (Text)"(self=Get ScreenTitle(), InText="LOCTEXT(\"Customized Schedule\")")
   > Fire Event Sub View Opened(View=@n5, bIsOverlay=false)
   n6 = DialogButtonClickedDispatcher!() -> (OutputDelegate, then, ButtonIndex:int) {is=title K2Node_CustomEvent}
   > n7 = Switch on Int!(execute:Exec, Selection:int=@n6.ButtonIndex) -> (0:Exec, 1:Exec) {bHasDefaultPin=false}
   > NotifyThymeEnabled(self:Fort Playtime Management Screen Object Reference)
   > n8 = Branch(Condition=Get IsPlaytimeBlockedFlow() -> (IsPlaytimeBlockedFlow:bool))
   > Set IsClosingCreativeModal(IsClosingCreativeModal:bool=true) -> (Output_Get:bool)
   > n9 = Set IsPlaytimeBlockedFlow(IsPlaytimeBlockedFlow:bool=false) -> (Output_Get:bool)
   n10 = Get CreativeModalWidget() -> (CreativeModalWidget:WBP UIKit Dialog Object Reference)
   DeactivateWidget(execute=[@n11.then, @n9.then, @n7.1], self=@n10)
   > Unbind all Events from Dialog Button Clicked Dispatcher!(
    execute:Exec
    self:WBP UIKit Dialog Internal Object Reference=Get Dialog_Internal!(self:WBP UIKit Dialog Object Reference=@n10) -> (Dialog_Internal:WBP UIKit Dialog Internal Object Reference)
   ) -> (then:Exec) {is=title K2Node_ClearDelegate}
   n11 = ExitScreen(execute=@n8.else)
   Event OnShowAllowCreativeTimeModeModal() {is=title K2Node_Event}
   > n12 = Push Simple UIKit Modal Dialog!(
    execute
    Context:User Widget Object Reference=Self-Reference()
    VM Class:UI Kit Dialog Viewmodel Class Reference=Get CreativeDialogVMClass() -> (CreativeDialogVMClass:Dialog VM Allow Creative Time Mode Class Reference)
    OnDialogButtonClicked:"Delegate (by ref)"=Create Event()
   ) -> (then, DialogActivatableWidget:WBP UIKit Dialog Object Reference, DialogVM:UI Kit Dialog Viewmodel Object Reference)
   > Set CreativeModalWidget(CreativeModalWidget:WBP UIKit Dialog Object Reference=@n12) -> (Output_Get:WBP UIKit Dialog Object Reference)
   Event On Activated()
   > SetText(
    self=Get RichText_CurrentTimezone() -> (RichText_CurrentTimezone:Common Rich Text Block Object Reference)
    InText=Format Text!(
     Format="LOCTEXT(\"Schedule is based on the <strong>{Timezone}</> timezone. To change timezone, please visit fn.gg/account.\")"
     Timezone:Text=GetTimezoneText!(self:Fort Playtime Management Screen Object Reference) -> (ReturnValue:Text)
    ) -> (Result)
   )
   n13 = Event OnShowAllowCreativeModalPlaytimeBlocked!() -> (OutputDelegate, then, CreativeModal:Common Activatable Widget Object Reference) {is=title K2Node_Event}
   > n14 = Construct Object from Class(Class=Get CreativeDialogVMClass(), self=Self-Reference()) -> (ReturnValue:Dialog VM Allow Creative Time Mode Object Reference)
   > n15 = Cast To WBP_UIKit_Dialog!(execute, Object=@n13.CreativeModal) -> (then, CastFailed, AsWBP UIKit Dialog:WBP UIKit Dialog Object Reference) {is=title K2Node_DynamicCast}
   > n16 = Set CreativeModalWidget(CreativeModalWidget=@n15)
   > ApplyDialogVM!(execute, self:WBP UIKit Dialog Object Reference=@n16, DialogVM:UI Kit Dialog Viewmodel Object Reference=@n14) -> (then)
   > Bind Event to Dialog Button Clicked Dispatcher(self:WBP UIKit Dialog Internal Object Reference=Get Dialog_Internal(self=@n15), Delegate=Create Event()) {is=title K2Node_AddDelegate}
   > Set IsPlaytimeBlockedFlow(IsPlaytimeBlockedFlow=true)
   Event OnPlaytimeScheduleRequested() {is=title K2Node_Event}
   > ShowPlaytimeScheduleSettings() {is=title K2Node_CallFunction}
   Event Construct()
   > ShowPlaytimeSettings() {is=title K2Node_CallFunction}
   > SetScrollWidget!(execute, ToScroll:Widget Object Reference=Get ScrollBoxSchedule() -> (ScrollBoxSchedule:Scroll Box Object Reference)) -> (then)
   Event OnInitialized()
   > SetVisual()
   Event OnCheckIfClosingCreativeModal() {is=title K2Node_Event}
   > Branch(Condition=Get IsClosingCreativeModal() -> (IsClosingCreativeModal:bool))
   > Set IsClosingCreativeModal(IsClosingCreativeModal=false)
   > ExitScreen()
   Event PreConstruct()
   > SetCompactUIStyle() {is=title K2Node_CallFunction}
  ]}
 GetPrimaryContentSetup {nodes [
   GetPrimaryContentSetup() {is=title K2Node_FunctionEntry}
   > Return Node!(
    execute
    ReturnValue:Primary Content Setup Structure=Make Primary Content Setup!(
     bShowBottomBar:bool=false
     bShowChatWidget:bool=false
     bEnableNewTextChat:bool=False
     bShowTopBar:bool=false
     bTabSupportsInterface:bool=False
    ) -> (PrimaryContentSetup:Primary Content Setup Structure) {is=title K2Node_MakeStruct}
   )
  ]}
 SetCompactUIStyle {nodes [
   SetCompactUIStyle() {is=title K2Node_FunctionEntry}
   > Branch(Condition=IsUsingCompactUI())
   > SetPadding!(
    execute
    self:Vertical Box Slot Object Reference=SlotAsVerticalBoxSlot(Widget=Get VB_Content() -> (VB_Content:Vertical Box Object Reference))
    InPadding_Left:float=48
    InPadding_Top:float=24
    InPadding_Right:float=48
    InPadding_Bottom:float=24
   ) -> (then)
   > SetPadding(
    self:Border Slot Object Reference=SlotAsBorderSlot!(Widget:Widget Object Reference=Get Button_CloseScreen() -> (Button_CloseScreen:Icon Text Button Object Reference)) -> (ReturnValue:Border Slot Object Reference)
    InPadding_Left=0
    InPadding_Top=0
    InPadding_Right=48
    InPadding_Bottom=8
   )
   > SetSize!(execute, self:Spacer Object Reference=Get Spacer_68() -> (Spacer_68:Spacer Object Reference), InSize:Vector 2D Structure=(X=48,Y=1)) -> (then)
  ]}
 SetVisual {nodes [
   SetVisual() {is=title K2Node_FunctionEntry}
   > Set IsEnableThyme(IsEnableThyme:bool=Get bEnableThyme!(self:Fort Runtime Options Object Reference=GetRuntimeOptions()) -> (bEnableThyme:bool)) -> (Output_Get:bool)
   n1 = Get IsEnableThyme() -> (IsEnableThyme:bool)
   > SetPadding!(
    execute
    self:Vertical Box Slot Object Reference=[
     SlotAsVerticalBoxSlot(Widget=Get AdditionalScheduleWidget() -> (AdditionalScheduleWidget:Fort Playtime Schedule Widget Object Reference))
     SlotAsVerticalBoxSlot(Widget=Get PlaytimeScheduleWidget() -> (PlaytimeScheduleWidget:Fort Playtime Schedule Widget Object Reference))
    ]
    InPadding:Margin Structure=Make Margin!(Top=SelectFloat(A=32, B=20, bPickA=@n1)) -> (Margin)
   ) -> (then)
   n2 = Get IsEnableThyme()
   > SetSize(self=Get Spacer() -> (Spacer:Spacer Object Reference), InSize=MakeVector2D(X=1, Y=SelectFloat(A=32, B=24, bPickA=@n2)))
   n3 = Get IsEnableThyme()
   > SetSize(self=Get Spacer_68(), InSize=MakeVector2D(X=SelectFloat(A=96, B=48, bPickA=@n3), Y=1))
   > SetWidthOverride(
    self=Get SettingsContainer() -> (SettingsContainer:Size Box Object Reference)
    InWidthOverride=SelectFloat(A=792, B=800, bPickA=Get IsEnableThyme())
   )
   > SetPadding(
    self=SlotAsVerticalBoxSlot(Widget=Get VB_Content())
    InPadding=Select(
     Option 0:Margin Structure=Get VB_PaddingOld() -> (VB_PaddingOld:Margin Structure)
     Option 1:Margin Structure=Get VB_PaddingNew() -> (VB_PaddingNew:Margin Structure)
     Index=Get IsEnableThyme()
    ) -> (ReturnValue:Margin Structure)
   )
   n4 = Get IsEnableThyme()
   > SetPadding(
    self=SlotAsVerticalBoxSlot(Widget=Get SizeBox_0() -> (SizeBox_0:Size Box Object Reference))
    InPadding=Make Margin!(Left:float=SelectFloat(A=-10, B=0, bPickA=@n4), Top=0, Right:float=0, Bottom:float=0) -> (Margin)
   )
  ]}
]
variables [
 CreativeDialogVMClass {type=Class, default=/Cilantro/Dialogs/DialogVM_AllowCreativeTimeMode.DialogVM_AllowCreativeTimeMode_C}
 CreativeModalWidget {type=WBP_UIKit_Dialog_C, default=null}
 IsClosingCreativeModal {type=boolean}
 IsPlaytimeBlockedFlow {type=boolean}
]
widgets {
 info {inheritedWidgetCount=6, parentClass=/Script/CilantroUI.FortPlaytimeManagementScreen, rootWidgetClass=&Overlay, widgetCount=31}
 widgets [
  Overlay_0 {widgetClassPath=&Overlay, widget=Overlay_0, properties {clipping=ClipToBounds, visibility=Visible}}
  BackgroundGradient {
   widgetClassPath=&Image
   parent=Overlay_0
   widget=BackgroundGradient
   properties {
    visibility=SelfHitTestInvisible
    brush {
     bIsDynamicallyLoaded=false
     drawAs=Image
     imageType=NoImage
     mirroring=NoMirror
     resourceName=None
     resourceObject=/FortniteCoreUI/Frontend/Art/M_UI_RadialGradient_Right
     tiling=NoTile
     imageSize=<32 32>
     margin=<ltrb 0 0 0 0>
     outlineSettings {
      bUseBrushTransparency=false
      roundingType=HalfHeightRadius
      width=0
      color {colorUseRule=UseColor_Specified, specifiedColor=<rgba 0 0 0 0>}
      cornerRadii=<0 0 0 0>
     }
     tintColor {colorUseRule=UseColor_Specified, specifiedColor=<rgba 1 1 1 1>}
     uVRegion {bIsValid=false, max=<0 0>, min=<0 0>}
    }
   }
   slot {class=OverlaySlot, properties {horizontalAlignment=HAlign_Fill, verticalAlignment=VAlign_Fill}}
  }
  WBP_CaptureForPostBufferUpdate {
   widgetClassPath=&WBP_CaptureForPostBufferUpdate
   bIsVariable=true
   parent=Overlay_0
   slot {class=OverlaySlot}
  }
  SafeZone {
   widgetClassPath=&SafeZone
   bIsVariable=true
   parent=Overlay_0
   widget=SafeZone
   slot {class=OverlaySlot, properties {horizontalAlignment=HAlign_Fill, verticalAlignment=VAlign_Fill}}
  }
  Overlay_3 {widgetClassPath=&Overlay, parent=SafeZone, widget=Overlay_3, slot {class=SafeZoneSlot}}
  VerticalBox_0 {
   widgetClassPath=&VerticalBox
   parent=Overlay_3
   widget=VerticalBox_0
   slot {class=OverlaySlot, properties {horizontalAlignment=HAlign_Fill, verticalAlignment=VAlign_Fill, padding=<ltrb 0 0 0 80>}}
  }
  VB_Content {
   widgetClassPath=&VerticalBox
   bIsVariable=true
   parent=VerticalBox_0
   widget=VB_Content
   slot {class=VerticalBoxSlot, properties {padding=<ltrb 60 64 60 24>, size {sizeRule=Fill, value=1}}}
  }
  Border_0 {
   widgetClassPath=&Border
   parent=VB_Content
   widget=Border_0
   properties {
    background {
     bIsDynamicallyLoaded=false
     drawAs=NoDrawType
     imageType=NoImage
     mirroring=NoMirror
     resourceName=None
     resourceObject=null
     tiling=NoTile
     imageSize=<32 32>
     margin=<ltrb 0 0 0 0>
     outlineSettings {
      bUseBrushTransparency=false
      roundingType=HalfHeightRadius
      width=0
      color {colorUseRule=UseColor_Specified, specifiedColor=<rgba 0 0 0 0>}
      cornerRadii=<0 0 0 0>
     }
     tintColor {colorUseRule=UseColor_Specified, specifiedColor=<rgba 1 1 1 1>}
     uVRegion {bIsValid=false, max=<0 0>, min=<0 0>}
    }
    contentColorAndOpacity=<rgba 0.6038 1 1 1>
    padding=<ltrb 0 0 0 0>
   }
   slot {class=VerticalBoxSlot}
  }
  ScreenTitle {
   widgetClassPath=&CommonTextBlock
   bIsVariable=true
   parent=Border_0
   widget=ScreenTitle
   properties {
    style=/FortUILibrary/Styles/Typography/_TextStyle_SubH2._TextStyle_SubH2_C
    text=Time Limits
    textTransformPolicy=ToUpper
    font {
     bForceMonospaced=false
     bMaterialIsStencil=false
     fontMaterial=null
     fontObject=&Burbank
     letterSpacing=0
     monospacedWidth=1
     size=33
     skewAmount=0
     typefaceFontName=BurbankRegular-Black
     outlineSettings {
      bApplyOutlineToDropShadows=false
      bMiteredCorners=false
      bSeparateFillAlpha=false
      outlineMaterial=null
      outlineSize=0
      outlineColor=<rgba 0 0 0 1>
     }
    }
    shadowOffset=<0 0>
    strikeBrush {
     bIsDynamicallyLoaded=false
     drawAs=NoDrawType
     imageType=NoImage
     mirroring=NoMirror
     resourceName=None
     resourceObject=null
     tiling=NoTile
     imageSize=<32 32>
     margin=<ltrb 0 0 0 0>
     outlineSettings {
      bUseBrushTransparency=false
      roundingType=HalfHeightRadius
      width=0
      color {colorUseRule=UseColor_Specified, specifiedColor=<rgba 0 0 0 0>}
      cornerRadii=<0 0 0 0>
     }
     tintColor {colorUseRule=UseColor_Specified, specifiedColor=<rgba 1 1 1 1>}
     uVRegion {bIsValid=false, max=<0 0>, min=<0 0>}
    }
   }
   slot {class=BorderSlot, properties {padding=<ltrb 0 0 0 0>}}
  }
  Spacer {
   widgetClassPath=&Spacer
   bIsVariable=true
   parent=VB_Content
   widget=Spacer
   properties {size=<1 24>}
   slot {class=VerticalBoxSlot}
  }
  HorizontalBox_41 {
   widgetClassPath=&HorizontalBox
   parent=VB_Content
   widget=HorizontalBox_41
   slot {class=VerticalBoxSlot, properties {size {sizeRule=Fill, value=1}}}
  }
  VerticalBox_2 {
   widgetClassPath=&VerticalBox
   parent=HorizontalBox_41
   widget=VerticalBox_2
   slot {class=HorizontalBoxSlot, properties {size {sizeRule=Fill, value=1}}}
  }
  SizeBox_0 {
   widgetClassPath=&SizeBox
   bIsVariable=true
   parent=VerticalBox_2
   widget=SizeBox_0
   slot {class=VerticalBoxSlot, properties {size {sizeRule=Fill, value=1}}}
  }
  Switcher_SubViews {
   widgetClassPath=&CommonActivatableWidgetSwitcher
   bInherited=true
   parent=SizeBox_0
   widget=Switcher_SubViews
   properties {transitionDuration=0}
   slot {class=SizeBoxSlot}
  }
  Panel_PlaytimeSettings {
   widgetClassPath=/FortniteCoreUI/UI/ContentControls/SettingsPanel_ContentControls
   bInherited=true
   parent=Switcher_SubViews
   widget=Panel_PlaytimeSettings
   slot {class=WidgetSwitcherSlot, properties {verticalAlignment=VAlign_Top}}
  }
  Panel_PlaytimeScheduleSettings {
   widgetClassPath=&SettingsPanel_ContentControls
   bInherited=true
   parent=Switcher_SubViews
   widget=Panel_PlaytimeScheduleSettings
   slot {class=WidgetSwitcherSlot, properties {verticalAlignment=VAlign_Top}}
  }
  RichText_CurrentTimezone {
   widgetClassPath=&CommonRichTextBlock
   bIsVariable=true
   parent=VerticalBox_2
   widget=RichText_CurrentTimezone
   properties {
    autoWrapText=true
    mobileTextBlockScale=1.3
    text="Schedule is based on the <strong>UTC -04:00 Eastern Daylight Time (America/New_York)</> timezone. To change timezone, please visit fn.gg/account."
    textStyleSet=/FortniteCoreUI/Settings/SettingsDescriptionStyles
    defaultTextStyleOverride {
     overflowPolicy=Clip
     transformPolicy=None
     colorAndOpacity {colorUseRule=UseColor_Specified, specifiedColor=<rgba 1 1 1 1>}
     font {
      bForceMonospaced=false
      bMaterialIsStencil=false
      fontMaterial=null
      fontObject=&NotoSans
      letterSpacing=0
      monospacedWidth=1
      size=24
      skewAmount=0
      typefaceFontName=NotoSans-Regular
      outlineSettings {
       bApplyOutlineToDropShadows=false
       bMiteredCorners=false
       bSeparateFillAlpha=false
       outlineMaterial=null
       outlineSize=0
       outlineColor=<rgba 0 0 0 1>
      }
     }
     highlightColor {colorUseRule=UseColor_Specified, specifiedColor=<rgba 0 0 0 1>}
     highlightShape {
      bIsDynamicallyLoaded=false
      drawAs=Image
      imageType=NoImage
      mirroring=NoMirror
      resourceName=None
      resourceObject=null
      tiling=NoTile
      imageSize=<32 32>
      margin=<ltrb 0 0 0 0>
      outlineSettings {
       bUseBrushTransparency=false
       roundingType=HalfHeightRadius
       width=0
       color {colorUseRule=UseColor_Specified, specifiedColor=<rgba 0 0 0 0>}
       cornerRadii=<0 0 0 0>
      }
      tintColor {colorUseRule=UseColor_Specified, specifiedColor=<rgba 1 1 1 1>}
      uVRegion {bIsValid=false, max=<0 0>, min=<0 0>}
     }
     selectedBackgroundColor {colorUseRule=UseColor_Foreground, specifiedColor=<rgba 1 0 1 1>}
     shadowColorAndOpacity=<rgba 0 0 0 1>
     shadowOffset=<0 0>
     strikeBrush {
      bIsDynamicallyLoaded=false
      drawAs=Image
      imageType=NoImage
      mirroring=NoMirror
      resourceName=None
      resourceObject=null
      tiling=NoTile
      imageSize=<32 32>
      margin=<ltrb 0 0 0 0>
      outlineSettings {
       bUseBrushTransparency=false
       roundingType=HalfHeightRadius
       width=0
       color {colorUseRule=UseColor_Specified, specifiedColor=<rgba 0 0 0 0>}
       cornerRadii=<0 0 0 0>
      }
      tintColor {colorUseRule=UseColor_Specified, specifiedColor=<rgba 1 1 1 1>}
      uVRegion {bIsValid=false, max=<0 0>, min=<0 0>}
     }
     underlineBrush {
      bIsDynamicallyLoaded=false
      drawAs=Image
      imageType=NoImage
      mirroring=NoMirror
      resourceName=None
      resourceObject=null
      tiling=NoTile
      imageSize=<32 32>
      margin=<ltrb 0 0 0 0>
      outlineSettings {
       bUseBrushTransparency=false
       roundingType=HalfHeightRadius
       width=0
       color {colorUseRule=UseColor_Specified, specifiedColor=<rgba 0 0 0 0>}
       cornerRadii=<0 0 0 0>
      }
      tintColor {colorUseRule=UseColor_Specified, specifiedColor=<rgba 1 1 1 1>}
      uVRegion {bIsValid=false, max=<0 0>, min=<0 0>}
     }
    }
   }
   slot {class=VerticalBoxSlot, properties {padding=<ltrb 0 10 0 0>}}
  }
  Spacer_68 {
   widgetClassPath=&Spacer
   bIsVariable=true
   parent=HorizontalBox_41
   widget=Spacer_68
   properties {size=<48 1>}
   slot {class=HorizontalBoxSlot}
  }
  RightSideSwitcher {
   widgetClassPath=&CommonWidgetSwitcherLegacy
   bIsVariable=true
   parent=HorizontalBox_41
   widget=RightSideSwitcher
   slot {class=HorizontalBoxSlot, properties {size {sizeRule=Fill, value=1}}}
  }
  SettingsContainer {
   widgetClassPath=&SizeBox
   bIsVariable=true
   parent=RightSideSwitcher
   widget=SettingsContainer
   properties {bOverride_WidthOverride=true, widthOverride=800}
   slot {class=WidgetSwitcherSlot}
  }
  ScrollBoxSchedule {widgetClassPath=&ScrollBox, bIsVariable=true, parent=SettingsContainer, widget=ScrollBoxSchedule, slot {class=SizeBoxSlot}}
  VerticalBox_3 {
   widgetClassPath=&VerticalBox
   parent=ScrollBoxSchedule
   widget=VerticalBox_3
   slot {class=ScrollBoxSlot, properties {verticalAlignment=VAlign_Top}}
  }
  Details_Settings {
   widgetClassPath=/FortniteCoreUI/Settings/ParentalControls_DetailView
   bInherited=true
   bIsVariable=true
   parent=VerticalBox_3
   widget=Details_Settings
   slot {class=VerticalBoxSlot}
  }
  PlaytimeScheduleWidget {
   widgetClassPath=&/WBP_PlaytimeSchedule
   bInherited=true
   bIsVariable=true
   parent=VerticalBox_3
   widget=PlaytimeScheduleWidget
   slot {class=VerticalBoxSlot, properties {padding=<ltrb 0 20 0 0>}}
  }
  AdditionalScheduleWidget {
   widgetClassPath=&WBP_PlaytimeSchedule
   bInherited=true
   bIsVariable=true
   parent=VerticalBox_3
   widget=AdditionalScheduleWidget
   properties {visibility=Collapsed}
   slot {class=VerticalBoxSlot, properties {padding=<ltrb 0 20 0 0>}}
  }
  Footer {
   widgetClassPath=&Border
   parent=Overlay_0
   widget=Footer
   properties {
    horizontalAlignment=HAlign_Right
    verticalAlignment=VAlign_Bottom
    background {
     bIsDynamicallyLoaded=false
     drawAs=Image
     imageType=NoImage
     mirroring=NoMirror
     resourceName=None
     resourceObject=null
     tiling=NoTile
     imageSize=<32 32>
     margin=<ltrb 0 0 0 0>
     outlineSettings {
      bUseBrushTransparency=false
      roundingType=HalfHeightRadius
      width=0
      color {colorUseRule=UseColor_Specified, specifiedColor=<rgba 0 0 0 0>}
      cornerRadii=<0 0 0 0>
     }
     tintColor {colorUseRule=UseColor_Specified, specifiedColor=<rgba 0.0024 0.016 0.117 1>}
     uVRegion {bIsValid=false, max=<0 0>, min=<0 0>}
    }
    padding=<ltrb 0 8 0 0>
   }
   slot {class=OverlaySlot, properties {horizontalAlignment=HAlign_Fill, verticalAlignment=VAlign_Bottom}}
  }
  SafeZone_0 {
   widgetClassPath=&SafeZone
   bIsVariable=true
   parent=Footer
   widget=SafeZone_0
   properties {padLeft=false, padTop=false}
   slot {class=BorderSlot, properties {horizontalAlignment=HAlign_Right, verticalAlignment=VAlign_Bottom, padding=<ltrb 0 8 0 0>}}
  }
  Button_CloseScreen {
   widgetClassPath=&IconTextButton
   bIsVariable=true
   parent=SafeZone_0
   widget=Button_CloseScreen
   properties {
    buttonText=Exit
    onButtonBaseClicked=()
    onButtonBaseDoubleClicked=()
    onButtonBaseHovered=()
    onButtonBaseUnhovered=()
    onSelectedChangedBase=()
    style=/FortUILibrary/Foundation/Buttons/ButtonStyles/Skew/ButtonStyle-Skew_Error.ButtonStyle-Skew_Error_C
    padding=<ltrb 0 8 60 8>
    triggeringInputAction {dataTable=&InputActionDataTable, rowName=Close}
   }
   slot {class=SafeZoneSlot, properties {hAlign=HAlign_Right, vAlign=VAlign_Bottom}}
  }
  Overlay_SavingSettings {
   widgetClassPath=&Overlay
   parent=Overlay_0
   widget=Overlay_SavingSettings
   properties {visibility=Collapsed}
   slot {class=OverlaySlot, properties {horizontalAlignment=HAlign_Fill, verticalAlignment=VAlign_Fill}}
  }
  Image_67 {
   widgetClassPath=&Image
   bIsVariable=true
   parent=Overlay_SavingSettings
   widget=Image_67
   properties {brush {
     bIsDynamicallyLoaded=false
     drawAs=Image
     imageType=NoImage
     mirroring=NoMirror
     resourceName=None
     resourceObject=null
     tiling=NoTile
     imageSize=<32 32>
     margin=<ltrb 0 0 0 0>
     outlineSettings {
      bUseBrushTransparency=false
      roundingType=HalfHeightRadius
      width=0
      color {colorUseRule=UseColor_Specified, specifiedColor=<rgba 0 0 0 0>}
      cornerRadii=<0 0 0 0>
     }
     tintColor {colorUseRule=UseColor_Specified, specifiedColor=<rgba 0 0 0 0.5>}
     uVRegion {bIsValid=false, max=<0 0>, min=<0 0>}
    }}
   slot {class=OverlaySlot, properties {horizontalAlignment=HAlign_Fill, verticalAlignment=VAlign_Fill}}
  }
  CircularThrobber_72 {
   widgetClassPath=/Script/UMG.CircularThrobber
   bIsVariable=true
   parent=Overlay_SavingSettings
   widget=CircularThrobber_72
   slot {class=OverlaySlot, properties {horizontalAlignment=HAlign_Center, verticalAlignment=VAlign_Center}}
  }
 ]
}
